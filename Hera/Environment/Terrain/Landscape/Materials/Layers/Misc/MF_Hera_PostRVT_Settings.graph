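# Landscape material function for Hera terrain that adjusts material attributes after runtime virtual texture (RVT) sampling.
# A shading path switch separates a high-end branch with optional static-switched snow from a mobile branch adding biome-colored grass fresnel at Medium quality.
# An optional GoldLands input then overrides Metallic, before the landscape visibility mask sets OpacityMask and a Lumen modification function produces the output.
# Snow debug blend and brightness-mask nodes are left unconnected.
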MaterialFunction /Hera/Environment/Terrain/Landscape/Materials/Layers/Misc/MF_Hera_PostRVT_Settings
graph {nodes [
  n1 = GetMaterialAttributes(None=NamedRerouteUsage()) {attributeGetTypes [#1]}
  n2 = GetMaterialAttributes(None=FunctionInput(Preview=GetMaterialAttributes().MaterialAttributes) {bUsePreviewValueAsDefault=true, inputName=Snow_PostRVT, inputType=FunctionInput_MaterialAttributes, sortPriority=101}) {attributeGetTypes [#1, #2, #3]}
  n3 = SetMaterialAttributes!(
   MaterialAttributes=@n1.MaterialAttributes
   BaseColor=LinearInterpolate(A=@n1.BaseColor, B=@n2.BaseColor, Alpha=@n2.BaseColor)
   EmissiveColor=Multiply(A=@n2.EmissiveColor, B=@n2.EmissiveColor)
  ) {attributeSetTypes [#1, #2]}
  BlendMaterialAttributes(A=@n3, B=@n1.MaterialAttributes, Alpha=ScalarParameter() {group=Snow, parameterName=Snow_DEBUG_DynamicPostRVT01, sortPriority=-900})
  FunctionInput() {
   bUsePreviewValueAsDefault=true
   inputName=VariationMasksMobile
   inputType=FunctionInput_Vector4
   sortPriority=1
   previewValue=<0.7 0.7 0.7 0.7>
  }
  n4 = NamedRerouteUsage()
  n5 = ShadingPathSwitch(Default=@n4, Deferred=@n4, Mobile=NamedRerouteUsage())
  n6 = Constant()
  n7 = StaticSwitch(
   True=SetMaterialAttributes!(
    MaterialAttributes=@n5
    Metallic=FunctionInput(Preview=@n6) {bUsePreviewValueAsDefault=true, inputName=GoldLands_PostRVT, inputType=FunctionInput_Scalar, sortPriority=103}
   ) {attributeSetTypes [#4]}
   False=@n5
   Value=FunctionInput(Preview=StaticBool()) {bUsePreviewValueAsDefault=true, inputName=UseGoldLandsForBiomeC, inputType=FunctionInput_StaticBool, sortPriority=102}
  )
  n8 = LandscapeVisibilityMask()
  FunctionOutput(None=MaterialFunctionCall!(In=SetMaterialAttributes!(MaterialAttributes=@n7, OpacityMask=@n8) {attributeSetTypes [#3]}) {materialFunction=/Game/Athena/Asteria/Environments/Landscape/Materials/MaterialFunctions/MF_Terrain_LumenModification}.Result)
  n9 = MaterialFunctionCall(Float3=@n1.BaseColor) {materialFunction=&BreakOutFloat3Components}
  n10 = Max(A=Max(A=@n9.R, B=@n9.R), B=@n9.B)
  n11 = Min(A=Min(A=@n9.R, B=@n9.R), B=@n9.B)
  n12 = Power(Base=Multiply(A=Add(A=@n10, B=@n11), B=@n11), Exp=Constant() {r=4})
  n13 = Constant() {r=500}
  Multiply(A=Saturate(None=Multiply(A=@n12, B=@n13) {constB=2})) {desc=Multiply against the BiomeC layer to limit the potential mask area}
  n14 = FunctionInput(Preview=StaticBool()) {bUsePreviewValueAsDefault=true, inputName=UseSnowForBiomeC, inputType=FunctionInput_StaticBool, sortPriority=100}
  NamedRerouteDeclaration(None=SetMaterialAttributes!(MaterialAttributes=StaticSwitch(True=@n3, False=@n1.MaterialAttributes, Value=@n14))) {name=HighEnd PostRVT, nodeColor=<rgba 0 1 0.2969 1>}
  NamedRerouteDeclaration(None=FunctionInput() {inputName=Material, inputType=FunctionInput_MaterialAttributes, sortPriority=-500}) {name=Input Material, nodeColor=<rgba 0.1953 0 1 1>}
  n15 = NamedRerouteUsage()
  n16 = FunctionInput() {bUsePreviewValueAsDefault=true, inputName=SwitchGrassLayerMask, inputType=FunctionInput_Scalar, sortPriority=104}
  n17 = Saturate(None=Add(A=Multiply(A=@n16) {constB=3.5}) {constB=-1.5})
  n18 = SetMaterialAttributes!(
   MaterialAttributes=@n15
   BaseColor=VectorParameter() {group=Mobile Grass, parameterName=Mobile Grass Fresnel, defaultValue=<rgba 0.0919 0.099 0.0283 1>}.RGB
  ) {attributeSetTypes [#1]}
  n19 = FunctionInput(Preview=Constant()) {bUsePreviewValueAsDefault=true, inputName=TerrainMasks_QuadA, inputType=FunctionInput_Vector4, sortPriority=110}
  n20 = FunctionInput(Preview=Constant()) {bUsePreviewValueAsDefault=true, inputName=TerrainMasks_QuadB, inputType=FunctionInput_Vector4, sortPriority=111}
  n21 = FunctionInput(Preview=Constant()) {bUsePreviewValueAsDefault=true, inputName=TerrainMasks_QuadC, inputType=FunctionInput_Vector4, sortPriority=112}
  n22 = FunctionInput(Preview=Constant()) {bUsePreviewValueAsDefault=true, inputName=TerrainMasks_QuadD, inputType=FunctionInput_Vector4, sortPriority=113}
  n23 = Saturate(None=GetMaterialAttributes(None=MaterialFunctionCall!(
     Material=@n18
     GrassEdgeMask
     DesertGrass_SimpleNoise
     TerrainMasks_QuadA=@n19
     TerrainMasks_QuadB=@n20
     TerrainMasks_QuadC=@n21
     TerrainMasks_QuadD=@n22
     GrassNoise
     VariationMasks Mobile
    ) {materialFunction=/Hera/Environment/Terrain/Landscape/Materials/Layers/BiomeColor/MF_Hera_Grass_BiomeColor}.Result Mobile) {attributeGetTypes [#1]}.BaseColor)
  NamedRerouteDeclaration(None=QualitySwitch(
    Default=@n15
    Medium=MaterialFunctionCall!(GrassMaterial=@n15, Layer Mask=@n17, Frensel Color=@n23) {materialFunction=/Game/Athena/Helios/Environments/Landscape/Materials/MaterialFunctions_Utilities/MF_MobileGrass_Fresnel}.Grass Fresnel
   )) {name=Mobile PostRVT, nodeColor=<rgba 0 0.0625 1 1>}
 ]}
