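# M_Hera_MeshPartition is the Hera terrain material, blending biome ground layers such as grass, gravel, forest floor, scree, snow and desert oasis.
# Landscape layer weights and biome mask quads feed per-layer material functions, each carrying parallel high-end and mobile material attributes.
# A post-RVT function adds layer detail, and mobile normals are derived separately.
# Masks also drive landscape grass spawning and prioritized physical materials, where earlier entries win ties.
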Material /Hera/Environment/Terrain/MeshTerrain/Materials/M_Hera_MeshPartition
graph {
 expressions {
  MaterialExpressionAbs_3 {class=MaterialExpressionAbs, inputs [None {from=MaterialExpressionComponentMask_4, output=""}], properties {}}
  MaterialExpressionAdd_0 {
   class=MaterialExpressionAdd
   inputs [
    A {from=MaterialExpressionMultiply_71}
    B {from=MaterialExpressionScalarParameter_14}
   ]
  }
  MaterialExpressionAdd_1 {
   class=MaterialExpressionAdd
   inputs [
    A {from=MaterialExpressionMultiply_57}
    B {from=MaterialExpressionScalarParameter_25}
   ]
  }
  MaterialExpressionAdd_2 {
   class=MaterialExpressionAdd
   inputs [
    A {from=MaterialExpressionMultiply_1}
    B {from=MaterialExpressionScalarParameter_2}
   ]
  }
  MaterialExpressionAdd_25 {
   class=MaterialExpressionAdd
   inputs [
    A {from=MaterialExpressionScalarParameter_49}
    B {from=MaterialExpressionWorldPosition_1, output=Z}
   ]
  }
  MaterialExpressionAdd_3 {
   class=MaterialExpressionAdd
   inputs [
    A {from=MaterialExpressionMultiply_26}
    B {from=MaterialExpressionScalarParameter_6}
   ]
  }
  MaterialExpressionAdd_4 {
   class=MaterialExpressionAdd
   inputs [
    A {from=MaterialExpressionMultiply_2}
    B {from=MaterialExpressionScalarParameter_8}
   ]
  }
  MaterialExpressionBlendMaterialAttributes_0 {
   class=MaterialExpressionBlendMaterialAttributes
   inputs [
    A {from=MaterialExpressionReroute_3, output=""}
    B {from=MaterialExpressionSetMaterialAttributes_1, output=""}
    Alpha {from=MaterialExpressionScalarParameter_0, output=""}
   ]
   properties {}
  }
  MaterialExpressionComponentMask_4 {class=MaterialExpressionComponentMask, inputs [None {from=MaterialExpressionNormalize_1}], properties {b=true}}
  MaterialExpressionComposite_10 {class=MaterialExpressionComposite, inputs [], properties {subgraphName=Stone_Physmat_Mask}}
  MaterialExpressionComposite_11 {class=MaterialExpressionComposite, inputs [], properties {subgraphName=Dirt_Physmat_Mask}}
  MaterialExpressionComposite_12 {
   class=MaterialExpressionComposite
   inputs [
    "HighEndMaterial(MA)" {from=MaterialExpressionNamedRerouteUsage_136, output=""}
    "MobileMaterial(MA)" {from=MaterialExpressionNamedRerouteUsage_137, output=""}
   ]
   properties {}
  }
  MaterialExpressionComposite_5 {class=MaterialExpressionComposite, inputs [], properties {subgraphName=Snow_Physmat_Mask}}
  MaterialExpressionComposite_8 {class=MaterialExpressionComposite, inputs [], properties {subgraphName=Sand_Physmat_Mask}}
  MaterialExpressionComposite_9 {class=MaterialExpressionComposite, inputs [], properties {subgraphName=Grass_Physmat_Mask}}
  MaterialExpressionConstant3Vector_0 {class=MaterialExpressionConstant3Vector, properties {constant=<rgba 0 0 1 0>}}
  MaterialExpressionConstant3Vector_1 {class=MaterialExpressionConstant3Vector, properties {constant=<rgba 0.21 0.2855 0.35 1>}}
  MaterialExpressionConstant_0 {class=MaterialExpressionConstant}
  MaterialExpressionConstant_1 {class=MaterialExpressionConstant, properties {desc=Temp11_}}
  MaterialExpressionConstant_10 {class=MaterialExpressionConstant}
  MaterialExpressionConstant_11 {class=MaterialExpressionConstant, properties {r=3}}
  MaterialExpressionConstant_12 {class=MaterialExpressionConstant}
  MaterialExpressionConstant_13 {class=MaterialExpressionConstant}
  MaterialExpressionConstant_14 {class=MaterialExpressionConstant, properties {r=1}}
  MaterialExpressionConstant_15 {class=MaterialExpressionConstant}
  MaterialExpressionConstant_16 {class=MaterialExpressionConstant, properties {desc=Temp10_}}
  MaterialExpressionConstant_17 {class=MaterialExpressionConstant}
  MaterialExpressionConstant_18 {class=MaterialExpressionConstant, properties {r=3}}
  MaterialExpressionConstant_19 {class=MaterialExpressionConstant, properties {desc=Temp1_Chaparral}}
  MaterialExpressionConstant_2 {class=MaterialExpressionConstant, properties {desc=Temp9_}}
  MaterialExpressionConstant_20 {class=MaterialExpressionConstant}
  MaterialExpressionConstant_21 {class=MaterialExpressionConstant, properties {r=0.75}}
  MaterialExpressionConstant_22 {class=MaterialExpressionConstant, properties {r=43}}
  MaterialExpressionConstant_23 {class=MaterialExpressionConstant, properties {r=-16.5}}
  MaterialExpressionConstant_24 {class=MaterialExpressionConstant, properties {desc=Temp2_RedWoods}}
  MaterialExpressionConstant_25 {class=MaterialExpressionConstant}
  MaterialExpressionConstant_26 {class=MaterialExpressionConstant}
  MaterialExpressionConstant_27 {class=MaterialExpressionConstant, properties {desc=Temp3_Boreal}}
  MaterialExpressionConstant_28 {class=MaterialExpressionConstant, properties {desc=Temp4_Desert}}
  MaterialExpressionConstant_29 {class=MaterialExpressionConstant}
  MaterialExpressionConstant_3 {class=MaterialExpressionConstant, properties {desc=Replace With SnowLayer_Alpha}}
  MaterialExpressionConstant_30 {class=MaterialExpressionConstant, properties {r=2}}
  MaterialExpressionConstant_31 {class=MaterialExpressionConstant, properties {desc=Temp5_Alien}}
  MaterialExpressionConstant_32 {class=MaterialExpressionConstant, properties {desc=Temp6_EchoFrost}}
  MaterialExpressionConstant_33 {class=MaterialExpressionConstant, properties {desc=Temp7_GolfCourse}}
  MaterialExpressionConstant_34 {class=MaterialExpressionConstant, properties {desc=Temp8_DarkNexus}}
  MaterialExpressionConstant_35 {class=MaterialExpressionConstant, properties {r=6}}
  MaterialExpressionConstant_36 {class=MaterialExpressionConstant, properties {r=3}}
  MaterialExpressionConstant_37 {class=MaterialExpressionConstant, properties {r=3}}
  MaterialExpressionConstant_38 {class=MaterialExpressionConstant}
  MaterialExpressionConstant_39 {class=MaterialExpressionConstant}
  MaterialExpressionConstant_4 {class=MaterialExpressionConstant, properties {desc=Replace With SnowLayer_Alpha}}
  MaterialExpressionConstant_40 {class=MaterialExpressionConstant}
  MaterialExpressionConstant_41 {class=MaterialExpressionConstant}
  MaterialExpressionConstant_42 {class=MaterialExpressionConstant, properties {desc=Temp12_RemoveLandscapeGrassTypes}}
  MaterialExpressionConstant_43 {class=MaterialExpressionConstant}
  MaterialExpressionConstant_44 {class=MaterialExpressionConstant}
  MaterialExpressionConstant_46 {class=MaterialExpressionConstant}
  MaterialExpressionConstant_47 {class=MaterialExpressionConstant, properties {r=1.5}}
  MaterialExpressionConstant_48 {class=MaterialExpressionConstant, properties {r=-2}}
  MaterialExpressionConstant_49 {class=MaterialExpressionConstant, properties {r=6}}
  MaterialExpressionConstant_5 {class=MaterialExpressionConstant, properties {desc=Replace With SnowLayer_Alpha}}
  MaterialExpressionConstant_50 {class=MaterialExpressionConstant, properties {r=6}}
  MaterialExpressionConstant_52 {class=MaterialExpressionConstant, properties {desc=BM_EchoFrost, r=1}}
  MaterialExpressionConstant_53 {class=MaterialExpressionConstant, properties {desc=BM_EchoFrost, r=1}}
  MaterialExpressionConstant_56 {class=MaterialExpressionConstant, properties {r=-2}}
  MaterialExpressionConstant_57 {class=MaterialExpressionConstant, properties {r=1.5}}
  MaterialExpressionConstant_6 {class=MaterialExpressionConstant}
  MaterialExpressionConstant_64 {class=MaterialExpressionConstant, properties {r=1}}
  MaterialExpressionConstant_66 {class=MaterialExpressionConstant}
  MaterialExpressionConstant_69 {class=MaterialExpressionConstant}
  MaterialExpressionConstant_7 {class=MaterialExpressionConstant, properties {r=6}}
  MaterialExpressionConstant_74 {class=MaterialExpressionConstant}
  MaterialExpressionConstant_76 {class=MaterialExpressionConstant}
  MaterialExpressionConstant_8 {class=MaterialExpressionConstant}
  MaterialExpressionConstant_85 {class=MaterialExpressionConstant, properties {r=3}}
  MaterialExpressionConstant_9 {class=MaterialExpressionConstant, properties {r=1}}
  MaterialExpressionConvert_0 {
   class=MaterialExpressionConvert
   inputs [
    None {from=MaterialExpressionNamedRerouteUsage_205}
    None {from=MaterialExpressionNamedRerouteUsage_371}
    None {from=MaterialExpressionNamedRerouteUsage_372}
    None {from=MaterialExpressionNamedRerouteUsage_373}
   ]
   properties {
    nodeName=Make Vector4
    convertInputs [
     {
      type=Scalar
      expressionInput {
       expression=&self:MaterialExpressionNamedRerouteUsage_205
       inputName=None
       mask=0
       maskA=0
       maskB=0
       maskG=0
       maskR=0
       outputIndex=0
      }
     }
     {
      type=Scalar
      expressionInput {
       expression=&self:MaterialExpressionNamedRerouteUsage_371
       inputName=None
       mask=0
       maskA=0
       maskB=0
       maskG=0
       maskR=0
       outputIndex=0
      }
     }
     {
      type=Scalar
      expressionInput {
       expression=&self:MaterialExpressionNamedRerouteUsage_372
       inputName=None
       mask=0
       maskA=0
       maskB=0
       maskG=0
       maskR=0
       outputIndex=0
      }
     }
     {
      type=Scalar
      expressionInput {
       expression=&self:MaterialExpressionNamedRerouteUsage_373
       inputName=None
       mask=0
       maskA=0
       maskB=0
       maskG=0
       maskR=0
       outputIndex=0
      }
     }
    ]
    convertOutputs [{type=Vector4, defaultValue=<rgba 0 0 0 1>}]
   }
  }
  MaterialExpressionCrossProduct_1 {
   class=MaterialExpressionCrossProduct
   inputs [
    A {from=MaterialExpressionDDX_1, output=""}
    B {from=MaterialExpressionDDY_1, output=""}
   ]
   properties {}
  }
  MaterialExpressionDDX_1 {class=MaterialExpressionDDX, inputs [Value {from=MaterialExpressionWorldPosition_0, output=XYZ}], properties {}}
  MaterialExpressionDDY_1 {class=MaterialExpressionDDY, inputs [Value {from=MaterialExpressionWorldPosition_0, output=XYZ}], properties {}}
  MaterialExpressionDivide_8 {
   class=MaterialExpressionDivide
   inputs [
    A {from=MaterialExpressionAdd_25}
    B {from=MaterialExpressionScalarParameter_3}
   ]
  }
  MaterialExpressionDotProduct_0 {
   class=MaterialExpressionDotProduct
   inputs [
    A {from=MaterialExpressionGetMaterialAttributes_6, output=Normal}
    B {from=MaterialExpressionConstant3Vector_0, output=""}
   ]
   properties {}
  }
  MaterialExpressionDotProduct_1 {
   class=MaterialExpressionDotProduct
   inputs [
    A {from=MaterialExpressionConvert_0, output=""}
    B {from=MaterialExpressionVectorParameter_4, output=RGBA}
   ]
   properties {}
  }
  MaterialExpressionGetMaterialAttributes_0 {
   class=MaterialExpressionGetMaterialAttributes
   inputs [None {from=MaterialExpressionNamedRerouteDeclaration_18, output=""}]
   properties {attributeGetTypes [#1]}
  }
  MaterialExpressionGetMaterialAttributes_1 {
   class=MaterialExpressionGetMaterialAttributes
   inputs [None {from=MaterialExpressionReroute_3, output=""}]
   properties {attributeGetTypes [#2]}
  }
  MaterialExpressionGetMaterialAttributes_2 {
   class=MaterialExpressionGetMaterialAttributes
   inputs [None {from=MaterialExpressionShadingPathSwitch_0, output=""}]
   properties {attributeGetTypes [
     #2
     #3
     #4
     #5
     #6
    ]}
  }
  MaterialExpressionGetMaterialAttributes_3 {
   class=MaterialExpressionGetMaterialAttributes
   inputs [None {from=MaterialExpressionMaterialFunctionCall_13, output=Result}]
   properties {attributeGetTypes [#2]}
  }
  MaterialExpressionGetMaterialAttributes_4 {class=MaterialExpressionGetMaterialAttributes, inputs [None {from=MaterialExpressionMaterialFunctionCall_26, output=PostRVT}]}
  MaterialExpressionGetMaterialAttributes_5 {
   class=MaterialExpressionGetMaterialAttributes
   inputs [None {from=MaterialExpressionReroute_3, output=""}]
   properties {attributeGetTypes [#6]}
  }
  MaterialExpressionGetMaterialAttributes_6 {
   class=MaterialExpressionGetMaterialAttributes
   inputs [None {from=MaterialExpressionStaticSwitch_48, output=""}]
   properties {attributeGetTypes [
     #2
     #3
     #4
     #5
     #1
    ]}
  }
  MaterialExpressionGetMaterialAttributes_7 {
   class=MaterialExpressionGetMaterialAttributes
   inputs [None {from=MaterialExpressionNamedRerouteUsage_63, output=""}]
   properties {attributeGetTypes [#2]}
  }
  MaterialExpressionIf_0 {
   class=MaterialExpressionIf
   inputs [
    A {from=MaterialExpressionSaturate_16, output=""}
    B {from=MaterialExpressionConstant_21, output=""}
    "A > B" {from=MaterialExpressionConstant_14, output=""}
    "A == B" {from=MaterialExpressionConstant_14, output=""}
    "A < B" {from=MaterialExpressionConstant_20, output=""}
   ]
   properties {}
  }
  MaterialExpressionLandscapeGrassOutput_0 {
   class=MaterialExpressionLandscapeGrassOutput
   inputs [
    Grass01_Lobby {from=MaterialExpressionNamedRerouteUsage_171, output=""}
    Scree {from=MaterialExpressionNamedRerouteUsage_156, output=""}
    ForestFloor01 {from=MaterialExpressionNamedRerouteUsage_58, output=""}
    FlowerGroup01 {from=MaterialExpressionNamedRerouteUsage_129, output=""}
    Beach {from=MaterialExpressionNamedRerouteUsage_339, output=""}
    Archipelago {from=MaterialExpressionNamedRerouteUsage_242, output=""}
    Seaweed {from=MaterialExpressionNamedRerouteUsage_357, output=""}
   ]
   properties {grassTypes [
     Grass01_Lobby {
      grassType=/Hera/Environment/Terrain/Landscape/LandscapeGrassTypes/LGT_Grass01_Lobby
      input {
       expression=&self:MaterialExpressionNamedRerouteUsage_171
       inputName=None
       mask=0
       maskA=0
       maskB=0
       maskG=0
       maskR=0
       outputIndex=0
      }
     }
     Scree {
      grassType=/Hera/Environment/Terrain/Landscape/LandscapeGrassTypes/LGT_Scree01
      input {
       expression=&self:MaterialExpressionNamedRerouteUsage_156
       inputName=None
       mask=0
       maskA=0
       maskB=0
       maskG=0
       maskR=0
       outputIndex=0
      }
     }
     ForestFloor01 {
      grassType=/Hera/Environment/Terrain/Landscape/LandscapeGrassTypes/LGT_ForestFloor
      input {
       expression=&self:MaterialExpressionNamedRerouteUsage_58
       inputName=None
       mask=0
       maskA=0
       maskB=0
       maskG=0
       maskR=0
       outputIndex=0
      }
     }
     FlowerGroup01 {
      grassType=/Hera/Environment/Terrain/Landscape/LandscapeGrassTypes/LGT_FlowerGroup01
      input {
       expression=&self:MaterialExpressionNamedRerouteUsage_129
       inputName=None
       mask=0
       maskA=0
       maskB=0
       maskG=0
       maskR=0
       outputIndex=0
      }
     }
     Beach {
      grassType=/Hera/Environment/Terrain/Landscape/LandscapeGrassTypes/LGT_Beach
      input {
       expression=&self:MaterialExpressionNamedRerouteUsage_339
       inputName=None
       mask=0
       maskA=0
       maskB=0
       maskG=0
       maskR=0
       outputIndex=0
      }
     }
     Archipelago {
      grassType=/Hera/Environment/Terrain/Landscape/LandscapeGrassTypes/LGT_Archipelago
      input {
       expression=&self:MaterialExpressionNamedRerouteUsage_242
       inputName=None
       mask=0
       maskA=0
       maskB=0
       maskG=0
       maskR=0
       outputIndex=0
      }
     }
     Seaweed {
      grassType=/Hera/Environment/Terrain/Landscape/LandscapeGrassTypes/PCG/LGT_Grass_Seaweed_A
      input {
       expression=&self:MaterialExpressionNamedRerouteUsage_357
       inputName=None
       mask=0
       maskA=0
       maskB=0
       maskG=0
       maskR=0
       outputIndex=0
      }
     }
    ]}
  }
  MaterialExpressionLandscapeLayerSample_0 {class=MaterialExpressionLandscapeLayerSample, inputs [], properties {parameterName=Gravel}}
  MaterialExpressionLandscapeLayerSample_1 {class=MaterialExpressionLandscapeLayerSample, inputs [], properties {parameterName=Sand}}
  MaterialExpressionLandscapeLayerSample_2 {class=MaterialExpressionLandscapeLayerSample, inputs [], properties {parameterName=Mud}}
  MaterialExpressionLandscapeLayerSample_3 {class=MaterialExpressionLandscapeLayerSample, inputs [], properties {parameterName=Asphalt}}
  MaterialExpressionLandscapeLayerSample_4 {class=MaterialExpressionLandscapeLayerSample, inputs [], properties {parameterName=BiomeA}}
  MaterialExpressionLandscapeLayerSample_5 {class=MaterialExpressionLandscapeLayerSample, inputs [], properties {parameterName=BiomeB}}
  MaterialExpressionLandscapeLayerSample_6 {class=MaterialExpressionLandscapeLayerSample, inputs [], properties {parameterName=BiomeC}}
  MaterialExpressionLandscapeLayerSample_8 {class=MaterialExpressionLandscapeLayerSample, inputs [], properties {parameterName=BiomeD}}
  MaterialExpressionLandscapePhysicalMaterialOutput_0 {
   class=MaterialExpressionLandscapePhysicalMaterialOutput
   inputs [
    Dirt {from=MaterialExpressionNamedRerouteUsage_77, output=""}
    Stone {from=MaterialExpressionNamedRerouteUsage_82, output=""}
    Grass {from=MaterialExpressionNamedRerouteUsage_88, output=""}
    Asphalt {from=MaterialExpressionNamedRerouteUsage_169, output=""}
    Sand {from=MaterialExpressionNamedRerouteUsage_87, output=""}
    Snow {from=MaterialExpressionNamedRerouteUsage_89, output=""}
   ]
   properties {
    desc="These are layered in order of importance.\r\nIf Dirt=1 and Stone=1, then Dirt will \"win\".\r\nThese values aren't clamped.\r\nSo a value of 2 should trump a value of 1."
    inputs [
     {
      physicalMaterial=/Game/FortressPhysicalMaterials/Dirt
      input {
       expression=&self:MaterialExpressionNamedRerouteUsage_77
       inputName=None
       mask=0
       maskA=0
       maskB=0
       maskG=0
       maskR=0
       outputIndex=0
      }
     }
     {
      physicalMaterial=/Game/FortressPhysicalMaterials/Stone
      input {
       expression=&self:MaterialExpressionNamedRerouteUsage_82
       inputName=None
       mask=0
       maskA=0
       maskB=0
       maskG=0
       maskR=0
       outputIndex=0
      }
     }
     {
      physicalMaterial=/Game/FortressPhysicalMaterials/Grass
      input {
       expression=&self:MaterialExpressionNamedRerouteUsage_88
       inputName=None
       mask=0
       maskA=0
       maskB=0
       maskG=0
       maskR=0
       outputIndex=0
      }
     }
     {
      physicalMaterial=/Game/FortressPhysicalMaterials/Asphalt
      input {
       expression=&self:MaterialExpressionNamedRerouteUsage_169
       inputName=None
       mask=0
       maskA=0
       maskB=0
       maskG=0
       maskR=0
       outputIndex=0
      }
     }
     {
      physicalMaterial=/Game/FortressPhysicalMaterials/Sand
      input {
       expression=&self:MaterialExpressionNamedRerouteUsage_87
       inputName=None
       mask=0
       maskA=0
       maskB=0
       maskG=0
       maskR=0
       outputIndex=0
      }
     }
     {
      physicalMaterial=/Game/FortressPhysicalMaterials/Snow
      input {
       expression=&self:MaterialExpressionNamedRerouteUsage_89
       inputName=None
       mask=0
       maskA=0
       maskB=0
       maskG=0
       maskR=0
       outputIndex=0
      }
     }
    ]
   }
  }
  MaterialExpressionLandscapeVisibilityMask_1 {class=MaterialExpressionLandscapeVisibilityMask, inputs [], properties {}}
  MaterialExpressionLinearInterpolate_0 {
   class=MaterialExpressionLinearInterpolate
   inputs [
    A {from=MaterialExpressionLinearInterpolate_10}
    B {from=MaterialExpressionVectorParameter_2, output=RGB}
    Alpha {from=MaterialExpressionMultiply_65}
   ]
  }
  MaterialExpressionLinearInterpolate_1 {
   class=MaterialExpressionLinearInterpolate
   inputs [
    A {from=MaterialExpressionConstant_56}
    B {from=MaterialExpressionConstant_57}
    Alpha {from=MaterialExpressionNamedRerouteUsage_140}
   ]
  }
  MaterialExpressionLinearInterpolate_10 {
   class=MaterialExpressionLinearInterpolate
   inputs [
    A {from=MaterialExpressionGetMaterialAttributes_6, output=Base Color}
    B {from=MaterialExpressionSaturate_15}
    Alpha {from=MaterialExpressionConstant_52}
   ]
  }
  MaterialExpressionLinearInterpolate_11 {
   class=MaterialExpressionLinearInterpolate
   inputs [
    A {from=MaterialExpressionNamedRerouteUsage_165}
    B {from=MaterialExpressionLinearInterpolate_16}
    Alpha {from=MaterialExpressionScalarParameter_26}
   ]
  }
  MaterialExpressionLinearInterpolate_12 {
   class=MaterialExpressionLinearInterpolate
   inputs [
    A {from=MaterialExpressionGetMaterialAttributes_3, output=Base Color}
    B {from=MaterialExpressionMultiply_37}
    Alpha {from=MaterialExpressionMultiply_36}
   ]
  }
  MaterialExpressionLinearInterpolate_13 {
   class=MaterialExpressionLinearInterpolate
   inputs [
    A {from=MaterialExpressionConstant_22}
    B {from=MaterialExpressionConstant_23}
    Alpha {from=MaterialExpressionAbs_3}
   ]
  }
  MaterialExpressionLinearInterpolate_15 {
   class=MaterialExpressionLinearInterpolate
   inputs [
    A {from=MaterialExpressionGetMaterialAttributes_2, output=Displacement}
    B {from=""}
    Alpha {from=MaterialExpressionSaturate_18}
   ]
   properties {constB=0.5}
  }
  MaterialExpressionLinearInterpolate_16 {
   class=MaterialExpressionLinearInterpolate
   inputs [
    A {from=MaterialExpressionNamedRerouteUsage_170}
    B {from=MaterialExpressionLinearInterpolate_17}
    Alpha {from=MaterialExpressionScalarParameter_27}
   ]
  }
  MaterialExpressionLinearInterpolate_17 {
   class=MaterialExpressionLinearInterpolate
   inputs [
    A {from=MaterialExpressionNamedRerouteUsage_161}
    B {from=MaterialExpressionNamedRerouteUsage_166}
    Alpha {from=MaterialExpressionScalarParameter_28}
   ]
  }
  MaterialExpressionLinearInterpolate_2 {
   class=MaterialExpressionLinearInterpolate
   inputs [
    A {from=MaterialExpressionSaturate_7}
    B {from=MaterialExpressionSaturate_6}
    Alpha {from=MaterialExpressionMultiply_72}
   ]
  }
  MaterialExpressionLinearInterpolate_3 {
   class=MaterialExpressionLinearInterpolate
   inputs [
    A {from=MaterialExpressionGetMaterialAttributes_7, output=Base Color}
    B {from=MaterialExpressionMultiply_14}
    Alpha {from=MaterialExpressionNamedRerouteUsage_62}
   ]
  }
  MaterialExpressionLinearInterpolate_4 {
   class=MaterialExpressionLinearInterpolate
   inputs [
    A {from=MaterialExpressionGetMaterialAttributes_6, output=Roughness}
    B {from=MaterialExpressionLinearInterpolate_2}
    Alpha {from=MaterialExpressionConstant_53}
   ]
  }
  MaterialExpressionLinearInterpolate_5 {
   class=MaterialExpressionLinearInterpolate
   inputs [
    A {from=MaterialExpressionGetMaterialAttributes_1, output=Base Color}
    B {from=MaterialExpressionMultiply_33}
    Alpha {from=""}
   ]
  }
  MaterialExpressionLinearInterpolate_6 {
   class=MaterialExpressionLinearInterpolate
   inputs [
    A {from=MaterialExpressionConstant_48}
    B {from=MaterialExpressionConstant_47}
    Alpha {from=MaterialExpressionNamedRerouteUsage_309}
   ]
  }
  MaterialExpressionLinearInterpolate_7 {
   class=MaterialExpressionLinearInterpolate
   inputs [
    A {from=MaterialExpressionNamedRerouteUsage_44}
    B {from=MaterialExpressionLinearInterpolate_9}
    Alpha {from=MaterialExpressionScalarParameter_21}
   ]
  }
  MaterialExpressionLinearInterpolate_8 {
   class=MaterialExpressionLinearInterpolate
   inputs [
    A {from=MaterialExpressionGetMaterialAttributes_6, output=Base Color}
    B {from=MaterialExpressionVectorParameter_3, output=RGB}
    Alpha {from=MaterialExpressionVectorParameter_3, output=RGB}
   ]
  }
  MaterialExpressionLinearInterpolate_9 {
   class=MaterialExpressionLinearInterpolate
   inputs [
    A {from=MaterialExpressionNamedRerouteUsage_66}
    B {from=MaterialExpressionLinearInterpolate_11}
    Alpha {from=MaterialExpressionScalarParameter_22}
   ]
  }
  MaterialExpressionMaterialFunctionCall_0 {
   class=MaterialExpressionMaterialFunctionCall
   inputs [
    Material {from=MaterialExpressionNamedRerouteUsage_280}
    MobileMaterial {from=MaterialExpressionNamedRerouteUsage_255}
    MobileBorealForestMaterial {from=MaterialExpressionNamedRerouteUsage_2}
    UseTangentNormals {from=""}
    LayerWeight {from=MaterialExpressionNamedRerouteUsage_9}
    TerrainMasks_QuadA {from=MaterialExpressionNamedRerouteUsage_39}
    TerrainMasks_QuadB {from=MaterialExpressionNamedRerouteUsage_33}
    TerrainMasks_QuadC {from=MaterialExpressionNamedRerouteUsage_34}
    TerrainMasks_QuadD {from=MaterialExpressionNamedRerouteUsage_37}
    VariationMasks Mobile {from=MaterialExpressionNamedRerouteUsage_124}
   ]
   properties {materialFunction=/Hera/Environment/Terrain/Landscape/Materials/Layers/MF_Hera_ForestFloor}
  }
  MaterialExpressionMaterialFunctionCall_1 {
   class=MaterialExpressionMaterialFunctionCall
   inputs [
    X {from=MaterialExpressionConstant_9}
    Y {from=MaterialExpressionConstant_8}
    Z {from=MaterialExpressionConstant_8}
    A {from=MaterialExpressionConstant_8}
   ]
   properties {materialFunction=&MakeFloat4}
  }
  MaterialExpressionMaterialFunctionCall_10 {
   class=MaterialExpressionMaterialFunctionCall
   inputs [
    In {from=MaterialExpressionMultiply_6}
    ZSpawnLimit {from=""}
   ]
   properties {materialFunction=/Game/Athena/Helios/Environments/Landscape/Materials/MaterialFunctions_Utilities/MF_Helios_LS_GrassSwitch}
  }
  MaterialExpressionMaterialFunctionCall_11 {
   class=MaterialExpressionMaterialFunctionCall
   inputs [
    Material {from=MaterialExpressionNamedRerouteUsage_173}
    MobileMaterial {from=MaterialExpressionNamedRerouteUsage_174}
    IsForStaticMesh {from=""}
    UseTangentNormals {from=""}
    LayerWeight {from=MaterialExpressionNamedRerouteUsage_1}
    MudWeight {from=MaterialExpressionNamedRerouteUsage_14}
    TerrainMasks_QuadA {from=MaterialExpressionNamedRerouteUsage_20}
    TerrainMasks_QuadB {from=MaterialExpressionNamedRerouteUsage_17}
    TerrainMasks_QuadC {from=MaterialExpressionNamedRerouteUsage_18}
    TerrainMasks_QuadD {from=MaterialExpressionNamedRerouteUsage_19}
    VariationMasks Mobile {from=MaterialExpressionNamedRerouteUsage_50}
   ]
   properties {materialFunction=/Hera/Environment/Terrain/Landscape/Materials/Layers/MF_Hera_Gravel}
  }
  MaterialExpressionMaterialFunctionCall_12 {
   class=MaterialExpressionMaterialFunctionCall
   inputs [In {from=MaterialExpressionSetMaterialAttributes_9}]
   properties {materialFunction=/Game/Packages/Fortress_SharedMaterials/Cinematics/CINE_MF_EnvPostFX}
  }
  MaterialExpressionMaterialFunctionCall_13 {
   class=MaterialExpressionMaterialFunctionCall
   inputs [
    Material {from=MaterialExpressionShadingPathSwitch_4}
    VariationMasksMobile {from=MaterialExpressionNamedRerouteUsage_95}
    UseSnowForBiomeC {from=MaterialExpressionNamedRerouteUsage_232}
    Snow_PostRVT {from=MaterialExpressionNamedRerouteUsage_101}
    UseGoldLandsForBiomeC {from=MaterialExpressionNamedRerouteUsage_295}
    GoldLands_PostRVT {from=MaterialExpressionNamedRerouteUsage_296}
    SwitchGrassLayerMask {from=MaterialExpressionNamedRerouteUsage_213}
    TerrainMasks_QuadA {from=MaterialExpressionNamedRerouteUsage_184}
    TerrainMasks_QuadB {from=MaterialExpressionNamedRerouteUsage_51}
    TerrainMasks_QuadC {from=MaterialExpressionNamedRerouteUsage_185}
    TerrainMasks_QuadD {from=MaterialExpressionNamedRerouteUsage_186}
   ]
   properties {
    desc=For Post RVT Details on landscape layers
    materialFunction=/Hera/Environment/Terrain/Landscape/Materials/Layers/Misc/MF_Hera_PostRVT_Settings
   }
  }
  MaterialExpressionMaterialFunctionCall_14 {
   class=MaterialExpressionMaterialFunctionCall
   inputs [
    In {from=MaterialExpressionMultiply_74}
    ZSpawnLimit {from=""}
   ]
   properties {materialFunction=&MF_Helios_LS_GrassSwitch}
  }
  MaterialExpressionMaterialFunctionCall_15 {
   class=MaterialExpressionMaterialFunctionCall
   inputs [
    Material {from=MaterialExpressionNamedRerouteUsage_206}
    MobileMaterial {from=MaterialExpressionNamedRerouteUsage_211}
    ChapCoastMobile {from=MaterialExpressionNamedRerouteUsage_307}
    GrassWeight {from=""}
    LayerWeight {from=MaterialExpressionNamedRerouteUsage_202}
    UseTangentNormals {from=""}
    TerrainMasks_QuadA {from=MaterialExpressionNamedRerouteUsage_125}
    TerrainMasks_QuadB {from=MaterialExpressionNamedRerouteUsage_110}
    TerrainMasks_QuadC {from=MaterialExpressionNamedRerouteUsage_122}
    TerrainMasks_QuadD {from=MaterialExpressionNamedRerouteUsage_123}
    VariationMasks Mobile {from=MaterialExpressionNamedRerouteUsage_133}
   ]
   properties {materialFunction=/Hera/Environment/Terrain/Landscape/Materials/Layers/MF_Hera_Grass_Desert}
  }
  MaterialExpressionMaterialFunctionCall_16 {
   class=MaterialExpressionMaterialFunctionCall
   inputs []
   properties {materialFunction=/Game/Athena/Helios/Environments/Landscape/Materials/MaterialFunctions_Utilities/MF_CliffSlopeMask}
  }
  MaterialExpressionMaterialFunctionCall_17 {
   class=MaterialExpressionMaterialFunctionCall
   inputs [
    In {from=MaterialExpressionMultiply_58}
    ZSpawnLimit {from=""}
   ]
   properties {materialFunction=&MF_Helios_LS_GrassSwitch}
  }
  MaterialExpressionMaterialFunctionCall_18 {
   class=MaterialExpressionMaterialFunctionCall
   inputs [
    BiomeMask {from=MaterialExpressionNamedRerouteUsage_138}
    IsForStaticMesh {from=""}
   ]
   properties {materialFunction=/Hera/Environment/Terrain/Landscape/Materials/Layers/Misc/MF_Desert_MacroDetail}
  }
  MaterialExpressionMaterialFunctionCall_19 {
   class=MaterialExpressionMaterialFunctionCall
   inputs [
    Material {from=MaterialExpressionNamedRerouteUsage_175}
    MobileMaterial {from=MaterialExpressionNamedRerouteUsage_176}
    GrassWeight {from=""}
    LayerWeight {from=MaterialExpressionNamedRerouteUsage_35}
    UseTangentNormals {from=""}
    TerrainMasks_QuadA {from=MaterialExpressionNamedRerouteUsage_99}
    TerrainMasks_QuadB {from=MaterialExpressionNamedRerouteUsage_96}
    TerrainMasks_QuadC {from=MaterialExpressionNamedRerouteUsage_97}
    TerrainMasks_QuadD {from=MaterialExpressionNamedRerouteUsage_98}
    VariationMasks Mobile {from=MaterialExpressionNamedRerouteUsage_107}
   ]
   properties {materialFunction=/Hera/Environment/Terrain/Landscape/Materials/Layers/MF_Hera_Grass_Chaparral}
  }
  MaterialExpressionMaterialFunctionCall_2 {
   class=MaterialExpressionMaterialFunctionCall
   inputs [
    Material {from=""}
    MobileMaterial {from=""}
    IsForStaticMesh {from=""}
    UseTangentNormals {from=""}
    UseCustomSlopeMask {from=""}
    CustomSlopeMask {from=""}
    Grass_EdgeOnly {from=MaterialExpressionNamedRerouteUsage_4}
    LayerWeight {from=MaterialExpressionNamedRerouteUsage_0}
    TerrainMasks_QuadA {from=MaterialExpressionNamedRerouteUsage_15}
    TerrainMasks_QuadB {from=MaterialExpressionNamedRerouteUsage_11}
    TerrainMasks_QuadC {from=MaterialExpressionNamedRerouteUsage_12}
    TerrainMasks_QuadD {from=MaterialExpressionNamedRerouteUsage_13}
    VariationMasks Mobile {from=MaterialExpressionNamedRerouteUsage_40}
    DesertGrass_SimpleNoise {from=MaterialExpressionNamedRerouteUsage_36}
   ]
   properties {materialFunction=/Hera/Environment/Terrain/Landscape/Materials/Layers/MF_Hera_Grass}
  }
  MaterialExpressionMaterialFunctionCall_20 {
   class=MaterialExpressionMaterialFunctionCall
   inputs [
    In {from=MaterialExpressionMultiply_64}
    ZSpawnLimit {from=""}
   ]
   properties {materialFunction=&MF_Helios_LS_GrassSwitch}
  }
  MaterialExpressionMaterialFunctionCall_21 {
   class=MaterialExpressionMaterialFunctionCall
   inputs [
    In {from=MaterialExpressionMultiply_12}
    ZSpawnLimit {from=""}
   ]
   properties {materialFunction=&MF_Helios_LS_GrassSwitch}
  }
  MaterialExpressionMaterialFunctionCall_22 {
   class=MaterialExpressionMaterialFunctionCall
   inputs [
    UseLandscapeCoords {from=MaterialExpressionStaticBool_0}
    "UseProjection(translucent)" {from=""}
   ]
   properties {materialFunction=/BRCoreMaterials/Hera/Terrain/BiomeMask/MF_Hera_BiomeMasks}
  }
  MaterialExpressionMaterialFunctionCall_23 {
   class=MaterialExpressionMaterialFunctionCall
   inputs [
    TerrainMasks_QuadA {from=MaterialExpressionNamedRerouteDeclaration_20}
    TerrainMasks_QuadB {from=MaterialExpressionNamedRerouteDeclaration_19}
    TerrainMasks_QuadC {from=MaterialExpressionNamedRerouteDeclaration_21}
    TerrainMasks_QuadD {from=MaterialExpressionNamedRerouteDeclaration_22}
   ]
   properties {materialFunction=/BRCoreMaterials/Hera/Terrain/BiomeMask/MF_HeraBiomeMasks_Isolated}
  }
  MaterialExpressionMaterialFunctionCall_24 {
   class=MaterialExpressionMaterialFunctionCall
   inputs [
    Material {from=MaterialExpressionNamedRerouteUsage_223}
    MobileMaterial {from=MaterialExpressionNamedRerouteUsage_224}
    Texture Mask {from=""}
    Texture UVs {from=""}
    IsStaticMesh {from=""}
   ]
   properties {materialFunction=/BRCoreMaterials/Hermes/Terrain/Landscape/Materials/Layers/MF_Hermes_Goldlands}
  }
  MaterialExpressionMaterialFunctionCall_25 {
   class=MaterialExpressionMaterialFunctionCall
   inputs [
    In {from=MaterialExpressionMultiply_43}
    ZSpawnLimit {from=""}
   ]
   properties {materialFunction=&MF_Helios_LS_GrassSwitch}
  }
  MaterialExpressionMaterialFunctionCall_26 {
   class=MaterialExpressionMaterialFunctionCall
   inputs [
    Material {from=MaterialExpressionNamedRerouteUsage_187}
    MobileMaterial {from=MaterialExpressionNamedRerouteUsage_188}
    IsForStaticMesh {from=""}
    UseTangentNormals {from=""}
    LayerWeight {from=MaterialExpressionNamedRerouteUsage_57}
    TerrainMasks_QuadA {from=""}
    TerrainMasks_QuadB {from=""}
    TerrainMasks_QuadC {from=""}
    TerrainMasks_QuadD {from=""}
    CliffMaterialOnly {from=MaterialExpressionNamedRerouteUsage_164}
    CliffLayerMask {from=MaterialExpressionNamedRerouteUsage_55}
    VariationMasks Mobile {from=MaterialExpressionNamedRerouteUsage_121}
   ]
   properties {materialFunction=/Hera/Environment/Terrain/Landscape/Materials/Layers/MF_Hera_Snow}
  }
  MaterialExpressionMaterialFunctionCall_27 {
   class=MaterialExpressionMaterialFunctionCall
   inputs [
    Material {from=MaterialExpressionNamedRerouteUsage_228}
    MobileMaterial {from=MaterialExpressionNamedRerouteUsage_229}
    MobileGrass {from=MaterialExpressionNamedRerouteUsage_130}
    LayerWeight {from=MaterialExpressionNamedRerouteUsage_230}
    UseTangentNormals {from=""}
    TerrainMasks_QuadA {from=MaterialExpressionNamedRerouteUsage_226}
    TerrainMasks_QuadB {from=MaterialExpressionNamedRerouteUsage_203}
    TerrainMasks_QuadC {from=MaterialExpressionNamedRerouteUsage_220}
    TerrainMasks_QuadD {from=MaterialExpressionNamedRerouteUsage_225}
    VariationMasks Mobile {from=MaterialExpressionNamedRerouteUsage_227}
   ]
   properties {materialFunction=/Game/Athena/Helios/Environments/Landscape/Materials/LandscapeLayers/MF_Helios_DesertOasis}
  }
  MaterialExpressionMaterialFunctionCall_28 {
   class=MaterialExpressionMaterialFunctionCall
   inputs [
    Material {from=MaterialExpressionReroute_18, output=Output}
    MobileMaterial {from=""}
    Texture Mask {from=""}
    Texture UVs {from=""}
    IsStaticMesh {from=""}
   ]
   properties {materialFunction=&MF_Hermes_Goldlands}
  }
  MaterialExpressionMaterialFunctionCall_29 {
   class=MaterialExpressionMaterialFunctionCall
   inputs [
    X {from=MaterialExpressionConstant_19}
    Y {from=MaterialExpressionConstant_24}
    Z {from=MaterialExpressionConstant_27}
    A {from=MaterialExpressionConstant_28}
   ]
   properties {desc=QuadA, materialFunction=&MakeFloat4}
  }
  MaterialExpressionMaterialFunctionCall_3 {
   class=MaterialExpressionMaterialFunctionCall
   inputs [
    X {from=MaterialExpressionConstant_31}
    Y {from=MaterialExpressionConstant_32}
    Z {from=MaterialExpressionConstant_33}
    A {from=MaterialExpressionConstant_34}
   ]
   properties {desc=QuadB, materialFunction=&MakeFloat4}
  }
  MaterialExpressionMaterialFunctionCall_30 {
   class=MaterialExpressionMaterialFunctionCall
   inputs []
   properties {materialFunction=/BRCoreMaterials/Hermes/Terrain/Landscape/Materials/MaterialFunctions/MF_Hermes_VariationMasks_Mobile}
  }
  MaterialExpressionMaterialFunctionCall_31 {
   class=MaterialExpressionMaterialFunctionCall
   inputs [
    Material {from=MaterialExpressionNamedRerouteUsage_257}
    MobileMaterial {from=MaterialExpressionNamedRerouteUsage_258}
    UseTangentNormals {from=""}
    LayerWeight {from=MaterialExpressionNamedRerouteUsage_120}
    TerrainMasks_QuadA {from=MaterialExpressionNamedRerouteUsage_118}
    TerrainMasks_QuadB {from=MaterialExpressionNamedRerouteUsage_115}
    TerrainMasks_QuadC {from=MaterialExpressionNamedRerouteUsage_116}
    TerrainMasks_QuadD {from=MaterialExpressionNamedRerouteUsage_117}
    VariationMasks Mobile {from=MaterialExpressionNamedRerouteUsage_119}
   ]
   properties {materialFunction=/Hera/Environment/Terrain/Landscape/Materials/Layers/MF_Hera_Scree}
  }
  MaterialExpressionMaterialFunctionCall_32 {
   class=MaterialExpressionMaterialFunctionCall
   inputs [
    In {from=MaterialExpressionMultiply_30}
    ZSpawnLimit {from=""}
   ]
   properties {materialFunction=&MF_Helios_LS_GrassSwitch}
  }
  MaterialExpressionMaterialFunctionCall_33 {
   class=MaterialExpressionMaterialFunctionCall
   inputs [
    Material {from=MaterialExpressionNamedRerouteUsage_179}
    MobileMaterial {from=MaterialExpressionNamedRerouteUsage_181}
    GrassWeight {from=""}
    LayerWeight {from=MaterialExpressionConstant_15}
    UseTangentNormals {from=""}
    TerrainMasks_QuadA {from=MaterialExpressionNamedRerouteUsage_111}
    TerrainMasks_QuadB {from=MaterialExpressionNamedRerouteUsage_49}
    TerrainMasks_QuadC {from=MaterialExpressionNamedRerouteUsage_52}
    TerrainMasks_QuadD {from=MaterialExpressionNamedRerouteUsage_109}
    VariationMasks Mobile {from=MaterialExpressionNamedRerouteUsage_126}
   ]
   properties {materialFunction=/Game/Athena/Helios/Environments/Landscape/Materials/LandscapeLayers/MF_Helios_Grass_BorealForest}
  }
  MaterialExpressionMaterialFunctionCall_34 {
   class=MaterialExpressionMaterialFunctionCall
   inputs [Scale {from=MaterialExpressionScalarParameter_37}]
   properties {materialFunction=/Game/Athena/Apollo/Environments/Landscape/Materials/Biome/CommonMaterials/MF_SimpleWorldUV}
  }
  MaterialExpressionMaterialFunctionCall_35 {
   class=MaterialExpressionMaterialFunctionCall
   inputs [
    Material {from=MaterialExpressionNamedRerouteUsage_282}
    MobileMaterial {from=MaterialExpressionNamedRerouteUsage_237}
    MobileBorealForestMaterial {from=MaterialExpressionNamedRerouteUsage_316}
    UseTangentNormals {from=""}
    LayerWeight {from=MaterialExpressionNamedRerouteUsage_196}
    TerrainMasks_QuadA {from=MaterialExpressionNamedRerouteUsage_200}
    TerrainMasks_QuadB {from=MaterialExpressionNamedRerouteUsage_197}
    TerrainMasks_QuadC {from=MaterialExpressionNamedRerouteUsage_198}
    TerrainMasks_QuadD {from=MaterialExpressionNamedRerouteUsage_199}
    VariationMasks Mobile {from=MaterialExpressionNamedRerouteUsage_201}
   ]
   properties {materialFunction=&MF_Hera_ForestFloor}
  }
  MaterialExpressionMaterialFunctionCall_36 {
   class=MaterialExpressionMaterialFunctionCall
   inputs [
    MobileMaterial {from=MaterialExpressionReroute_21, output="MobileMaterial(MA)"}
    TransformTangents {from=""}
   ]
   properties {materialFunction=/Game/Athena/Helios/Environments/Landscape/Materials/MaterialFunctions_Utilities/MF_MobileDeriveNormals}
  }
  MaterialExpressionMaterialFunctionCall_37 {
   class=MaterialExpressionMaterialFunctionCall
   inputs [
    Material {from=MaterialExpressionNamedRerouteUsage_275}
    MobileMaterial {from=MaterialExpressionNamedRerouteUsage_276}
    UseTangentNormals {from=""}
    LayerWeight {from=MaterialExpressionNamedRerouteUsage_277}
    TerrainMasks_QuadA {from=MaterialExpressionNamedRerouteUsage_272}
    TerrainMasks_QuadB {from=MaterialExpressionNamedRerouteUsage_216}
    TerrainMasks_QuadC {from=MaterialExpressionNamedRerouteUsage_260}
    TerrainMasks_QuadD {from=MaterialExpressionNamedRerouteUsage_261}
    VariationMasks Mobile {from=MaterialExpressionNamedRerouteUsage_273}
   ]
   properties {materialFunction=/Game/Athena/Helios/Environments/Landscape/Materials/LandscapeLayers/MF_Helios_DesertSaltLake}
  }
  MaterialExpressionMaterialFunctionCall_38 {
   class=MaterialExpressionMaterialFunctionCall
   inputs [
    Material {from=MaterialExpressionNamedRerouteUsage_284}
    MobileMaterial {from=MaterialExpressionNamedRerouteUsage_236}
    Mobile Sand Material {from=MaterialExpressionNamedRerouteUsage_331}
    UseTangentNormals {from=""}
    LayerWeight {from=MaterialExpressionNamedRerouteUsage_43}
    TerrainMasks_QuadA {from=MaterialExpressionNamedRerouteUsage_240}
    TerrainMasks_QuadB {from=MaterialExpressionNamedRerouteUsage_194}
    TerrainMasks_QuadC {from=MaterialExpressionNamedRerouteUsage_238}
    TerrainMasks_QuadD {from=MaterialExpressionNamedRerouteUsage_239}
    VariationMasks Mobile {from=MaterialExpressionNamedRerouteUsage_241}
    OD_MD_Normals {from=MaterialExpressionNamedRerouteUsage_322}
    OD_MD_SedimentLayers {from=MaterialExpressionNamedRerouteUsage_330}
   ]
   properties {materialFunction=/Hera/Environment/Terrain/Landscape/Materials/Layers/MF_Hera_DesertWavySand}
  }
  MaterialExpressionMaterialFunctionCall_39 {
   class=MaterialExpressionMaterialFunctionCall
   inputs [
    In {from=MaterialExpressionMultiply_50}
    ZSpawnLimit {from=""}
   ]
   properties {materialFunction=&MF_Helios_LS_GrassSwitch}
  }
  MaterialExpressionMaterialFunctionCall_4 {
   class=MaterialExpressionMaterialFunctionCall
   inputs [
    Material {from=MaterialExpressionMaterialFunctionCall_6, output=BlendedMaterial}
    MobileMaterial {from=MaterialExpressionMaterialFunctionCall_6, output=BlendedMaterial}
    LayerWeight {from=MaterialExpressionNamedRerouteUsage_3}
    UseTangentNormals {from=""}
    TerrainMasks_QuadA {from=MaterialExpressionNamedRerouteUsage_28}
    TerrainMasks_QuadB {from=MaterialExpressionNamedRerouteUsage_25}
    TerrainMasks_QuadC {from=MaterialExpressionNamedRerouteUsage_26}
    TerrainMasks_QuadD {from=MaterialExpressionNamedRerouteUsage_27}
    VariationMasks Mobile {from=MaterialExpressionNamedRerouteUsage_112}
    OD_MD_Normals {from=MaterialExpressionMaterialFunctionCall_18, output=Normals}
    OD_MD_SedimentLayers {from=MaterialExpressionMaterialFunctionCall_18, output=Normals}
    InputMudLayerMask {from=MaterialExpressionNamedRerouteUsage_358}
    InputSeaweedLayerMask {from=""}
   ]
   properties {materialFunction=/Hera/Environment/Terrain/Landscape/Materials/Layers/MF_Hera_Sand}
  }
  MaterialExpressionMaterialFunctionCall_40 {
   class=MaterialExpressionMaterialFunctionCall
   inputs [
    Material {from=MaterialExpressionNamedRerouteUsage_302}
    MobileMaterial {from=MaterialExpressionNamedRerouteUsage_303}
    UseTangentNormals {from=""}
    LayerWeight {from=MaterialExpressionNamedRerouteUsage_304}
    TerrainMasks_QuadA {from=MaterialExpressionNamedRerouteUsage_293}
    TerrainMasks_QuadB {from=MaterialExpressionNamedRerouteUsage_290}
    TerrainMasks_QuadC {from=MaterialExpressionNamedRerouteUsage_291}
    TerrainMasks_QuadD {from=MaterialExpressionNamedRerouteUsage_292}
    VariationMasks Mobile {from=MaterialExpressionNamedRerouteUsage_294}
   ]
   properties {materialFunction=/Hera/Environment/Terrain/Landscape/Materials/Layers/MF_Hera_ArchipelagoSeaweed}
  }
  MaterialExpressionMaterialFunctionCall_41 {
   class=MaterialExpressionMaterialFunctionCall
   inputs [
    X {from=MaterialExpressionConstant_2}
    Y {from=MaterialExpressionConstant_16}
    Z {from=MaterialExpressionConstant_1}
    A {from=MaterialExpressionConstant_42}
   ]
   properties {desc=QuadC, materialFunction=&MakeFloat4}
  }
  MaterialExpressionMaterialFunctionCall_42 {
   class=MaterialExpressionMaterialFunctionCall
   inputs [
    In {from=MaterialExpressionMultiply_56}
    ZSpawnLimit {from=""}
   ]
   properties {materialFunction=&MF_Helios_LS_GrassSwitch}
  }
  MaterialExpressionMaterialFunctionCall_43 {
   class=MaterialExpressionMaterialFunctionCall
   inputs [
    In {from=MaterialExpressionMultiply_48}
    ZSpawnLimit {from=""}
   ]
   properties {materialFunction=&MF_Helios_LS_GrassSwitch}
  }
  MaterialExpressionMaterialFunctionCall_44 {
   class=MaterialExpressionMaterialFunctionCall
   inputs [
    Material {from=MaterialExpressionNamedRerouteUsage_312}
    MobileMaterial {from=MaterialExpressionNamedRerouteUsage_315}
    UseTangentNormals {from=""}
    LayerWeight {from=MaterialExpressionNamedRerouteUsage_323}
    TerrainMasks_QuadA {from=MaterialExpressionNamedRerouteUsage_298}
    TerrainMasks_QuadB {from=MaterialExpressionNamedRerouteUsage_262}
    TerrainMasks_QuadC {from=MaterialExpressionNamedRerouteUsage_263}
    TerrainMasks_QuadD {from=MaterialExpressionNamedRerouteUsage_297}
    VariationMasks Mobile {from=MaterialExpressionNamedRerouteUsage_299}
   ]
   properties {materialFunction=/Hera/Environment/Terrain/Landscape/Materials/Layers/MF_Hera_EchoFrost}
  }
  MaterialExpressionMaterialFunctionCall_45 {
   class=MaterialExpressionMaterialFunctionCall
   inputs [
    In {from=MaterialExpressionNamedRerouteUsage_288}
    ZSpawnLimit {from=""}
   ]
   properties {materialFunction=&MF_Helios_LS_GrassSwitch}
  }
  MaterialExpressionMaterialFunctionCall_46 {
   class=MaterialExpressionMaterialFunctionCall
   inputs [
    In {from=MaterialExpressionMultiply_75}
    ZSpawnLimit {from=""}
   ]
   properties {materialFunction=&MF_Helios_LS_GrassSwitch}
  }
  MaterialExpressionMaterialFunctionCall_49 {
   class=MaterialExpressionMaterialFunctionCall
   inputs [
    In {from=MaterialExpressionMultiply_25}
    ZSpawnLimit {from=""}
   ]
   properties {materialFunction=&MF_Helios_LS_GrassSwitch}
  }
  MaterialExpressionMaterialFunctionCall_5 {
   class=MaterialExpressionMaterialFunctionCall
   inputs [
    Material {from=MaterialExpressionNamedRerouteUsage_300}
    MobileMaterial {from=MaterialExpressionNamedRerouteUsage_301}
    LayerWeight {from=MaterialExpressionNamedRerouteUsage_6}
    UseTangentNormals {from=""}
    TerrainMasks_QuadA {from=MaterialExpressionNamedRerouteUsage_41}
    TerrainMasks_QuadB {from=MaterialExpressionNamedRerouteUsage_30}
    TerrainMasks_QuadC {from=MaterialExpressionNamedRerouteUsage_31}
    TerrainMasks_QuadD {from=MaterialExpressionNamedRerouteUsage_32}
    VariationMasks Mobile {from=MaterialExpressionNamedRerouteUsage_114}
    MossDesaturateMask {from=MaterialExpressionNamedRerouteUsage_193}
   ]
   properties {materialFunction=/Hera/Environment/Terrain/Landscape/Materials/Layers/MF_Hera_Cliff_MeshTerrain}
  }
  MaterialExpressionMaterialFunctionCall_50 {
   class=MaterialExpressionMaterialFunctionCall
   inputs [
    In {from=MaterialExpressionShadingPathSwitch_2}
    ZSpawnLimit {from=""}
   ]
   properties {materialFunction=&MF_Helios_LS_GrassSwitch}
  }
  MaterialExpressionMaterialFunctionCall_52 {
   class=MaterialExpressionMaterialFunctionCall
   inputs [
    In {from=MaterialExpressionMultiply_21}
    ZSpawnLimit {from=""}
   ]
   properties {materialFunction=&MF_Helios_LS_GrassSwitch}
  }
  MaterialExpressionMaterialFunctionCall_55 {
   class=MaterialExpressionMaterialFunctionCall
   inputs [
    In {from=MaterialExpressionMultiply_27}
    ZSpawnLimit {from=""}
   ]
   properties {materialFunction=&MF_Helios_LS_GrassSwitch}
  }
  MaterialExpressionMaterialFunctionCall_6 {
   class=MaterialExpressionMaterialFunctionCall
   inputs [
    Material {from=MaterialExpressionMaterialFunctionCall_11, output=BlendedMaterial}
    MobileMaterial {from=MaterialExpressionMaterialFunctionCall_11, output=BlendedMaterial}
    MobileMudMaterial {from=MaterialExpressionMaterialFunctionCall_11, output=BlendedMaterial}
    ForMudMaterial {from=MaterialExpressionMaterialFunctionCall_11, output=BlendedMaterial}
    LayerWeight {from=MaterialExpressionNamedRerouteUsage_7}
    UseTangentNormals {from=""}
    TerrainMasks_QuadA {from=MaterialExpressionNamedRerouteUsage_24}
    TerrainMasks_QuadB {from=MaterialExpressionNamedRerouteUsage_21}
    TerrainMasks_QuadC {from=MaterialExpressionNamedRerouteUsage_22}
    TerrainMasks_QuadD {from=MaterialExpressionNamedRerouteUsage_23}
    VariationMasks Mobile {from=MaterialExpressionNamedRerouteUsage_91}
   ]
   properties {materialFunction=/Hera/Environment/Terrain/Landscape/Materials/Layers/MF_Hera_Mud}
  }
  MaterialExpressionMaterialFunctionCall_7 {
   class=MaterialExpressionMaterialFunctionCall
   inputs [
    Material {from=MaterialExpressionNamedRerouteUsage_246}
    MobileMaterial {from=MaterialExpressionNamedRerouteUsage_247}
    UseTangentNormals {from=""}
    LayerWeight {from=MaterialExpressionNamedRerouteUsage_8}
    TerrainMasks_QuadA {from=MaterialExpressionNamedRerouteUsage_90}
    TerrainMasks_QuadB {from=MaterialExpressionNamedRerouteUsage_61}
    TerrainMasks_QuadC {from=MaterialExpressionNamedRerouteUsage_70}
    TerrainMasks_QuadD {from=MaterialExpressionNamedRerouteUsage_71}
    VariationMasks Mobile {from=MaterialExpressionNamedRerouteUsage_113}
   ]
   properties {materialFunction=/Hera/Environment/Terrain/Landscape/Materials/Layers/MF_Hera_Asphalt}
  }
  MaterialExpressionMaterialFunctionCall_8 {
   class=MaterialExpressionMaterialFunctionCall
   inputs [
    In {from=MaterialExpressionMultiply_32}
    ZSpawnLimit {from=""}
   ]
   properties {materialFunction=&MF_Helios_LS_GrassSwitch}
  }
  MaterialExpressionMaterialFunctionCall_9 {
   class=MaterialExpressionMaterialFunctionCall
   inputs [
    Material {from=MaterialExpressionNamedRerouteUsage_268}
    MobileMaterial {from=MaterialExpressionNamedRerouteUsage_269}
    UseTangentNormals {from=""}
    LayerWeight {from=MaterialExpressionNamedRerouteUsage_289}
    TerrainMasks_QuadA {from=MaterialExpressionNamedRerouteUsage_252}
    TerrainMasks_QuadB {from=MaterialExpressionNamedRerouteUsage_249}
    TerrainMasks_QuadC {from=MaterialExpressionNamedRerouteUsage_250}
    TerrainMasks_QuadD {from=MaterialExpressionNamedRerouteUsage_251}
    VariationMasks Mobile {from=MaterialExpressionNamedRerouteUsage_253}
   ]
   properties {materialFunction=/Hermes/Terrain/Landscape/Materials/Layers/MF_Hermes_Wetlands_MossyForest}
  }
  MaterialExpressionMax_0 {
   class=MaterialExpressionMax
   inputs [
    A {from=MaterialExpressionNamedRerouteDeclaration_0}
    B {from=MaterialExpressionNamedRerouteDeclaration_2}
   ]
  }
  MaterialExpressionMax_1 {
   class=MaterialExpressionMax
   inputs [
    A {from=MaterialExpressionNamedRerouteDeclaration_1}
    B {from=MaterialExpressionNamedRerouteDeclaration_3}
   ]
  }
  MaterialExpressionMax_10 {
   class=MaterialExpressionMax
   inputs [
    A {from=MaterialExpressionMax_7}
    B {from=MaterialExpressionNamedRerouteUsage_45}
   ]
  }
  MaterialExpressionMax_11 {
   class=MaterialExpressionMax
   inputs [
    A {from=MaterialExpressionNamedRerouteUsage_73}
    B {from=MaterialExpressionNamedRerouteUsage_74}
   ]
  }
  MaterialExpressionMax_12 {
   class=MaterialExpressionMax
   inputs [
    A {from=MaterialExpressionMax_25}
    B {from=MaterialExpressionNamedRerouteUsage_147}
   ]
  }
  MaterialExpressionMax_13 {
   class=MaterialExpressionMax
   inputs [
    A {from=MaterialExpressionNamedRerouteUsage_59}
    B {from=MaterialExpressionNamedRerouteUsage_75}
   ]
  }
  MaterialExpressionMax_14 {
   class=MaterialExpressionMax
   inputs [
    A {from=MaterialExpressionMax_13}
    B {from=MaterialExpressionNamedRerouteUsage_72}
   ]
  }
  MaterialExpressionMax_15 {
   class=MaterialExpressionMax
   inputs [
    A {from=MaterialExpressionMax_14}
    B {from=MaterialExpressionNamedRerouteUsage_76}
   ]
  }
  MaterialExpressionMax_16 {
   class=MaterialExpressionMax
   inputs [
    A {from=MaterialExpressionNamedRerouteUsage_78}
    B {from=MaterialExpressionOneMinus_10}
   ]
  }
  MaterialExpressionMax_17 {
   class=MaterialExpressionMax
   inputs [
    A {from=MaterialExpressionNamedRerouteUsage_81}
    B {from=MaterialExpressionNamedRerouteUsage_80}
   ]
  }
  MaterialExpressionMax_18 {
   class=MaterialExpressionMax
   inputs [
    A {from=MaterialExpressionNamedRerouteUsage_83}
    B {from=MaterialExpressionNamedRerouteUsage_85}
   ]
  }
  MaterialExpressionMax_19 {
   class=MaterialExpressionMax
   inputs [
    A {from=MaterialExpressionNamedRerouteUsage_84}
    B {from=MaterialExpressionNamedRerouteUsage_93}
   ]
  }
  MaterialExpressionMax_2 {
   class=MaterialExpressionMax
   inputs [
    A {from=MaterialExpressionMax_5}
    B {from=MaterialExpressionNamedRerouteUsage_139}
   ]
  }
  MaterialExpressionMax_20 {
   class=MaterialExpressionMax
   inputs [
    A {from=MaterialExpressionNamedRerouteUsage_92}
    B {from=MaterialExpressionConstant_3}
   ]
  }
  MaterialExpressionMax_21 {
   class=MaterialExpressionMax
   inputs [
    A {from=MaterialExpressionMax_17}
    B {from=MaterialExpressionConstant_5}
   ]
  }
  MaterialExpressionMax_22 {
   class=MaterialExpressionMax
   inputs [
    A {from=MaterialExpressionMax_32}
    B {from=MaterialExpressionNamedRerouteUsage_150}
   ]
  }
  MaterialExpressionMax_23 {
   class=MaterialExpressionMax
   inputs [
    A {from=MaterialExpressionMax_49}
    B {from=MaterialExpressionConstant_4}
   ]
  }
  MaterialExpressionMax_24 {
   class=MaterialExpressionMax
   inputs [
    A {from=MaterialExpressionSaturate_31}
    B {from=MaterialExpressionNamedRerouteUsage_5}
   ]
  }
  MaterialExpressionMax_25 {
   class=MaterialExpressionMax
   inputs [
    A {from=MaterialExpressionMax_2}
    B {from=MaterialExpressionNamedRerouteUsage_209}
   ]
  }
  MaterialExpressionMax_26 {
   class=MaterialExpressionMax
   inputs [
    A {from=MaterialExpressionMax_9}
    B {from=MaterialExpressionNamedRerouteUsage_221}
   ]
  }
  MaterialExpressionMax_27 {
   class=MaterialExpressionMax
   inputs [
    A {from=MaterialExpressionNamedRerouteDeclaration_7}
    B {from=MaterialExpressionNamedRerouteDeclaration_8}
   ]
  }
  MaterialExpressionMax_28 {
   class=MaterialExpressionMax
   inputs [
    A {from=MaterialExpressionNamedRerouteDeclaration_4}
    B {from=MaterialExpressionNamedRerouteDeclaration_6}
   ]
  }
  MaterialExpressionMax_29 {
   class=MaterialExpressionMax
   inputs [
    A {from=MaterialExpressionMax_26}
    B {from=MaterialExpressionNamedRerouteUsage_222}
   ]
  }
  MaterialExpressionMax_3 {
   class=MaterialExpressionMax
   inputs [
    A {from=MaterialExpressionNamedRerouteUsage_131}
    B {from=MaterialExpressionNamedRerouteUsage_132}
   ]
  }
  MaterialExpressionMax_30 {
   class=MaterialExpressionMax
   inputs [
    A {from=MaterialExpressionNamedRerouteUsage_69}
    B {from=MaterialExpressionNamedRerouteUsage_127}
   ]
  }
  MaterialExpressionMax_31 {
   class=MaterialExpressionMax
   inputs [
    A {from=MaterialExpressionMax_30}
    B {from=MaterialExpressionNamedRerouteUsage_128}
   ]
  }
  MaterialExpressionMax_32 {
   class=MaterialExpressionMax
   inputs [
    A {from=MaterialExpressionMax_31}
    B {from=MaterialExpressionNamedRerouteUsage_146}
   ]
  }
  MaterialExpressionMax_33 {
   class=MaterialExpressionMax
   inputs [
    A {from=MaterialExpressionMax_34}
    B {from=MaterialExpressionNamedRerouteUsage_369}
   ]
  }
  MaterialExpressionMax_34 {
   class=MaterialExpressionMax
   inputs [
    A {from=MaterialExpressionMax_50}
    B {from=MaterialExpressionNamedRerouteUsage_308}
   ]
  }
  MaterialExpressionMax_35 {
   class=MaterialExpressionMax
   inputs [
    A {from=MaterialExpressionMax_36}
    B {from=MaterialExpressionNamedRerouteUsage_65}
   ]
  }
  MaterialExpressionMax_36 {
   class=MaterialExpressionMax
   inputs [
    A {from=MaterialExpressionMax_41}
    B {from=MaterialExpressionNamedRerouteUsage_67}
   ]
  }
  MaterialExpressionMax_37 {
   class=MaterialExpressionMax
   inputs [
    A {from=MaterialExpressionMax_22}
    B {from=MaterialExpressionNamedRerouteUsage_168}
   ]
  }
  MaterialExpressionMax_38 {
   class=MaterialExpressionMax
   inputs [
    A {from=MaterialExpressionNamedRerouteUsage_172}
    B {from=MaterialExpressionMax_11}
   ]
  }
  MaterialExpressionMax_39 {
   class=MaterialExpressionMax
   inputs [
    A {from=MaterialExpressionNamedRerouteUsage_157}
    B {from=MaterialExpressionNamedRerouteUsage_158}
   ]
  }
  MaterialExpressionMax_4 {
   class=MaterialExpressionMax
   inputs [
    A {from=MaterialExpressionMax_3}
    B {from=MaterialExpressionSaturate_12}
   ]
  }
  MaterialExpressionMax_40 {
   class=MaterialExpressionMax
   inputs [
    A {from=MaterialExpressionReroute_0}
    B {from=MaterialExpressionSaturate_5}
   ]
  }
  MaterialExpressionMax_41 {
   class=MaterialExpressionMax
   inputs [
    A {from=MaterialExpressionMax_39}
    B {from=MaterialExpressionNamedRerouteUsage_68}
   ]
  }
  MaterialExpressionMax_42 {
   class=MaterialExpressionMax
   inputs [
    A {from=MaterialExpressionMax_46}
    B {from=MaterialExpressionNamedRerouteUsage_243}
   ]
  }
  MaterialExpressionMax_43 {
   class=MaterialExpressionMax
   inputs [
    A {from=MaterialExpressionMax_0}
    B {from=MaterialExpressionMax_1}
   ]
  }
  MaterialExpressionMax_44 {
   class=MaterialExpressionMax
   inputs [
    A {from=MaterialExpressionMax_42}
    B {from=MaterialExpressionNamedRerouteUsage_248}
   ]
  }
  MaterialExpressionMax_45 {
   class=MaterialExpressionMax
   inputs [
    A {from=MaterialExpressionMax_28}
    B {from=MaterialExpressionNamedRerouteDeclaration_7}
   ]
  }
  MaterialExpressionMax_46 {
   class=MaterialExpressionMax
   inputs [
    A {from=MaterialExpressionMax_29}
    B {from=MaterialExpressionNamedRerouteUsage_259}
   ]
  }
  MaterialExpressionMax_47 {
   class=MaterialExpressionMax
   inputs [
    A {from=MaterialExpressionMax_70}
    B {from=MaterialExpressionNamedRerouteUsage_167}
   ]
  }
  MaterialExpressionMax_48 {
   class=MaterialExpressionMax
   inputs [
    A {from=MaterialExpressionNamedRerouteUsage_344}
    B {from=MaterialExpressionNamedRerouteUsage_348}
   ]
  }
  MaterialExpressionMax_49 {
   class=MaterialExpressionMax
   inputs [
    A {from=MaterialExpressionMax_19}
    B {from=MaterialExpressionNamedRerouteUsage_180}
   ]
  }
  MaterialExpressionMax_5 {
   class=MaterialExpressionMax
   inputs [
    A {from=MaterialExpressionMax_4}
    B {from=MaterialExpressionNamedRerouteUsage_135}
   ]
  }
  MaterialExpressionMax_50 {
   class=MaterialExpressionMax
   inputs [
    A {from=MaterialExpressionMax_12}
    B {from=MaterialExpressionNamedRerouteUsage_327}
   ]
  }
  MaterialExpressionMax_51 {
   class=MaterialExpressionMax
   inputs [
    A {from=MaterialExpressionMax_43}
    B {from=MaterialExpressionMax_45}
   ]
  }
  MaterialExpressionMax_52 {
   class=MaterialExpressionMax
   inputs [
    A {from=MaterialExpressionMax_35}
    B {from=MaterialExpressionNamedRerouteUsage_368}
   ]
  }
  MaterialExpressionMax_53 {
   class=MaterialExpressionMax
   inputs [
    A {from=MaterialExpressionMax_24}
    B {from=MaterialExpressionNamedRerouteUsage_16}
   ]
  }
  MaterialExpressionMax_54 {
   class=MaterialExpressionMax
   inputs [
    A {from=MaterialExpressionMax_53}
    B {from=MaterialExpressionNamedRerouteUsage_318}
   ]
  }
  MaterialExpressionMax_55 {
   class=MaterialExpressionMax
   inputs [
    A {from=MaterialExpressionMax_44}
    B {from=MaterialExpressionNamedRerouteUsage_235}
   ]
  }
  MaterialExpressionMax_56 {
   class=MaterialExpressionMax
   inputs [
    A {from=MaterialExpressionNamedRerouteUsage_38}
    B {from=MaterialExpressionNamedRerouteUsage_149}
   ]
  }
  MaterialExpressionMax_57 {
   class=MaterialExpressionMax
   inputs [
    A {from=MaterialExpressionNamedRerouteUsage_155}
    B {from=MaterialExpressionNamedRerouteUsage_367}
   ]
  }
  MaterialExpressionMax_59 {
   class=MaterialExpressionMax
   inputs [
    A {from=MaterialExpressionMax_56}
    B {from=MaterialExpressionNamedRerouteUsage_329}
   ]
  }
  MaterialExpressionMax_6 {
   class=MaterialExpressionMax
   inputs [
    A {from=MaterialExpressionMax_51}
    B {from=MaterialExpressionNamedRerouteDeclaration_11}
   ]
  }
  MaterialExpressionMax_60 {
   class=MaterialExpressionMax
   inputs [
    A {from=MaterialExpressionMax_72}
    B {from=MaterialExpressionNamedRerouteUsage_334}
   ]
  }
  MaterialExpressionMax_61 {
   class=MaterialExpressionMax
   inputs [
    A {from=MaterialExpressionMax_60}
    B {from=MaterialExpressionNamedRerouteUsage_333}
   ]
  }
  MaterialExpressionMax_62 {
   class=MaterialExpressionMax
   inputs [
    A {from=MaterialExpressionMax_61}
    B {from=MaterialExpressionNamedRerouteUsage_336}
   ]
  }
  MaterialExpressionMax_63 {
   class=MaterialExpressionMax
   inputs [
    A {from=MaterialExpressionMax_62}
    B {from=MaterialExpressionNamedRerouteUsage_335}
   ]
  }
  MaterialExpressionMax_65 {
   class=MaterialExpressionMax
   inputs [
    A {from=MaterialExpressionNamedRerouteUsage_341}
    B {from=MaterialExpressionNamedRerouteUsage_342}
   ]
  }
  MaterialExpressionMax_66 {
   class=MaterialExpressionMax
   inputs [
    A {from=MaterialExpressionMax_74}
    B {from=MaterialExpressionNamedRerouteUsage_343}
   ]
  }
  MaterialExpressionMax_67 {
   class=MaterialExpressionMax
   inputs [
    A {from=MaterialExpressionMax_66}
    B {from=MaterialExpressionNamedRerouteUsage_346}
   ]
  }
  MaterialExpressionMax_68 {
   class=MaterialExpressionMax
   inputs [
    A {from=MaterialExpressionMax_67}
    B {from=MaterialExpressionNamedRerouteUsage_345}
   ]
  }
  MaterialExpressionMax_69 {
   class=MaterialExpressionMax
   inputs [
    A {from=MaterialExpressionMax_48}
    B {from=MaterialExpressionNamedRerouteUsage_362}
   ]
  }
  MaterialExpressionMax_7 {
   class=MaterialExpressionMax
   inputs [
    A {from=MaterialExpressionSaturate_17}
    B {from=MaterialExpressionNamedRerouteUsage_106}
   ]
  }
  MaterialExpressionMax_70 {
   class=MaterialExpressionMax
   inputs [
    A {from=MaterialExpressionMax_37}
    B {from=MaterialExpressionNamedRerouteUsage_365}
   ]
  }
  MaterialExpressionMax_71 {
   class=MaterialExpressionMax
   inputs [
    A {from=MaterialExpressionMax_63}
    B {from=MaterialExpressionSaturate_0}
   ]
  }
  MaterialExpressionMax_72 {
   class=MaterialExpressionMax
   inputs [
    A {from=MaterialExpressionMax_59}
    B {from=MaterialExpressionNamedRerouteUsage_354}
   ]
  }
  MaterialExpressionMax_73 {
   class=MaterialExpressionMax
   inputs [
    A {from=MaterialExpressionOneMinus_2}
    B {from=MaterialExpressionNamedRerouteUsage_366}
   ]
  }
  MaterialExpressionMax_74 {
   class=MaterialExpressionMax
   inputs [
    A {from=MaterialExpressionMax_65}
    B {from=MaterialExpressionNamedRerouteUsage_355}
   ]
  }
  MaterialExpressionMax_75 {
   class=MaterialExpressionMax
   inputs [
    A {from=MaterialExpressionMax_77}
    B {from=MaterialExpressionSaturate_2}
   ]
  }
  MaterialExpressionMax_76 {
   class=MaterialExpressionMax
   inputs [
    A {from=MaterialExpressionNamedRerouteUsage_332}
    B {from=MaterialExpressionNamedRerouteUsage_328}
   ]
  }
  MaterialExpressionMax_77 {
   class=MaterialExpressionMax
   inputs [
    A {from=MaterialExpressionMax_76}
    B {from=MaterialExpressionNamedRerouteUsage_317}
   ]
  }
  MaterialExpressionMax_8 {
   class=MaterialExpressionMax
   inputs [
    A {from=MaterialExpressionNamedRerouteUsage_48}
    B {from=MaterialExpressionMax_18}
   ]
  }
  MaterialExpressionMax_9 {
   class=MaterialExpressionMax
   inputs [
    A {from=MaterialExpressionNamedRerouteUsage_217}
    B {from=MaterialExpressionNamedRerouteUsage_219}
   ]
  }
  MaterialExpressionMeshPartitionChannelSample_1 {
   class=MaterialExpressionMeshPartitionChannelSample
   inputs [ChannelTextureInput {from=MaterialExpressionMeshPartitionResource_1, output=Channel Texture}]
   properties {channel=Mat_Gravel, desc=Mat_Gravel, meshPartitionDefinition=/Hera/Environment/Terrain/MeshTerrain/MPD_Hera}
  }
  MaterialExpressionMeshPartitionChannelSample_12 {
   class=MaterialExpressionMeshPartitionChannelSample
   inputs [ChannelTextureInput {from=MaterialExpressionMeshPartitionResource_1, output=Channel Texture}]
   properties {channel=Biome_D, desc=Biome_D, meshPartitionDefinition=&MPD_Hera}
  }
  MaterialExpressionMeshPartitionChannelSample_2 {
   class=MaterialExpressionMeshPartitionChannelSample
   inputs [ChannelTextureInput {from=MaterialExpressionMeshPartitionResource_1, output=Channel Texture}]
   properties {channel=Mat_Mud, desc=Mat_Mud, meshPartitionDefinition=&MPD_Hera}
  }
  MaterialExpressionMeshPartitionChannelSample_3 {
   class=MaterialExpressionMeshPartitionChannelSample
   inputs [ChannelTextureInput {from=MaterialExpressionMeshPartitionResource_1, output=Channel Texture}]
   properties {channel=Mat_Sand, desc=Mat_Sand, meshPartitionDefinition=&MPD_Hera}
  }
  MaterialExpressionMeshPartitionChannelSample_4 {
   class=MaterialExpressionMeshPartitionChannelSample
   inputs [ChannelTextureInput {from=MaterialExpressionMeshPartitionResource_1, output=Channel Texture}]
   properties {channel=Mat_Asphalt, desc=Mat_Asphalt, meshPartitionDefinition=&MPD_Hera}
  }
  MaterialExpressionMeshPartitionChannelSample_5 {
   class=MaterialExpressionMeshPartitionChannelSample
   inputs [ChannelTextureInput {from=MaterialExpressionMeshPartitionResource_1, output=Channel Texture}]
   properties {channel=Biome_A, desc=Biome_A, meshPartitionDefinition=&MPD_Hera}
  }
  MaterialExpressionMeshPartitionChannelSample_6 {
   class=MaterialExpressionMeshPartitionChannelSample
   inputs [ChannelTextureInput {from=MaterialExpressionMeshPartitionResource_1, output=Channel Texture}]
   properties {channel=Biome_B, desc=Biome_B, meshPartitionDefinition=&MPD_Hera}
  }
  MaterialExpressionMeshPartitionChannelSample_7 {
   class=MaterialExpressionMeshPartitionChannelSample
   inputs [ChannelTextureInput {from=MaterialExpressionMeshPartitionResource_1, output=Channel Texture}]
   properties {channel=Biome_C, desc=Biome_C, meshPartitionDefinition=&MPD_Hera}
  }
  MaterialExpressionMeshPartitionResource_1 {class=MaterialExpressionMeshPartitionResource, inputs [], properties {}}
  MaterialExpressionMultiply_0 {
   class=MaterialExpressionMultiply
   inputs [
    A {from=MaterialExpressionSaturate_3}
    B {from=MaterialExpressionOneMinus_20}
   ]
  }
  MaterialExpressionMultiply_1 {
   class=MaterialExpressionMultiply
   inputs [
    A {from=MaterialExpressionMaterialFunctionCall_15, output=LayerMask}
    B {from=MaterialExpressionScalarParameter_1}
   ]
  }
  MaterialExpressionMultiply_10 {
   class=MaterialExpressionMultiply
   inputs [
    A {from=MaterialExpressionNamedRerouteUsage_159}
    B {from=MaterialExpressionConstant_35}
   ]
  }
  MaterialExpressionMultiply_11 {
   class=MaterialExpressionMultiply
   inputs [
    A {from=MaterialExpressionSaturate_10}
    B {from=MaterialExpressionMultiply_69}
   ]
  }
  MaterialExpressionMultiply_12 {
   class=MaterialExpressionMultiply
   inputs [
    A {from=MaterialExpressionMultiply_13}
    B {from=MaterialExpressionNamedRerouteUsage_349}
   ]
  }
  MaterialExpressionMultiply_13 {
   class=MaterialExpressionMultiply
   inputs [
    A {from=MaterialExpressionMultiply_20}
    B {from=MaterialExpressionNamedRerouteUsage_350}
   ]
  }
  MaterialExpressionMultiply_14 {
   class=MaterialExpressionMultiply
   inputs [
    A {from=MaterialExpressionGetMaterialAttributes_7, output=Base Color}
    B {from=MaterialExpressionScalarParameter_23}
   ]
  }
  MaterialExpressionMultiply_15 {
   class=MaterialExpressionMultiply
   inputs [
    A {from=MaterialExpressionMax_38}
    B {from=MaterialExpressionOneMinus_9}
   ]
  }
  MaterialExpressionMultiply_16 {
   class=MaterialExpressionMultiply
   inputs [
    A {from=MaterialExpressionNamedRerouteUsage_79}
    B {from=MaterialExpressionOneMinus_10}
   ]
  }
  MaterialExpressionMultiply_17 {
   class=MaterialExpressionMultiply
   inputs [
    A {from=MaterialExpressionMax_8}
    B {from=MaterialExpressionOneMinus_11}
   ]
  }
  MaterialExpressionMultiply_18 {
   class=MaterialExpressionMultiply
   inputs [
    A {from=MaterialExpressionNamedRerouteUsage_86}
    B {from=MaterialExpressionOneMinus_12}
   ]
  }
  MaterialExpressionMultiply_19 {
   class=MaterialExpressionMultiply
   inputs [
    A {from=MaterialExpressionOneMinus_17}
    B {from=MaterialExpressionOneMinus_13}
   ]
  }
  MaterialExpressionMultiply_2 {
   class=MaterialExpressionMultiply
   inputs [
    A {from=MaterialExpressionMaterialFunctionCall_33, output=BorealForest_GrassMask}
    B {from=MaterialExpressionScalarParameter_7}
   ]
  }
  MaterialExpressionMultiply_20 {
   class=MaterialExpressionMultiply
   inputs [
    A {from=MaterialExpressionMultiply_49}
    B {from=MaterialExpressionNamedRerouteUsage_340}
   ]
  }
  MaterialExpressionMultiply_21 {
   class=MaterialExpressionMultiply
   inputs [
    A {from=MaterialExpressionMultiply_66}
    B {from=MaterialExpressionMultiply_31}
   ]
  }
  MaterialExpressionMultiply_22 {
   class=MaterialExpressionMultiply
   inputs [
    A {from=MaterialExpressionOneMinus_19}
    B {from=MaterialExpressionNamedRerouteUsage_152}
   ]
  }
  MaterialExpressionMultiply_23 {
   class=MaterialExpressionMultiply
   inputs [
    A {from=MaterialExpressionMultiply_11}
    B {from=MaterialExpressionNamedRerouteUsage_100}
   ]
  }
  MaterialExpressionMultiply_24 {
   class=MaterialExpressionMultiply
   inputs [
    A {from=MaterialExpressionConstant_30}
    B {from=MaterialExpressionMultiply_19}
   ]
  }
  MaterialExpressionMultiply_25 {
   class=MaterialExpressionMultiply
   inputs [
    A {from=MaterialExpressionMultiply_68}
    B {from=MaterialExpressionNamedRerouteUsage_102}
   ]
  }
  MaterialExpressionMultiply_26 {
   class=MaterialExpressionMultiply
   inputs [
    A {from=MaterialExpressionMaterialFunctionCall_19, output=ChaparralCoast_GrassMask}
    B {from=MaterialExpressionScalarParameter_5}
   ]
  }
  MaterialExpressionMultiply_27 {
   class=MaterialExpressionMultiply
   inputs [
    A {from=MaterialExpressionStaticSwitch_16}
    B {from=MaterialExpressionNamedRerouteUsage_104}
   ]
  }
  MaterialExpressionMultiply_28 {
   class=MaterialExpressionMultiply
   inputs [
    A {from=MaterialExpressionSaturate_1}
    B {from=MaterialExpressionNamedRerouteUsage_270}
   ]
  }
  MaterialExpressionMultiply_29 {
   class=MaterialExpressionMultiply
   inputs [
    A {from=MaterialExpressionNamedRerouteUsage_162}
    B {from=MaterialExpressionConstant_11}
   ]
  }
  MaterialExpressionMultiply_3 {
   class=MaterialExpressionMultiply
   inputs [
    A {from=""}
    B {from=MaterialExpressionNamedRerouteUsage_54}
   ]
  }
  MaterialExpressionMultiply_30 {
   class=MaterialExpressionMultiply
   inputs [
    A {from=MaterialExpressionMultiply_8}
    B {from=MaterialExpressionNamedRerouteUsage_105}
   ]
  }
  MaterialExpressionMultiply_31 {
   class=MaterialExpressionMultiply
   inputs [
    A {from=MaterialExpressionMultiply_28}
    B {from=MaterialExpressionNamedRerouteUsage_151}
   ]
  }
  MaterialExpressionMultiply_32 {
   class=MaterialExpressionMultiply
   inputs [
    A {from=MaterialExpressionMultiply_0}
    B {from=MaterialExpressionNamedRerouteUsage_108}
   ]
  }
  MaterialExpressionMultiply_33 {
   class=MaterialExpressionMultiply
   inputs [
    A {from=MaterialExpressionGetMaterialAttributes_1, output=Base Color}
    B {from=MaterialExpressionVectorParameter_0, output=RGB}
   ]
  }
  MaterialExpressionMultiply_34 {
   class=MaterialExpressionMultiply
   inputs [
    A {from=MaterialExpressionMultiply_22}
    B {from=MaterialExpressionNamedRerouteUsage_160}
   ]
  }
  MaterialExpressionMultiply_35 {
   class=MaterialExpressionMultiply
   inputs [
    A {from=MaterialExpressionSaturate_24}
    B {from=MaterialExpressionOneMinus_25}
   ]
  }
  MaterialExpressionMultiply_36 {
   class=MaterialExpressionMultiply
   inputs [
    A {from=MaterialExpressionScalarParameter_9}
    B {from=MaterialExpressionIf_0}
   ]
  }
  MaterialExpressionMultiply_37 {
   class=MaterialExpressionMultiply
   inputs [
    A {from=MaterialExpressionGetMaterialAttributes_3, output=Base Color}
    B {from=MaterialExpressionVectorParameter_1, output=RGB}
   ]
  }
  MaterialExpressionMultiply_38 {
   class=MaterialExpressionMultiply
   inputs [
    A {from=MaterialExpressionGetMaterialAttributes_2, output=Displacement}
    B {from=MaterialExpressionScalarParameter_10}
   ]
  }
  MaterialExpressionMultiply_39 {
   class=MaterialExpressionMultiply
   inputs [
    A {from=MaterialExpressionConstant_18}
    B {from=MaterialExpressionNamedRerouteUsage_56}
   ]
  }
  MaterialExpressionMultiply_4 {
   class=MaterialExpressionMultiply
   inputs [
    A {from=MaterialExpressionNamedRerouteUsage_42}
    B {from=MaterialExpressionConstant_37}
   ]
  }
  MaterialExpressionMultiply_40 {
   class=MaterialExpressionMultiply
   inputs [
    A {from=MaterialExpressionOneMinus_16}
    B {from=MaterialExpressionNamedRerouteUsage_143}
   ]
  }
  MaterialExpressionMultiply_41 {
   class=MaterialExpressionMultiply
   inputs [
    A {from=MaterialExpressionOneMinus_18}
    B {from=MaterialExpressionNamedRerouteUsage_154}
   ]
  }
  MaterialExpressionMultiply_42 {
   class=MaterialExpressionMultiply
   inputs [
    A {from=MaterialExpressionSaturate_20}
    B {from=MaterialExpressionOneMinus_5}
   ]
  }
  MaterialExpressionMultiply_43 {
   class=MaterialExpressionMultiply
   inputs [
    A {from=MaterialExpressionOneMinus_7}
    B {from=MaterialExpressionNamedRerouteUsage_254}
   ]
  }
  MaterialExpressionMultiply_44 {
   class=MaterialExpressionMultiply
   inputs [
    A {from=MaterialExpressionNamedRerouteUsage_134}
    B {from=""}
   ]
   properties {constB=2.75}
  }
  MaterialExpressionMultiply_45 {
   class=MaterialExpressionMultiply
   inputs [
    A {from=MaterialExpressionNamedRerouteUsage_210}
    B {from=MaterialExpressionConstant_36}
   ]
  }
  MaterialExpressionMultiply_46 {
   class=MaterialExpressionMultiply
   inputs [
    A {from=MaterialExpressionOneMinus_14}
    B {from=MaterialExpressionNamedRerouteUsage_356}
   ]
  }
  MaterialExpressionMultiply_47 {
   class=MaterialExpressionMultiply
   inputs [
    A {from=MaterialExpressionOneMinus_1}
    B {from=MaterialExpressionNamedRerouteUsage_143}
   ]
  }
  MaterialExpressionMultiply_48 {
   class=MaterialExpressionMultiply
   inputs [
    A {from=MaterialExpressionMultiply_40}
    B {from=MaterialExpressionNamedRerouteUsage_145}
   ]
  }
  MaterialExpressionMultiply_49 {
   class=MaterialExpressionMultiply
   inputs [
    A {from=MaterialExpressionOneMinus_15}
    B {from=MaterialExpressionNamedRerouteUsage_347}
   ]
  }
  MaterialExpressionMultiply_5 {
   class=MaterialExpressionMultiply
   inputs [
    A {from=MaterialExpressionNamedRerouteUsage_320}
    B {from=MaterialExpressionConstant_85}
   ]
  }
  MaterialExpressionMultiply_50 {
   class=MaterialExpressionMultiply
   inputs [
    A {from=MaterialExpressionMultiply_47}
    B {from=MaterialExpressionNamedRerouteUsage_145}
   ]
  }
  MaterialExpressionMultiply_51 {
   class=MaterialExpressionMultiply
   inputs [
    A {from=MaterialExpressionNamedRerouteUsage_53}
    B {from=MaterialExpressionNamedRerouteUsage_218}
   ]
  }
  MaterialExpressionMultiply_52 {
   class=MaterialExpressionMultiply
   inputs [
    A {from=MaterialExpressionNamedRerouteUsage_286}
    B {from=MaterialExpressionNamedRerouteUsage_287}
   ]
  }
  MaterialExpressionMultiply_53 {
   class=MaterialExpressionMultiply
   inputs [
    A {from=MaterialExpressionNamedRerouteUsage_352}
    B {from=MaterialExpressionConstant_49}
   ]
  }
  MaterialExpressionMultiply_54 {
   class=MaterialExpressionMultiply
   inputs [
    A {from=MaterialExpressionOneMinus_26}
    B {from=MaterialExpressionConstant_50}
   ]
  }
  MaterialExpressionMultiply_55 {
   class=MaterialExpressionMultiply
   inputs [
    A {from=MaterialExpressionMultiply_61}
    B {from=MaterialExpressionNamedRerouteUsage_353}
   ]
  }
  MaterialExpressionMultiply_56 {
   class=MaterialExpressionMultiply
   inputs [
    A {from=MaterialExpressionMultiply_55}
    B {from=MaterialExpressionNamedRerouteUsage_360}
   ]
  }
  MaterialExpressionMultiply_57 {
   class=MaterialExpressionMultiply
   inputs [
    A {from=MaterialExpressionOneMinus_22}
    B {from=MaterialExpressionScalarParameter_24}
   ]
  }
  MaterialExpressionMultiply_58 {
   class=MaterialExpressionMultiply
   inputs [
    A {from=MaterialExpressionMultiply_59}
    B {from=MaterialExpressionNamedRerouteUsage_324}
   ]
  }
  MaterialExpressionMultiply_59 {
   class=MaterialExpressionMultiply
   inputs [
    A {from=MaterialExpressionMultiply_35}
    B {from=MaterialExpressionNamedRerouteUsage_325}
   ]
  }
  MaterialExpressionMultiply_6 {
   class=MaterialExpressionMultiply
   inputs [
    A {from=MaterialExpressionMultiply_7}
    B {from=MaterialExpressionNamedRerouteUsage_337}
   ]
  }
  MaterialExpressionMultiply_60 {
   class=MaterialExpressionMultiply
   inputs [
    A {from=MaterialExpressionMultiply_62}
    B {from=MaterialExpressionNamedRerouteUsage_361}
   ]
  }
  MaterialExpressionMultiply_61 {
   class=MaterialExpressionMultiply
   inputs [
    A {from=MaterialExpressionMultiply_60}
    B {from=MaterialExpressionNamedRerouteUsage_364}
   ]
  }
  MaterialExpressionMultiply_62 {
   class=MaterialExpressionMultiply
   inputs [
    A {from=MaterialExpressionSaturate_0}
    B {from=MaterialExpressionSaturate_4}
   ]
  }
  MaterialExpressionMultiply_63 {
   class=MaterialExpressionMultiply
   inputs [
    A {from=MaterialExpressionGetMaterialAttributes_6, output=Opacity Mask}
    B {from=MaterialExpressionSaturate_23}
   ]
  }
  MaterialExpressionMultiply_64 {
   class=MaterialExpressionMultiply
   inputs [
    A {from=MaterialExpressionRound_3}
    B {from=MaterialExpressionTextureSample_6, output=R}
   ]
  }
  MaterialExpressionMultiply_65 {
   class=MaterialExpressionMultiply
   inputs [
    A {from=MaterialExpressionMultiply_63}
    B {from=MaterialExpressionConstant_52}
   ]
  }
  MaterialExpressionMultiply_66 {
   class=MaterialExpressionMultiply
   inputs [
    A {from=MaterialExpressionOneMinus_8}
    B {from=MaterialExpressionNamedRerouteUsage_311}
   ]
  }
  MaterialExpressionMultiply_67 {
   class=MaterialExpressionMultiply
   inputs [
    A {from=MaterialExpressionScalarParameter_4}
    B {from=MaterialExpressionGetMaterialAttributes_6, output=Roughness}
   ]
  }
  MaterialExpressionMultiply_68 {
   class=MaterialExpressionMultiply
   inputs [
    A {from=MaterialExpressionMultiply_41}
    B {from=MaterialExpressionNamedRerouteUsage_313}
   ]
  }
  MaterialExpressionMultiply_69 {
   class=MaterialExpressionMultiply
   inputs [
    A {from=MaterialExpressionSaturate_8}
    B {from=MaterialExpressionNamedRerouteUsage_314}
   ]
  }
  MaterialExpressionMultiply_7 {
   class=MaterialExpressionMultiply
   inputs [
    A {from=MaterialExpressionMultiply_46}
    B {from=MaterialExpressionNamedRerouteUsage_338}
   ]
  }
  MaterialExpressionMultiply_70 {
   class=MaterialExpressionMultiply
   inputs [
    A {from=MaterialExpressionScalarParameter_12}
    B {from=MaterialExpressionGetMaterialAttributes_6, output=Roughness}
   ]
  }
  MaterialExpressionMultiply_71 {
   class=MaterialExpressionMultiply
   inputs [
    A {from=MaterialExpressionScalarParameter_13}
    B {from=MaterialExpressionMultiply_63}
   ]
  }
  MaterialExpressionMultiply_72 {
   class=MaterialExpressionMultiply
   inputs [
    A {from=MaterialExpressionSaturate_14}
    B {from=MaterialExpressionConstant_53}
   ]
  }
  MaterialExpressionMultiply_73 {
   class=MaterialExpressionMultiply
   inputs [
    A {from=MaterialExpressionMaterialFunctionCall_15, output=LayerMask}
    B {from=MaterialExpressionSaturate_29}
   ]
  }
  MaterialExpressionMultiply_74 {
   class=MaterialExpressionMultiply
   inputs [
    A {from=MaterialExpressionMultiply_51}
    B {from=MaterialExpressionOneMinus_23}
   ]
  }
  MaterialExpressionMultiply_75 {
   class=MaterialExpressionMultiply
   inputs [
    A {from=MaterialExpressionNamedRerouteUsage_375}
    B {from=MaterialExpressionNamedRerouteUsage_370}
   ]
  }
  MaterialExpressionMultiply_8 {
   class=MaterialExpressionMultiply
   inputs [
    A {from=MaterialExpressionNamedRerouteUsage_60}
    B {from=MaterialExpressionOneMinus_21}
   ]
  }
  MaterialExpressionMultiply_9 {
   class=MaterialExpressionMultiply
   inputs [
    A {from=MaterialExpressionOneMinus_4}
    B {from=MaterialExpressionConstant_7}
   ]
  }
  MaterialExpressionNamedRerouteDeclaration_0 {
   class=MaterialExpressionNamedRerouteDeclaration
   inputs [None {from=MaterialExpressionStaticSwitchParameter_1}]
   properties {name=Gravel_Weight, nodeColor=<rgba 0.5 0 1 1>}
  }
  MaterialExpressionNamedRerouteDeclaration_1 {
   class=MaterialExpressionNamedRerouteDeclaration
   inputs [None {from=MaterialExpressionStaticSwitchParameter_3}]
   properties {name=Sand_Weight, nodeColor=<rgba 0.5 0 1 1>}
  }
  MaterialExpressionNamedRerouteDeclaration_10 {
   class=MaterialExpressionNamedRerouteDeclaration
   inputs [None {from=MaterialExpressionStaticBoolParameter_0}]
   properties {name=IsLobbyIsland, nodeColor=<rgba 0.5 0 1 1>}
  }
  MaterialExpressionNamedRerouteDeclaration_100 {
   class=MaterialExpressionNamedRerouteDeclaration
   inputs [None {from=MaterialExpressionStaticSwitch_18}]
   properties {name=L0CC_DesertGrassMask, nodeColor=<rgba 0.5 0 1 1>}
  }
  MaterialExpressionNamedRerouteDeclaration_101 {
   class=MaterialExpressionNamedRerouteDeclaration
   inputs [None {from=MaterialExpressionStaticSwitch_39}]
   properties {name=MobileMaterialUpTo_Goldlands, nodeColor=<rgba 0.5 0 1 1>}
  }
  MaterialExpressionNamedRerouteDeclaration_102 {
   class=MaterialExpressionNamedRerouteDeclaration
   inputs [None {from=MaterialExpressionMaterialFunctionCall_18, output=SedimentLayers}]
   properties {name=Desert_MacroDetail_SedimentLayers, nodeColor=<rgba 0.5 0 1 1>}
  }
  MaterialExpressionNamedRerouteDeclaration_103 {
   class=MaterialExpressionNamedRerouteDeclaration
   inputs [None {from=MaterialExpressionStaticSwitch_31}]
   properties {name=Asphalt03LayerMask, nodeColor=<rgba 0.5 0 1 1>}
  }
  MaterialExpressionNamedRerouteDeclaration_104 {
   class=MaterialExpressionNamedRerouteDeclaration
   inputs [None {from=MaterialExpressionStaticSwitch_15}]
   properties {name=L0CB_Material, nodeColor=<rgba 0.5 0 1 1>}
  }
  MaterialExpressionNamedRerouteDeclaration_105 {
   class=MaterialExpressionNamedRerouteDeclaration
   inputs [None {from=MaterialExpressionStaticSwitch_17}]
   properties {name=L0CB_MobileMaterial, nodeColor=<rgba 0.5 0 1 1>}
  }
  MaterialExpressionNamedRerouteDeclaration_106 {
   class=MaterialExpressionNamedRerouteDeclaration
   inputs [None {from=MaterialExpressionStaticBoolParameter_8}]
   properties {name=UseAsphalt03, nodeColor=<rgba 0.5 0 1 1>}
  }
  MaterialExpressionNamedRerouteDeclaration_107 {
   class=MaterialExpressionNamedRerouteDeclaration
   inputs [None {from=MaterialExpressionStaticSwitch_19}]
   properties {name=MaterialUpTo_Asphalt03, nodeColor=<rgba 0.5 0 1 1>}
  }
  MaterialExpressionNamedRerouteDeclaration_108 {
   class=MaterialExpressionNamedRerouteDeclaration
   inputs [None {from=MaterialExpressionStaticSwitch_20}]
   properties {name=MobileMaterialUpTo_Asphalt03, nodeColor=<rgba 0.5 0 1 1>}
  }
  MaterialExpressionNamedRerouteDeclaration_109 {
   class=MaterialExpressionNamedRerouteDeclaration
   inputs [None {from=MaterialExpressionStaticBoolParameter_11}]
   properties {name=UseBiomeA03, nodeColor=<rgba 0.5 0 1 1>}
  }
  MaterialExpressionNamedRerouteDeclaration_11 {
   class=MaterialExpressionNamedRerouteDeclaration
   inputs [None {from=MaterialExpressionMaterialFunctionCall_16, output=Result}]
   properties {name=Cliff_Weight, nodeColor=<rgba 0.5 0 1 1>}
  }
  MaterialExpressionNamedRerouteDeclaration_110 {
   class=MaterialExpressionNamedRerouteDeclaration
   inputs [None {from=MaterialExpressionStaticSwitch_32}]
   properties {name=MaterialUpTo_BiomeA03, nodeColor=<rgba 0.5 0 1 1>}
  }
  MaterialExpressionNamedRerouteDeclaration_111 {
   class=MaterialExpressionNamedRerouteDeclaration
   inputs [None {from=MaterialExpressionStaticSwitch_33}]
   properties {name=MobileMaterialUpTo_BiomeA03, nodeColor=<rgba 0.5 0 1 1>}
  }
  MaterialExpressionNamedRerouteDeclaration_112 {
   class=MaterialExpressionNamedRerouteDeclaration
   inputs [None {from=MaterialExpressionStaticSwitch_37}]
   properties {name=BiomeA03LayerMask, nodeColor=<rgba 0.5 0 1 1>}
  }
  MaterialExpressionNamedRerouteDeclaration_113 {
   class=MaterialExpressionNamedRerouteDeclaration
   inputs [None {from=MaterialExpressionStaticSwitch_50}]
   properties {name=ArchipelagoSeaweedLayerMask, nodeColor=<rgba 0.5 0 1 1>}
  }
  MaterialExpressionNamedRerouteDeclaration_114 {
   class=MaterialExpressionNamedRerouteDeclaration
   inputs [None {from=MaterialExpressionStaticBoolParameter_16}]
   properties {name=UseBiomeC_ArchipelagoSeaweed, nodeColor=<rgba 0.5 0 1 1>}
  }
  MaterialExpressionNamedRerouteDeclaration_115 {
   class=MaterialExpressionNamedRerouteDeclaration
   inputs [None {from=MaterialExpressionStaticSwitch_48}]
   properties {name=MaterialUpTo_BiomeC01, nodeColor=<rgba 0.5 0 1 1>}
  }
  MaterialExpressionNamedRerouteDeclaration_116 {
   class=MaterialExpressionNamedRerouteDeclaration
   inputs [None {from=MaterialExpressionStaticSwitch_49}]
   properties {name=MobileMaterialUpTo_BiomeC01, nodeColor=<rgba 0.5 0 1 1>}
  }
  MaterialExpressionNamedRerouteDeclaration_117 {
   class=MaterialExpressionNamedRerouteDeclaration
   inputs [None {from=MaterialExpressionStaticSwitch_45}]
   properties {name=Asphalt04LayerMask, nodeColor=<rgba 0.5 0 1 1>}
  }
  MaterialExpressionNamedRerouteDeclaration_118 {
   class=MaterialExpressionNamedRerouteDeclaration
   inputs [None {from=MaterialExpressionStaticBoolParameter_12}]
   properties {name=UseAsphalt04, nodeColor=<rgba 0.5 0 1 1>}
  }
  MaterialExpressionNamedRerouteDeclaration_119 {
   class=MaterialExpressionNamedRerouteDeclaration
   inputs [None {from=MaterialExpressionStaticSwitch_43}]
   properties {name=MaterialUpTo_Asphalt04, nodeColor=<rgba 0.5 0 1 1>}
  }
  MaterialExpressionNamedRerouteDeclaration_12 {
   class=MaterialExpressionNamedRerouteDeclaration
   inputs [None {from=MaterialExpressionMaterialFunctionCall_2, output=LayerMask}]
   properties {name=Layer0_DefaultGrassLayerMask, nodeColor=<rgba 0.5 0 1 1>}
  }
  MaterialExpressionNamedRerouteDeclaration_120 {
   class=MaterialExpressionNamedRerouteDeclaration
   inputs [None {from=MaterialExpressionStaticSwitch_44}]
   properties {name=MobileMaterialUpTo_Asphalt04, nodeColor=<rgba 0.5 0 1 1>}
  }
  MaterialExpressionNamedRerouteDeclaration_121 {
   class=MaterialExpressionNamedRerouteDeclaration
   inputs [None {from=MaterialExpressionStaticSwitch_53}]
   properties {name=EchoFrostLayerMask, nodeColor=<rgba 0.5 0 1 1>}
  }
  MaterialExpressionNamedRerouteDeclaration_122 {
   class=MaterialExpressionNamedRerouteDeclaration
   inputs [None {from=MaterialExpressionStaticBoolParameter_17}]
   properties {name=UseBiomeD_Bigsplat, nodeColor=<rgba 0.5 0 1 1>}
  }
  MaterialExpressionNamedRerouteDeclaration_123 {
   class=MaterialExpressionNamedRerouteDeclaration
   inputs [None {from=MaterialExpressionStaticSwitch_51}]
   properties {name=MaterialUpTo_BiomeD01, nodeColor=<rgba 0.5 0 1 1>}
  }
  MaterialExpressionNamedRerouteDeclaration_124 {
   class=MaterialExpressionNamedRerouteDeclaration
   inputs [None {from=MaterialExpressionStaticSwitch_52}]
   properties {name=MobileMaterialUpTo_BiomeD01, nodeColor=<rgba 0.5 0 1 1>}
  }
  MaterialExpressionNamedRerouteDeclaration_125 {
   class=MaterialExpressionNamedRerouteDeclaration
   inputs [None {from=MaterialExpressionMaterialFunctionCall_18, output=UnmodifiedNormals}]
   properties {name=Desert_MacroDetail_UnmodifiedNormals, nodeColor=<rgba 0.5 0 1 1>}
  }
  MaterialExpressionNamedRerouteDeclaration_126 {
   class=MaterialExpressionNamedRerouteDeclaration
   inputs [None {from=MaterialExpressionMaterialFunctionCall_23, output=03_Desert}]
   properties {name=BM_Desert, nodeColor=<rgba 0.5 0 1 1>}
  }
  MaterialExpressionNamedRerouteDeclaration_127 {
   class=MaterialExpressionNamedRerouteDeclaration
   inputs [None {from=MaterialExpressionMaterialFunctionCall_23, output=04_GoldenFall}]
   properties {name=BM_GoldenFall, nodeColor=<rgba 0.5 0 1 1>}
  }
  MaterialExpressionNamedRerouteDeclaration_128 {
   class=MaterialExpressionNamedRerouteDeclaration
   inputs [None {from=MaterialExpressionMaterialFunctionCall_46, output=Result}]
   properties {name=LGT_GoldenFall, nodeColor=<rgba 0.5 0 1 1>}
  }
  MaterialExpressionNamedRerouteDeclaration_129 {
   class=MaterialExpressionNamedRerouteDeclaration
   inputs [None {from=MaterialExpressionMaterialFunctionCall_10, output=Result}]
   properties {name=LGT_Beach, nodeColor=<rgba 0.5 0 1 1>}
  }
  MaterialExpressionNamedRerouteDeclaration_13 {
   class=MaterialExpressionNamedRerouteDeclaration
   inputs [None {from=MaterialExpressionMaterialFunctionCall_4, output=LayerMask}]
   properties {name=SandLayerMask, nodeColor=<rgba 0.5 0 1 1>}
  }
  MaterialExpressionNamedRerouteDeclaration_130 {
   class=MaterialExpressionNamedRerouteDeclaration
   inputs [None {from=MaterialExpressionMaterialFunctionCall_21, output=Result}]
   properties {name=LGT_RedWoods, nodeColor=<rgba 0.5 0 1 1>}
  }
  MaterialExpressionNamedRerouteDeclaration_14 {
   class=MaterialExpressionNamedRerouteDeclaration
   inputs [None {from=MaterialExpressionMaterialFunctionCall_11, output=LayerMask}]
   properties {name=GravelLayerMask, nodeColor=<rgba 0.5 0 1 1>}
  }
  MaterialExpressionNamedRerouteDeclaration_15 {
   class=MaterialExpressionNamedRerouteDeclaration
   inputs [None {from=MaterialExpressionMaterialFunctionCall_6, output=LayerMask}]
   properties {name=MudLayerMask, nodeColor=<rgba 0.5 0 1 1>}
  }
  MaterialExpressionNamedRerouteDeclaration_16 {
   class=MaterialExpressionNamedRerouteDeclaration
   inputs [None {from=MaterialExpressionStaticSwitch_36}]
   properties {name=AsphaltLayerMask, nodeColor=<rgba 0.5 0 1 1>}
  }
  MaterialExpressionNamedRerouteDeclaration_17 {
   class=MaterialExpressionNamedRerouteDeclaration
   inputs [None {from=MaterialExpressionMaterialFunctionCall_5, output=GrassMask}]
   properties {name=CliffLayerGrassMask, nodeColor=<rgba 0.6172 0 1 1>}
  }
  MaterialExpressionNamedRerouteDeclaration_18 {
   class=MaterialExpressionNamedRerouteDeclaration
   inputs [None {from=MaterialExpressionNamedRerouteDeclaration_99}]
   properties {name=Highend:SurfaceMaterialResult, nodeColor=<rgba 0.5 0 1 1>}
  }
  MaterialExpressionNamedRerouteDeclaration_19 {
   class=MaterialExpressionNamedRerouteDeclaration
   inputs [None {from=MaterialExpressionReroute_2}]
   properties {name=TerrainMasks_QuadB, nodeColor=<rgba 0.5 0 1 1>}
  }
  MaterialExpressionNamedRerouteDeclaration_2 {
   class=MaterialExpressionNamedRerouteDeclaration
   inputs [None {from=MaterialExpressionStaticSwitchParameter_2}]
   properties {name=Mud_Weight, nodeColor=<rgba 0.5 0 1 1>}
  }
  MaterialExpressionNamedRerouteDeclaration_20 {
   class=MaterialExpressionNamedRerouteDeclaration
   inputs [None {from=MaterialExpressionReroute_1}]
   properties {name=TerrainMasks_QuadA, nodeColor=<rgba 0.5 0 1 1>}
  }
  MaterialExpressionNamedRerouteDeclaration_21 {
   class=MaterialExpressionNamedRerouteDeclaration
   inputs [None {from=MaterialExpressionReroute_4}]
   properties {name=TerrainMasks_QuadC, nodeColor=<rgba 0.5 0 1 1>}
  }
  MaterialExpressionNamedRerouteDeclaration_22 {
   class=MaterialExpressionNamedRerouteDeclaration
   inputs [None {from=MaterialExpressionMaterialFunctionCall_22, output=QuadD_RGBA}]
   properties {name=TerrainMasks_QuadD, nodeColor=<rgba 0.5 0 1 1>}
  }
  MaterialExpressionNamedRerouteDeclaration_23 {
   class=MaterialExpressionNamedRerouteDeclaration
   inputs [None {from=MaterialExpressionStaticBoolParameter_1}]
   properties {name=UseSnowForBiomeC, nodeColor=<rgba 0.5 0 1 1>}
  }
  MaterialExpressionNamedRerouteDeclaration_24 {
   class=MaterialExpressionNamedRerouteDeclaration
   inputs [None {from=MaterialExpressionMaterialFunctionCall_5, output=LayerMask}]
   properties {name=CliffLayerMask, nodeColor=<rgba 0.6172 0 1 1>}
  }
  MaterialExpressionNamedRerouteDeclaration_25 {
   class=MaterialExpressionNamedRerouteDeclaration
   inputs [None {from=MaterialExpressionMaterialFunctionCall_8, output=Result}]
   properties {name=LGT_ForestFloor01, nodeColor=<rgba 0.5 0 1 1>}
  }
  MaterialExpressionNamedRerouteDeclaration_26 {
   class=MaterialExpressionNamedRerouteDeclaration
   inputs [None {from=MaterialExpressionStaticSwitch_42}]
   properties {name=SnowLayerGrassMask, nodeColor=<rgba 0.6172 0 1 1>}
  }
  MaterialExpressionNamedRerouteDeclaration_27 {
   class=MaterialExpressionNamedRerouteDeclaration
   inputs [None {from=MaterialExpressionMaterialFunctionCall_32, output=Result}]
   properties {name=LGT_MossyForest, nodeColor=<rgba 0.5 0 1 1>}
  }
  MaterialExpressionNamedRerouteDeclaration_28 {
   class=MaterialExpressionNamedRerouteDeclaration
   inputs [None {from=MaterialExpressionStaticSwitch_5}]
   properties {name=L0CA_ChaparralGrassMask, nodeColor=<rgba 0.5 0 1 1>}
  }
  MaterialExpressionNamedRerouteDeclaration_29 {
   class=MaterialExpressionNamedRerouteDeclaration
   inputs [None {from=MaterialExpressionOneMinus_6}]
   properties {name=Grass_Weight_NoCliff, nodeColor=<rgba 0.5 0 1 1>}
  }
  MaterialExpressionNamedRerouteDeclaration_3 {
   class=MaterialExpressionNamedRerouteDeclaration
   inputs [None {from=MaterialExpressionStaticSwitchParameter_4}]
   properties {name=Asphalt_Weight, nodeColor=<rgba 0.5 0 1 1>}
  }
  MaterialExpressionNamedRerouteDeclaration_30 {
   class=MaterialExpressionNamedRerouteDeclaration
   inputs [None {from=MaterialExpressionNamedRerouteDeclaration_101}]
   properties {name=MOBILE:SurfaceMaterialResult, nodeColor=<rgba 0.5 0 1 1>}
  }
  MaterialExpressionNamedRerouteDeclaration_31 {
   class=MaterialExpressionNamedRerouteDeclaration
   inputs [None {from=MaterialExpressionMaterialFunctionCall_30, output=VariationMasks Mobile}]
   properties {name=VariationMasksMobile, nodeColor=<rgba 1 0.9141 0 1>}
  }
  MaterialExpressionNamedRerouteDeclaration_32 {
   class=MaterialExpressionNamedRerouteDeclaration
   inputs [None {from=MaterialExpressionSetMaterialAttributes_4}]
   properties {name=Snow_PostRVT, nodeColor=<rgba 0.6172 0 1 1>}
  }
  MaterialExpressionNamedRerouteDeclaration_33 {
   class=MaterialExpressionNamedRerouteDeclaration
   inputs [None {from=MaterialExpressionStaticSwitch_30}]
   properties {name=ScreeLayerMask, nodeColor=<rgba 0.5 0 1 1>}
  }
  MaterialExpressionNamedRerouteDeclaration_34 {
   class=MaterialExpressionNamedRerouteDeclaration
   inputs [None {from=MaterialExpressionMaterialFunctionCall_52, output=Result}]
   properties {name=LGT_ChaparralGrass, nodeColor=<rgba 0.5 0 1 1>}
  }
  MaterialExpressionNamedRerouteDeclaration_35 {
   class=MaterialExpressionNamedRerouteDeclaration
   inputs [None {from=MaterialExpressionMaterialFunctionCall_23, output=02_Archipelago}]
   properties {name=BM_Archipelago, nodeColor=<rgba 0.5 0 1 1>}
  }
  MaterialExpressionNamedRerouteDeclaration_36 {
   class=MaterialExpressionNamedRerouteDeclaration
   inputs [None {from=MaterialExpressionMaterialFunctionCall_23, output=00_Chaparral}]
   properties {name=BM_Chaparral, nodeColor=<rgba 0.5 0 1 1>}
  }
  MaterialExpressionNamedRerouteDeclaration_37 {
   class=MaterialExpressionNamedRerouteDeclaration
   inputs [None {from=MaterialExpressionStaticSwitch_8}]
   properties {name=ForestFloorLayerMask, nodeColor=<rgba 0.5 0 1 1>}
  }
  MaterialExpressionNamedRerouteDeclaration_38 {
   class=MaterialExpressionNamedRerouteDeclaration
   inputs [None {from=MaterialExpressionMultiply_42}]
   properties {name=Asphalt_physmat_Weightmap, nodeColor=<rgba 0.5 0 1 1>}
  }
  MaterialExpressionNamedRerouteDeclaration_39 {
   class=MaterialExpressionNamedRerouteDeclaration
   inputs [None {from=MaterialExpressionMaterialFunctionCall_15, output=DesertGrass_SimpleNoise}]
   properties {name=DesertGrass_SimpleNoise, nodeColor=<rgba 0.5 0 1 1>}
  }
  MaterialExpressionNamedRerouteDeclaration_4 {
   class=MaterialExpressionNamedRerouteDeclaration
   inputs [None {from=MaterialExpressionStaticSwitchParameter_5}]
   properties {name=BiomeA_Weight, nodeColor=<rgba 0.5 0 1 1>}
  }
  MaterialExpressionNamedRerouteDeclaration_40 {
   class=MaterialExpressionNamedRerouteDeclaration
   inputs [None {from=MaterialExpressionStaticSwitch_1}]
   properties {name=LGT_Grass01, nodeColor=<rgba 0.5 0 1 1>}
  }
  MaterialExpressionNamedRerouteDeclaration_41 {
   class=MaterialExpressionNamedRerouteDeclaration
   inputs [None {from=MaterialExpressionReroute_13, output=Output}]
   properties {name=Dirt_Physmat_WeightMap, nodeColor=<rgba 0.5 0 1 1>}
  }
  MaterialExpressionNamedRerouteDeclaration_42 {
   class=MaterialExpressionNamedRerouteDeclaration
   inputs [None {from=MaterialExpressionMaterialFunctionCall_23, output=15_RemoveLandscapeGrassTypes}]
   properties {name=BM_RemoveLandscapeGrassTypes, nodeColor=<rgba 0.5 0 1 1>}
  }
  MaterialExpressionNamedRerouteDeclaration_43 {
   class=MaterialExpressionNamedRerouteDeclaration
   inputs [None {from=MaterialExpressionMaterialFunctionCall_42, output=Result}]
   properties {name=LGT_DesertSand, nodeColor=<rgba 0.5 0 1 1>}
  }
  MaterialExpressionNamedRerouteDeclaration_44 {
   class=MaterialExpressionNamedRerouteDeclaration
   inputs [None {from=MaterialExpressionMaterialFunctionCall_20, output=Result}]
   properties {name=LGT_FlowerGroup01, nodeColor=<rgba 0.5 0 1 1>}
  }
  MaterialExpressionNamedRerouteDeclaration_45 {
   class=MaterialExpressionNamedRerouteDeclaration
   inputs [None {from=MaterialExpressionMaterialFunctionCall_19, output=MobileMaterial}]
   properties {name=Layer0_ChapGrassMobileMaterial, nodeColor=<rgba 0.5 0 1 1>}
  }
  MaterialExpressionNamedRerouteDeclaration_46 {
   class=MaterialExpressionNamedRerouteDeclaration
   inputs [None {from=MaterialExpressionMaterialFunctionCall_23, output=01_RedWoods}]
   properties {name=BM_RedWoods, nodeColor=<rgba 0.5 0 1 1>}
  }
  MaterialExpressionNamedRerouteDeclaration_47 {
   class=MaterialExpressionNamedRerouteDeclaration
   inputs [None {from=MaterialExpressionReroute_12, output=Output}]
   properties {name=Stone_Physmat_weightmap, nodeColor=<rgba 0.5 0 1 1>}
  }
  MaterialExpressionNamedRerouteDeclaration_48 {
   class=MaterialExpressionNamedRerouteDeclaration
   inputs [None {from=MaterialExpressionMaterialFunctionCall_14, output=Result}]
   properties {name=LGT_Archipelago, nodeColor=<rgba 0.5 0 1 1>}
  }
  MaterialExpressionNamedRerouteDeclaration_49 {
   class=MaterialExpressionNamedRerouteDeclaration
   inputs [None {from=MaterialExpressionMaterialFunctionCall_45, output=Result}]
   properties {name=LGT_ArchipelagoSeaweed, nodeColor=<rgba 0.5 0 1 1>}
  }
  MaterialExpressionNamedRerouteDeclaration_5 {
   class=MaterialExpressionNamedRerouteDeclaration
   inputs [None {from=MaterialExpressionMaterialFunctionCall_33, output=MobileLayerMaterial}]
   properties {name=MobileBorealForestMaterial-Unused, nodeColor=<rgba 0.4766 1 0 1>}
  }
  MaterialExpressionNamedRerouteDeclaration_50 {
   class=MaterialExpressionNamedRerouteDeclaration
   inputs [None {from=MaterialExpressionMaterialFunctionCall_50, output=Result}]
   properties {name=LGT_Gravel, nodeColor=<rgba 0.5 0 1 1>}
  }
  MaterialExpressionNamedRerouteDeclaration_51 {
   class=MaterialExpressionNamedRerouteDeclaration
   inputs [None {from=MaterialExpressionMaterialFunctionCall_5, output=CliffMaterialOnly}]
   properties {name=CliffMaterialOnly, nodeColor=<rgba 0.6172 0 1 1>}
  }
  MaterialExpressionNamedRerouteDeclaration_52 {
   class=MaterialExpressionNamedRerouteDeclaration
   inputs [None {from=MaterialExpressionReroute_11, output=Output}]
   properties {name=Grass_Physmat_Weightmap, nodeColor=<rgba 0.5 0 1 1>}
  }
  MaterialExpressionNamedRerouteDeclaration_53 {
   class=MaterialExpressionNamedRerouteDeclaration
   inputs [None {from=MaterialExpressionReroute_10, output=Output}]
   properties {name=Sand_Physmat_Weightmap, nodeColor=<rgba 0.5 0 1 1>}
  }
  MaterialExpressionNamedRerouteDeclaration_54 {
   class=MaterialExpressionNamedRerouteDeclaration
   inputs [None {from=MaterialExpressionReroute_9, output=Output}]
   properties {name=Snow_physmat_Weightmap, nodeColor=<rgba 0.5 0 1 1>}
  }
  MaterialExpressionNamedRerouteDeclaration_55 {
   class=MaterialExpressionNamedRerouteDeclaration
   inputs [None {from=MaterialExpressionStaticSwitch_9}]
   properties {name=L0CB_GrassMask, nodeColor=<rgba 0.5 0 1 1>}
  }
  MaterialExpressionNamedRerouteDeclaration_56 {
   class=MaterialExpressionNamedRerouteDeclaration
   inputs [None {from=MaterialExpressionStaticSwitch_21}]
   properties {name=LGT_Grass01_Lobby, desc=Lobby, nodeColor=<rgba 0.5 0 1 1>}
  }
  MaterialExpressionNamedRerouteDeclaration_57 {
   class=MaterialExpressionNamedRerouteDeclaration
   inputs [None {from=MaterialExpressionMax_51}]
   properties {name=Grass_EdgeOnly, nodeColor=<rgba 0.5 0 1 1>}
  }
  MaterialExpressionNamedRerouteDeclaration_58 {
   class=MaterialExpressionNamedRerouteDeclaration
   inputs [None {from=MaterialExpressionStaticSwitch_10}]
   properties {name=L0CC_Material, nodeColor=<rgba 0.5 0 1 1>}
  }
  MaterialExpressionNamedRerouteDeclaration_59 {
   class=MaterialExpressionNamedRerouteDeclaration
   inputs [None {from=MaterialExpressionMaterialFunctionCall_55, output=Result}]
   properties {name=LGT_Snow, nodeColor=<rgba 0.5 0 1 1>}
  }
  MaterialExpressionNamedRerouteDeclaration_6 {
   class=MaterialExpressionNamedRerouteDeclaration
   inputs [None {from=MaterialExpressionStaticSwitchParameter_6}]
   properties {name=BiomeB_Weight, nodeColor=<rgba 0.5 0 1 1>}
  }
  MaterialExpressionNamedRerouteDeclaration_60 {
   class=MaterialExpressionNamedRerouteDeclaration
   inputs [None {from=MaterialExpressionMaterialFunctionCall_18, output=Normals}]
   properties {name=Desert_MacroDetail_Normals, nodeColor=<rgba 0.5 0 1 1>}
  }
  MaterialExpressionNamedRerouteDeclaration_61 {
   class=MaterialExpressionNamedRerouteDeclaration
   inputs [None {from=MaterialExpressionMaterialFunctionCall_49, output=Result}]
   properties {name=LGT_Scree, nodeColor=<rgba 0.5 0 1 1>}
  }
  MaterialExpressionNamedRerouteDeclaration_62 {
   class=MaterialExpressionNamedRerouteDeclaration
   inputs [None {from=MaterialExpressionMaterialFunctionCall_25, output=Result}]
   properties {name=LGT_Goldlands01, nodeColor=<rgba 0.5 0 1 1>}
  }
  MaterialExpressionNamedRerouteDeclaration_63 {
   class=MaterialExpressionNamedRerouteDeclaration
   inputs [None {from=MaterialExpressionStaticSwitch_41}]
   properties {name=Goldlands_GrassMask, nodeColor=<rgba 0.5 0 1 1>}
  }
  MaterialExpressionNamedRerouteDeclaration_64 {
   class=MaterialExpressionNamedRerouteDeclaration
   inputs [None {from=MaterialExpressionStaticSwitch_12}]
   properties {name=L0CC_MobileMaterial, nodeColor=<rgba 0.5 0 1 1>}
  }
  MaterialExpressionNamedRerouteDeclaration_65 {
   class=MaterialExpressionNamedRerouteDeclaration
   inputs [None {from=MaterialExpressionMaterialFunctionCall_2, output=BlendedMaterial}]
   properties {name=Layer0_DefaultGrass, nodeColor=<rgba 0.5 0 1 1>}
  }
  MaterialExpressionNamedRerouteDeclaration_66 {
   class=MaterialExpressionNamedRerouteDeclaration
   inputs [None {from=MaterialExpressionMaterialFunctionCall_2, output=MobileBlendedMaterial}]
   properties {name=Layer0_DefaultGrassMobile, nodeColor=<rgba 0.5 0 1 1>}
  }
  MaterialExpressionNamedRerouteDeclaration_67 {
   class=MaterialExpressionNamedRerouteDeclaration
   inputs [None {from=MaterialExpressionStaticSwitch_11}]
   properties {name=L0CA_ChaparralGrass, nodeColor=<rgba 0.5 0 1 1>}
  }
  MaterialExpressionNamedRerouteDeclaration_68 {
   class=MaterialExpressionNamedRerouteDeclaration
   inputs [None {from=MaterialExpressionStaticSwitch_13}]
   properties {name=L0CA_ChaparralGrassMobile, nodeColor=<rgba 0.5 0 1 1>}
  }
  MaterialExpressionNamedRerouteDeclaration_69 {
   class=MaterialExpressionNamedRerouteDeclaration
   inputs [None {from=MaterialExpressionMaterialFunctionCall_5, output=BlendedMaterial}]
   properties {name=AllLayersWithCliff, nodeColor=<rgba 0.6172 0 1 1>}
  }
  MaterialExpressionNamedRerouteDeclaration_7 {
   class=MaterialExpressionNamedRerouteDeclaration
   inputs [None {from=MaterialExpressionStaticSwitchParameter_7}]
   properties {name=BiomeC_Weight, nodeColor=<rgba 0.5 0 1 1>}
  }
  MaterialExpressionNamedRerouteDeclaration_70 {
   class=MaterialExpressionNamedRerouteDeclaration
   inputs [None {from=MaterialExpressionMaterialFunctionCall_5, output=MobileBlendedMaterial}]
   properties {name=AllLayersWithCliff_Mobile, nodeColor=<rgba 0.6172 0 1 1>}
  }
  MaterialExpressionNamedRerouteDeclaration_71 {
   class=MaterialExpressionNamedRerouteDeclaration
   inputs [None {from=MaterialExpressionStaticSwitch_2}]
   properties {name=SnowMossDesat, nodeColor=<rgba 0.6172 0 1 1>}
  }
  MaterialExpressionNamedRerouteDeclaration_72 {
   class=MaterialExpressionNamedRerouteDeclaration
   inputs [None {from=MaterialExpressionStaticBoolParameter_4}]
   properties {name=UseBiomeA_ForestFloor, nodeColor=<rgba 0.5 0 1 1>}
  }
  MaterialExpressionNamedRerouteDeclaration_73 {
   class=MaterialExpressionNamedRerouteDeclaration
   inputs [None {from=MaterialExpressionStaticSwitch_34}]
   properties {name=MaterialUpTo_Asphalt01, nodeColor=<rgba 0.5 0 1 1>}
  }
  MaterialExpressionNamedRerouteDeclaration_74 {
   class=MaterialExpressionNamedRerouteDeclaration
   inputs [None {from=MaterialExpressionStaticSwitch_35}]
   properties {name=MobileMaterialUpTo_Asphalt01, nodeColor=<rgba 0.5 0 1 1>}
  }
  MaterialExpressionNamedRerouteDeclaration_75 {
   class=MaterialExpressionNamedRerouteDeclaration
   inputs [None {from=MaterialExpressionStaticSwitch_3}]
   properties {name=MaterialUpTo_BiomeA01, nodeColor=<rgba 0.5 0 1 1>}
  }
  MaterialExpressionNamedRerouteDeclaration_76 {
   class=MaterialExpressionNamedRerouteDeclaration
   inputs [None {from=MaterialExpressionStaticSwitch_6}]
   properties {name=MobileMaterialUpTo_BiomeA01, nodeColor=<rgba 0.5 0 1 1>}
  }
  MaterialExpressionNamedRerouteDeclaration_77 {
   class=MaterialExpressionNamedRerouteDeclaration
   inputs [None {from=MaterialExpressionMaterialFunctionCall_24, output=MetallicMask}]
   properties {name=Goldlands MetallicMask, nodeColor=<rgba 0.5 0 1 1>}
  }
  MaterialExpressionNamedRerouteDeclaration_78 {
   class=MaterialExpressionNamedRerouteDeclaration
   inputs [None {from=MaterialExpressionStaticBoolParameter_5}]
   properties {name=UseBiomeA02, nodeColor=<rgba 0.5 0 1 1>}
  }
  MaterialExpressionNamedRerouteDeclaration_79 {
   class=MaterialExpressionNamedRerouteDeclaration
   inputs [None {from=MaterialExpressionStaticSwitch_22}]
   properties {name=MaterialUpTo_BiomeA02, nodeColor=<rgba 0.5 0 1 1>}
  }
  MaterialExpressionNamedRerouteDeclaration_8 {
   class=MaterialExpressionNamedRerouteDeclaration
   inputs [None {from=MaterialExpressionStaticSwitchParameter_8}]
   properties {name=BiomeD_Weight, nodeColor=<rgba 0.5 0 1 1>}
  }
  MaterialExpressionNamedRerouteDeclaration_80 {
   class=MaterialExpressionNamedRerouteDeclaration
   inputs [None {from=MaterialExpressionStaticSwitch_23}]
   properties {name=MobileMaterialUpTo_BiomeA02, nodeColor=<rgba 0.5 0 1 1>}
  }
  MaterialExpressionNamedRerouteDeclaration_81 {
   class=MaterialExpressionNamedRerouteDeclaration
   inputs [None {from=MaterialExpressionStaticSwitch_40}]
   properties {name=Goldlands LayerMask, nodeColor=<rgba 0.5 0 1 1>}
  }
  MaterialExpressionNamedRerouteDeclaration_82 {
   class=MaterialExpressionNamedRerouteDeclaration
   inputs [None {from=MaterialExpressionStaticSwitch_55}]
   properties {name=DesertGrass_YuccaMask, nodeColor=<rgba 0.5 0 1 1>}
  }
  MaterialExpressionNamedRerouteDeclaration_83 {
   class=MaterialExpressionNamedRerouteDeclaration
   inputs [None {from=MaterialExpressionStaticSwitch_0}]
   properties {name=MaterialUpTo_Snow, nodeColor=<rgba 0.5 0 1 1>}
  }
  MaterialExpressionNamedRerouteDeclaration_84 {
   class=MaterialExpressionNamedRerouteDeclaration
   inputs [None {from=MaterialExpressionStaticSwitch_14}]
   properties {name=MaterialUpTo_SnowMobile, nodeColor=<rgba 0.5 0 1 1>}
  }
  MaterialExpressionNamedRerouteDeclaration_85 {
   class=MaterialExpressionNamedRerouteDeclaration
   inputs [None {from=MaterialExpressionMaterialFunctionCall_4, output=BlendedMaterial}]
   properties {name=MaterialUpTo_Sand, nodeColor=<rgba 0.5 0 1 1>}
  }
  MaterialExpressionNamedRerouteDeclaration_86 {
   class=MaterialExpressionNamedRerouteDeclaration
   inputs [None {from=MaterialExpressionStaticBoolParameter_7}]
   properties {name=UseBiomeB_Scree, nodeColor=<rgba 0.5 0 1 1>}
  }
  MaterialExpressionNamedRerouteDeclaration_87 {
   class=MaterialExpressionNamedRerouteDeclaration
   inputs [None {from=MaterialExpressionStaticSwitch_28}]
   properties {name=MaterialUpTo_BiomeB01, nodeColor=<rgba 0.5 0 1 1>}
  }
  MaterialExpressionNamedRerouteDeclaration_88 {
   class=MaterialExpressionNamedRerouteDeclaration
   inputs [None {from=MaterialExpressionStaticSwitch_29}]
   properties {name=MobileMaterialUpTo_BiomeB01, nodeColor=<rgba 0.5 0 1 1>}
  }
  MaterialExpressionNamedRerouteDeclaration_89 {
   class=MaterialExpressionNamedRerouteDeclaration
   inputs [None {from=MaterialExpressionStaticSwitch_24}]
   properties {name=BiomeA02LayerMask, nodeColor=<rgba 0.5 0 1 1>}
  }
  MaterialExpressionNamedRerouteDeclaration_9 {
   class=MaterialExpressionNamedRerouteDeclaration
   inputs [None {from=MaterialExpressionOneMinus_0}]
   properties {name=Grass_Weight, nodeColor=<rgba 0.5 0 1 1>}
  }
  MaterialExpressionNamedRerouteDeclaration_90 {
   class=MaterialExpressionNamedRerouteDeclaration
   inputs [None {from=MaterialExpressionStaticBoolParameter_14}]
   properties {name=UseBiomeC_Goldlands, nodeColor=<rgba 0.5 0 1 1>}
  }
  MaterialExpressionNamedRerouteDeclaration_91 {
   class=MaterialExpressionNamedRerouteDeclaration
   inputs [None {from=MaterialExpressionMaterialFunctionCall_4, output=MobileMaterial}]
   properties {name=Mobile Sand, nodeColor=<rgba 0 1 0.6953 1>}
  }
  MaterialExpressionNamedRerouteDeclaration_92 {
   class=MaterialExpressionNamedRerouteDeclaration
   inputs [None {from=MaterialExpressionMaterialFunctionCall_17, output=Result}]
   properties {name=LGT_DesertYucca, nodeColor=<rgba 0.5 0 1 1>}
  }
  MaterialExpressionNamedRerouteDeclaration_93 {
   class=MaterialExpressionNamedRerouteDeclaration
   inputs [None {from=MaterialExpressionMaterialFunctionCall_4, output=MobileBlendedMaterial}]
   properties {name=MobileMaterialUpTo_Sand, nodeColor=<rgba 0.5 0 1 1>}
  }
  MaterialExpressionNamedRerouteDeclaration_94 {
   class=MaterialExpressionNamedRerouteDeclaration
   inputs [None {from=MaterialExpressionStaticSwitch_27}]
   properties {name=Asphalt02LayerMask_MossyForest, nodeColor=<rgba 0.5 0 1 1>}
  }
  MaterialExpressionNamedRerouteDeclaration_95 {
   class=MaterialExpressionNamedRerouteDeclaration
   inputs [None {from=MaterialExpressionStaticBoolParameter_6}]
   properties {name=UseAsphalt02, nodeColor=<rgba 0.5 0 1 1>}
  }
  MaterialExpressionNamedRerouteDeclaration_96 {
   class=MaterialExpressionNamedRerouteDeclaration
   inputs [None {from=MaterialExpressionStaticSwitch_25}]
   properties {name=MaterialUpTo_Asphalt02, nodeColor=<rgba 0.5 0 1 1>}
  }
  MaterialExpressionNamedRerouteDeclaration_97 {
   class=MaterialExpressionNamedRerouteDeclaration
   inputs [None {from=MaterialExpressionStaticSwitch_26}]
   properties {name=MobileMaterialUpTo_Asphalt02, nodeColor=<rgba 0.5 0 1 1>}
  }
  MaterialExpressionNamedRerouteDeclaration_98 {
   class=MaterialExpressionNamedRerouteDeclaration
   inputs [None {from=MaterialExpressionStaticBoolParameter_9}]
   properties {name=UseAsphalt, nodeColor=<rgba 0.5 0 1 1>}
  }
  MaterialExpressionNamedRerouteDeclaration_99 {
   class=MaterialExpressionNamedRerouteDeclaration
   inputs [None {from=MaterialExpressionStaticSwitch_46}]
   properties {name=MaterialUpTo_Goldlands, nodeColor=<rgba 0.5 0 1 1>}
  }
  MaterialExpressionNamedRerouteUsage_0 {class=MaterialExpressionNamedRerouteUsage}
  MaterialExpressionNamedRerouteUsage_1 {class=MaterialExpressionNamedRerouteUsage}
  MaterialExpressionNamedRerouteUsage_10 {class=MaterialExpressionNamedRerouteUsage}
  MaterialExpressionNamedRerouteUsage_100 {class=MaterialExpressionNamedRerouteUsage}
  MaterialExpressionNamedRerouteUsage_101 {class=MaterialExpressionNamedRerouteUsage}
  MaterialExpressionNamedRerouteUsage_102 {class=MaterialExpressionNamedRerouteUsage}
  MaterialExpressionNamedRerouteUsage_103 {class=MaterialExpressionNamedRerouteUsage}
  MaterialExpressionNamedRerouteUsage_104 {class=MaterialExpressionNamedRerouteUsage}
  MaterialExpressionNamedRerouteUsage_105 {class=MaterialExpressionNamedRerouteUsage}
  MaterialExpressionNamedRerouteUsage_106 {class=MaterialExpressionNamedRerouteUsage}
  MaterialExpressionNamedRerouteUsage_107 {class=MaterialExpressionNamedRerouteUsage}
  MaterialExpressionNamedRerouteUsage_108 {class=MaterialExpressionNamedRerouteUsage}
  MaterialExpressionNamedRerouteUsage_109 {class=MaterialExpressionNamedRerouteUsage}
  MaterialExpressionNamedRerouteUsage_11 {class=MaterialExpressionNamedRerouteUsage}
  MaterialExpressionNamedRerouteUsage_110 {class=MaterialExpressionNamedRerouteUsage}
  MaterialExpressionNamedRerouteUsage_111 {class=MaterialExpressionNamedRerouteUsage}
  MaterialExpressionNamedRerouteUsage_112 {class=MaterialExpressionNamedRerouteUsage}
  MaterialExpressionNamedRerouteUsage_113 {class=MaterialExpressionNamedRerouteUsage}
  MaterialExpressionNamedRerouteUsage_114 {class=MaterialExpressionNamedRerouteUsage}
  MaterialExpressionNamedRerouteUsage_115 {class=MaterialExpressionNamedRerouteUsage}
  MaterialExpressionNamedRerouteUsage_116 {class=MaterialExpressionNamedRerouteUsage}
  MaterialExpressionNamedRerouteUsage_117 {class=MaterialExpressionNamedRerouteUsage}
  MaterialExpressionNamedRerouteUsage_118 {class=MaterialExpressionNamedRerouteUsage}
  MaterialExpressionNamedRerouteUsage_119 {class=MaterialExpressionNamedRerouteUsage}
  MaterialExpressionNamedRerouteUsage_12 {class=MaterialExpressionNamedRerouteUsage}
  MaterialExpressionNamedRerouteUsage_120 {class=MaterialExpressionNamedRerouteUsage}
  MaterialExpressionNamedRerouteUsage_121 {class=MaterialExpressionNamedRerouteUsage}
  MaterialExpressionNamedRerouteUsage_122 {class=MaterialExpressionNamedRerouteUsage}
  MaterialExpressionNamedRerouteUsage_123 {class=MaterialExpressionNamedRerouteUsage}
  MaterialExpressionNamedRerouteUsage_124 {class=MaterialExpressionNamedRerouteUsage}
  MaterialExpressionNamedRerouteUsage_125 {class=MaterialExpressionNamedRerouteUsage}
  MaterialExpressionNamedRerouteUsage_126 {class=MaterialExpressionNamedRerouteUsage}
  MaterialExpressionNamedRerouteUsage_127 {class=MaterialExpressionNamedRerouteUsage}
  MaterialExpressionNamedRerouteUsage_128 {class=MaterialExpressionNamedRerouteUsage}
  MaterialExpressionNamedRerouteUsage_129 {class=MaterialExpressionNamedRerouteUsage}
  MaterialExpressionNamedRerouteUsage_13 {class=MaterialExpressionNamedRerouteUsage}
  MaterialExpressionNamedRerouteUsage_130 {class=MaterialExpressionNamedRerouteUsage}
  MaterialExpressionNamedRerouteUsage_131 {class=MaterialExpressionNamedRerouteUsage}
  MaterialExpressionNamedRerouteUsage_132 {class=MaterialExpressionNamedRerouteUsage}
  MaterialExpressionNamedRerouteUsage_133 {class=MaterialExpressionNamedRerouteUsage}
  MaterialExpressionNamedRerouteUsage_134 {class=MaterialExpressionNamedRerouteUsage}
  MaterialExpressionNamedRerouteUsage_135 {class=MaterialExpressionNamedRerouteUsage}
  MaterialExpressionNamedRerouteUsage_136 {class=MaterialExpressionNamedRerouteUsage}
  MaterialExpressionNamedRerouteUsage_137 {class=MaterialExpressionNamedRerouteUsage}
  MaterialExpressionNamedRerouteUsage_138 {class=MaterialExpressionNamedRerouteUsage}
  MaterialExpressionNamedRerouteUsage_139 {class=MaterialExpressionNamedRerouteUsage}
  MaterialExpressionNamedRerouteUsage_14 {class=MaterialExpressionNamedRerouteUsage}
  MaterialExpressionNamedRerouteUsage_140 {class=MaterialExpressionNamedRerouteUsage}
  MaterialExpressionNamedRerouteUsage_141 {class=MaterialExpressionNamedRerouteUsage}
  MaterialExpressionNamedRerouteUsage_142 {class=MaterialExpressionNamedRerouteUsage}
  MaterialExpressionNamedRerouteUsage_143 {class=MaterialExpressionNamedRerouteUsage}
  MaterialExpressionNamedRerouteUsage_145 {class=MaterialExpressionNamedRerouteUsage}
  MaterialExpressionNamedRerouteUsage_146 {class=MaterialExpressionNamedRerouteUsage}
  MaterialExpressionNamedRerouteUsage_147 {class=MaterialExpressionNamedRerouteUsage}
  MaterialExpressionNamedRerouteUsage_148 {class=MaterialExpressionNamedRerouteUsage}
  MaterialExpressionNamedRerouteUsage_149 {class=MaterialExpressionNamedRerouteUsage}
  MaterialExpressionNamedRerouteUsage_15 {class=MaterialExpressionNamedRerouteUsage}
  MaterialExpressionNamedRerouteUsage_150 {class=MaterialExpressionNamedRerouteUsage}
  MaterialExpressionNamedRerouteUsage_151 {class=MaterialExpressionNamedRerouteUsage}
  MaterialExpressionNamedRerouteUsage_152 {class=MaterialExpressionNamedRerouteUsage}
  MaterialExpressionNamedRerouteUsage_153 {class=MaterialExpressionNamedRerouteUsage}
  MaterialExpressionNamedRerouteUsage_154 {class=MaterialExpressionNamedRerouteUsage}
  MaterialExpressionNamedRerouteUsage_155 {class=MaterialExpressionNamedRerouteUsage}
  MaterialExpressionNamedRerouteUsage_156 {class=MaterialExpressionNamedRerouteUsage}
  MaterialExpressionNamedRerouteUsage_157 {class=MaterialExpressionNamedRerouteUsage}
  MaterialExpressionNamedRerouteUsage_158 {class=MaterialExpressionNamedRerouteUsage}
  MaterialExpressionNamedRerouteUsage_159 {class=MaterialExpressionNamedRerouteUsage}
  MaterialExpressionNamedRerouteUsage_16 {class=MaterialExpressionNamedRerouteUsage}
  MaterialExpressionNamedRerouteUsage_160 {class=MaterialExpressionNamedRerouteUsage}
  MaterialExpressionNamedRerouteUsage_161 {class=MaterialExpressionNamedRerouteUsage}
  MaterialExpressionNamedRerouteUsage_162 {class=MaterialExpressionNamedRerouteUsage}
  MaterialExpressionNamedRerouteUsage_164 {class=MaterialExpressionNamedRerouteUsage}
  MaterialExpressionNamedRerouteUsage_165 {class=MaterialExpressionNamedRerouteUsage}
  MaterialExpressionNamedRerouteUsage_166 {class=MaterialExpressionNamedRerouteUsage}
  MaterialExpressionNamedRerouteUsage_167 {class=MaterialExpressionNamedRerouteUsage}
  MaterialExpressionNamedRerouteUsage_168 {class=MaterialExpressionNamedRerouteUsage}
  MaterialExpressionNamedRerouteUsage_169 {class=MaterialExpressionNamedRerouteUsage}
  MaterialExpressionNamedRerouteUsage_17 {class=MaterialExpressionNamedRerouteUsage}
  MaterialExpressionNamedRerouteUsage_170 {class=MaterialExpressionNamedRerouteUsage}
  MaterialExpressionNamedRerouteUsage_171 {class=MaterialExpressionNamedRerouteUsage}
  MaterialExpressionNamedRerouteUsage_172 {class=MaterialExpressionNamedRerouteUsage}
  MaterialExpressionNamedRerouteUsage_173 {class=MaterialExpressionNamedRerouteUsage}
  MaterialExpressionNamedRerouteUsage_174 {class=MaterialExpressionNamedRerouteUsage}
  MaterialExpressionNamedRerouteUsage_175 {class=MaterialExpressionNamedRerouteUsage}
  MaterialExpressionNamedRerouteUsage_176 {class=MaterialExpressionNamedRerouteUsage}
  MaterialExpressionNamedRerouteUsage_177 {class=MaterialExpressionNamedRerouteUsage}
  MaterialExpressionNamedRerouteUsage_178 {class=MaterialExpressionNamedRerouteUsage}
  MaterialExpressionNamedRerouteUsage_179 {class=MaterialExpressionNamedRerouteUsage}
  MaterialExpressionNamedRerouteUsage_18 {class=MaterialExpressionNamedRerouteUsage}
  MaterialExpressionNamedRerouteUsage_180 {class=MaterialExpressionNamedRerouteUsage}
  MaterialExpressionNamedRerouteUsage_181 {class=MaterialExpressionNamedRerouteUsage}
  MaterialExpressionNamedRerouteUsage_182 {class=MaterialExpressionNamedRerouteUsage}
  MaterialExpressionNamedRerouteUsage_183 {class=MaterialExpressionNamedRerouteUsage}
  MaterialExpressionNamedRerouteUsage_184 {class=MaterialExpressionNamedRerouteUsage}
  MaterialExpressionNamedRerouteUsage_185 {class=MaterialExpressionNamedRerouteUsage}
  MaterialExpressionNamedRerouteUsage_186 {class=MaterialExpressionNamedRerouteUsage}
  MaterialExpressionNamedRerouteUsage_187 {class=MaterialExpressionNamedRerouteUsage}
  MaterialExpressionNamedRerouteUsage_188 {class=MaterialExpressionNamedRerouteUsage}
  MaterialExpressionNamedRerouteUsage_189 {class=MaterialExpressionNamedRerouteUsage}
  MaterialExpressionNamedRerouteUsage_19 {class=MaterialExpressionNamedRerouteUsage}
  MaterialExpressionNamedRerouteUsage_190 {class=MaterialExpressionNamedRerouteUsage}
  MaterialExpressionNamedRerouteUsage_191 {class=MaterialExpressionNamedRerouteUsage}
  MaterialExpressionNamedRerouteUsage_192 {class=MaterialExpressionNamedRerouteUsage}
  MaterialExpressionNamedRerouteUsage_193 {class=MaterialExpressionNamedRerouteUsage}
  MaterialExpressionNamedRerouteUsage_194 {class=MaterialExpressionNamedRerouteUsage}
  MaterialExpressionNamedRerouteUsage_195 {class=MaterialExpressionNamedRerouteUsage}
  MaterialExpressionNamedRerouteUsage_196 {class=MaterialExpressionNamedRerouteUsage}
  MaterialExpressionNamedRerouteUsage_197 {class=MaterialExpressionNamedRerouteUsage}
  MaterialExpressionNamedRerouteUsage_198 {class=MaterialExpressionNamedRerouteUsage}
  MaterialExpressionNamedRerouteUsage_199 {class=MaterialExpressionNamedRerouteUsage}
  MaterialExpressionNamedRerouteUsage_2 {class=MaterialExpressionNamedRerouteUsage}
  MaterialExpressionNamedRerouteUsage_20 {class=MaterialExpressionNamedRerouteUsage}
  MaterialExpressionNamedRerouteUsage_200 {class=MaterialExpressionNamedRerouteUsage}
  MaterialExpressionNamedRerouteUsage_201 {class=MaterialExpressionNamedRerouteUsage}
  MaterialExpressionNamedRerouteUsage_202 {class=MaterialExpressionNamedRerouteUsage}
  MaterialExpressionNamedRerouteUsage_203 {class=MaterialExpressionNamedRerouteUsage}
  MaterialExpressionNamedRerouteUsage_204 {class=MaterialExpressionNamedRerouteUsage}
  MaterialExpressionNamedRerouteUsage_205 {class=MaterialExpressionNamedRerouteUsage}
  MaterialExpressionNamedRerouteUsage_206 {class=MaterialExpressionNamedRerouteUsage}
  MaterialExpressionNamedRerouteUsage_207 {class=MaterialExpressionNamedRerouteUsage}
  MaterialExpressionNamedRerouteUsage_208 {class=MaterialExpressionNamedRerouteUsage}
  MaterialExpressionNamedRerouteUsage_209 {class=MaterialExpressionNamedRerouteUsage}
  MaterialExpressionNamedRerouteUsage_21 {class=MaterialExpressionNamedRerouteUsage}
  MaterialExpressionNamedRerouteUsage_210 {class=MaterialExpressionNamedRerouteUsage}
  MaterialExpressionNamedRerouteUsage_211 {class=MaterialExpressionNamedRerouteUsage}
  MaterialExpressionNamedRerouteUsage_212 {class=MaterialExpressionNamedRerouteUsage}
  MaterialExpressionNamedRerouteUsage_213 {class=MaterialExpressionNamedRerouteUsage}
  MaterialExpressionNamedRerouteUsage_214 {class=MaterialExpressionNamedRerouteUsage}
  MaterialExpressionNamedRerouteUsage_215 {class=MaterialExpressionNamedRerouteUsage}
  MaterialExpressionNamedRerouteUsage_216 {class=MaterialExpressionNamedRerouteUsage}
  MaterialExpressionNamedRerouteUsage_217 {class=MaterialExpressionNamedRerouteUsage}
  MaterialExpressionNamedRerouteUsage_218 {class=MaterialExpressionNamedRerouteUsage}
  MaterialExpressionNamedRerouteUsage_219 {class=MaterialExpressionNamedRerouteUsage}
  MaterialExpressionNamedRerouteUsage_22 {class=MaterialExpressionNamedRerouteUsage}
  MaterialExpressionNamedRerouteUsage_220 {class=MaterialExpressionNamedRerouteUsage}
  MaterialExpressionNamedRerouteUsage_221 {class=MaterialExpressionNamedRerouteUsage}
  MaterialExpressionNamedRerouteUsage_222 {class=MaterialExpressionNamedRerouteUsage}
  MaterialExpressionNamedRerouteUsage_223 {class=MaterialExpressionNamedRerouteUsage}
  MaterialExpressionNamedRerouteUsage_224 {class=MaterialExpressionNamedRerouteUsage}
  MaterialExpressionNamedRerouteUsage_225 {class=MaterialExpressionNamedRerouteUsage}
  MaterialExpressionNamedRerouteUsage_226 {class=MaterialExpressionNamedRerouteUsage}
  MaterialExpressionNamedRerouteUsage_227 {class=MaterialExpressionNamedRerouteUsage}
  MaterialExpressionNamedRerouteUsage_228 {class=MaterialExpressionNamedRerouteUsage}
  MaterialExpressionNamedRerouteUsage_229 {class=MaterialExpressionNamedRerouteUsage}
  MaterialExpressionNamedRerouteUsage_23 {class=MaterialExpressionNamedRerouteUsage}
  MaterialExpressionNamedRerouteUsage_230 {class=MaterialExpressionNamedRerouteUsage}
  MaterialExpressionNamedRerouteUsage_231 {class=MaterialExpressionNamedRerouteUsage}
  MaterialExpressionNamedRerouteUsage_232 {class=MaterialExpressionNamedRerouteUsage}
  MaterialExpressionNamedRerouteUsage_233 {class=MaterialExpressionNamedRerouteUsage}
  MaterialExpressionNamedRerouteUsage_234 {class=MaterialExpressionNamedRerouteUsage}
  MaterialExpressionNamedRerouteUsage_235 {class=MaterialExpressionNamedRerouteUsage}
  MaterialExpressionNamedRerouteUsage_236 {class=MaterialExpressionNamedRerouteUsage}
  MaterialExpressionNamedRerouteUsage_237 {class=MaterialExpressionNamedRerouteUsage}
  MaterialExpressionNamedRerouteUsage_238 {class=MaterialExpressionNamedRerouteUsage}
  MaterialExpressionNamedRerouteUsage_239 {class=MaterialExpressionNamedRerouteUsage}
  MaterialExpressionNamedRerouteUsage_24 {class=MaterialExpressionNamedRerouteUsage}
  MaterialExpressionNamedRerouteUsage_240 {class=MaterialExpressionNamedRerouteUsage}
  MaterialExpressionNamedRerouteUsage_241 {class=MaterialExpressionNamedRerouteUsage}
  MaterialExpressionNamedRerouteUsage_242 {class=MaterialExpressionNamedRerouteUsage}
  MaterialExpressionNamedRerouteUsage_243 {class=MaterialExpressionNamedRerouteUsage}
  MaterialExpressionNamedRerouteUsage_244 {class=MaterialExpressionNamedRerouteUsage}
  MaterialExpressionNamedRerouteUsage_245 {class=MaterialExpressionNamedRerouteUsage}
  MaterialExpressionNamedRerouteUsage_246 {class=MaterialExpressionNamedRerouteUsage}
  MaterialExpressionNamedRerouteUsage_247 {class=MaterialExpressionNamedRerouteUsage}
  MaterialExpressionNamedRerouteUsage_248 {class=MaterialExpressionNamedRerouteUsage}
  MaterialExpressionNamedRerouteUsage_249 {class=MaterialExpressionNamedRerouteUsage}
  MaterialExpressionNamedRerouteUsage_25 {class=MaterialExpressionNamedRerouteUsage}
  MaterialExpressionNamedRerouteUsage_250 {class=MaterialExpressionNamedRerouteUsage}
  MaterialExpressionNamedRerouteUsage_251 {class=MaterialExpressionNamedRerouteUsage}
  MaterialExpressionNamedRerouteUsage_252 {class=MaterialExpressionNamedRerouteUsage}
  MaterialExpressionNamedRerouteUsage_253 {class=MaterialExpressionNamedRerouteUsage}
  MaterialExpressionNamedRerouteUsage_254 {class=MaterialExpressionNamedRerouteUsage}
  MaterialExpressionNamedRerouteUsage_255 {class=MaterialExpressionNamedRerouteUsage}
  MaterialExpressionNamedRerouteUsage_256 {class=MaterialExpressionNamedRerouteUsage}
  MaterialExpressionNamedRerouteUsage_257 {class=MaterialExpressionNamedRerouteUsage}
  MaterialExpressionNamedRerouteUsage_258 {class=MaterialExpressionNamedRerouteUsage}
  MaterialExpressionNamedRerouteUsage_259 {class=MaterialExpressionNamedRerouteUsage}
  MaterialExpressionNamedRerouteUsage_26 {class=MaterialExpressionNamedRerouteUsage}
  MaterialExpressionNamedRerouteUsage_260 {class=MaterialExpressionNamedRerouteUsage}
  MaterialExpressionNamedRerouteUsage_261 {class=MaterialExpressionNamedRerouteUsage}
  MaterialExpressionNamedRerouteUsage_262 {class=MaterialExpressionNamedRerouteUsage}
  MaterialExpressionNamedRerouteUsage_263 {class=MaterialExpressionNamedRerouteUsage}
  MaterialExpressionNamedRerouteUsage_264 {class=MaterialExpressionNamedRerouteUsage}
  MaterialExpressionNamedRerouteUsage_265 {class=MaterialExpressionNamedRerouteUsage}
  MaterialExpressionNamedRerouteUsage_266 {class=MaterialExpressionNamedRerouteUsage}
  MaterialExpressionNamedRerouteUsage_267 {class=MaterialExpressionNamedRerouteUsage}
  MaterialExpressionNamedRerouteUsage_268 {class=MaterialExpressionNamedRerouteUsage}
  MaterialExpressionNamedRerouteUsage_269 {class=MaterialExpressionNamedRerouteUsage}
  MaterialExpressionNamedRerouteUsage_27 {class=MaterialExpressionNamedRerouteUsage}
  MaterialExpressionNamedRerouteUsage_270 {class=MaterialExpressionNamedRerouteUsage}
  MaterialExpressionNamedRerouteUsage_272 {class=MaterialExpressionNamedRerouteUsage}
  MaterialExpressionNamedRerouteUsage_273 {class=MaterialExpressionNamedRerouteUsage}
  MaterialExpressionNamedRerouteUsage_274 {class=MaterialExpressionNamedRerouteUsage}
  MaterialExpressionNamedRerouteUsage_275 {class=MaterialExpressionNamedRerouteUsage}
  MaterialExpressionNamedRerouteUsage_276 {class=MaterialExpressionNamedRerouteUsage}
  MaterialExpressionNamedRerouteUsage_277 {class=MaterialExpressionNamedRerouteUsage}
  MaterialExpressionNamedRerouteUsage_278 {class=MaterialExpressionNamedRerouteUsage}
  MaterialExpressionNamedRerouteUsage_279 {class=MaterialExpressionNamedRerouteUsage}
  MaterialExpressionNamedRerouteUsage_28 {class=MaterialExpressionNamedRerouteUsage}
  MaterialExpressionNamedRerouteUsage_280 {class=MaterialExpressionNamedRerouteUsage}
  MaterialExpressionNamedRerouteUsage_281 {class=MaterialExpressionNamedRerouteUsage}
  MaterialExpressionNamedRerouteUsage_282 {class=MaterialExpressionNamedRerouteUsage}
  MaterialExpressionNamedRerouteUsage_283 {class=MaterialExpressionNamedRerouteUsage}
  MaterialExpressionNamedRerouteUsage_284 {class=MaterialExpressionNamedRerouteUsage}
  MaterialExpressionNamedRerouteUsage_285 {class=MaterialExpressionNamedRerouteUsage}
  MaterialExpressionNamedRerouteUsage_286 {class=MaterialExpressionNamedRerouteUsage}
  MaterialExpressionNamedRerouteUsage_287 {class=MaterialExpressionNamedRerouteUsage}
  MaterialExpressionNamedRerouteUsage_288 {class=MaterialExpressionNamedRerouteUsage}
  MaterialExpressionNamedRerouteUsage_289 {class=MaterialExpressionNamedRerouteUsage}
  MaterialExpressionNamedRerouteUsage_29 {class=MaterialExpressionNamedRerouteUsage}
  MaterialExpressionNamedRerouteUsage_290 {class=MaterialExpressionNamedRerouteUsage}
  MaterialExpressionNamedRerouteUsage_291 {class=MaterialExpressionNamedRerouteUsage}
  MaterialExpressionNamedRerouteUsage_292 {class=MaterialExpressionNamedRerouteUsage}
  MaterialExpressionNamedRerouteUsage_293 {class=MaterialExpressionNamedRerouteUsage}
  MaterialExpressionNamedRerouteUsage_294 {class=MaterialExpressionNamedRerouteUsage}
  MaterialExpressionNamedRerouteUsage_295 {class=MaterialExpressionNamedRerouteUsage}
  MaterialExpressionNamedRerouteUsage_296 {class=MaterialExpressionNamedRerouteUsage}
  MaterialExpressionNamedRerouteUsage_297 {class=MaterialExpressionNamedRerouteUsage}
  MaterialExpressionNamedRerouteUsage_298 {class=MaterialExpressionNamedRerouteUsage}
  MaterialExpressionNamedRerouteUsage_299 {class=MaterialExpressionNamedRerouteUsage}
  MaterialExpressionNamedRerouteUsage_3 {class=MaterialExpressionNamedRerouteUsage}
  MaterialExpressionNamedRerouteUsage_30 {class=MaterialExpressionNamedRerouteUsage}
  MaterialExpressionNamedRerouteUsage_300 {class=MaterialExpressionNamedRerouteUsage}
  MaterialExpressionNamedRerouteUsage_301 {class=MaterialExpressionNamedRerouteUsage}
  MaterialExpressionNamedRerouteUsage_302 {class=MaterialExpressionNamedRerouteUsage}
  MaterialExpressionNamedRerouteUsage_303 {class=MaterialExpressionNamedRerouteUsage}
  MaterialExpressionNamedRerouteUsage_304 {class=MaterialExpressionNamedRerouteUsage}
  MaterialExpressionNamedRerouteUsage_305 {class=MaterialExpressionNamedRerouteUsage}
  MaterialExpressionNamedRerouteUsage_306 {class=MaterialExpressionNamedRerouteUsage}
  MaterialExpressionNamedRerouteUsage_307 {class=MaterialExpressionNamedRerouteUsage}
  MaterialExpressionNamedRerouteUsage_308 {class=MaterialExpressionNamedRerouteUsage}
  MaterialExpressionNamedRerouteUsage_309 {class=MaterialExpressionNamedRerouteUsage}
  MaterialExpressionNamedRerouteUsage_31 {class=MaterialExpressionNamedRerouteUsage}
  MaterialExpressionNamedRerouteUsage_310 {class=MaterialExpressionNamedRerouteUsage}
  MaterialExpressionNamedRerouteUsage_311 {class=MaterialExpressionNamedRerouteUsage}
  MaterialExpressionNamedRerouteUsage_312 {class=MaterialExpressionNamedRerouteUsage}
  MaterialExpressionNamedRerouteUsage_313 {class=MaterialExpressionNamedRerouteUsage}
  MaterialExpressionNamedRerouteUsage_314 {class=MaterialExpressionNamedRerouteUsage}
  MaterialExpressionNamedRerouteUsage_315 {class=MaterialExpressionNamedRerouteUsage}
  MaterialExpressionNamedRerouteUsage_316 {class=MaterialExpressionNamedRerouteUsage}
  MaterialExpressionNamedRerouteUsage_317 {class=MaterialExpressionNamedRerouteUsage}
  MaterialExpressionNamedRerouteUsage_318 {class=MaterialExpressionNamedRerouteUsage}
  MaterialExpressionNamedRerouteUsage_319 {class=MaterialExpressionNamedRerouteUsage}
  MaterialExpressionNamedRerouteUsage_32 {class=MaterialExpressionNamedRerouteUsage}
  MaterialExpressionNamedRerouteUsage_320 {class=MaterialExpressionNamedRerouteUsage}
  MaterialExpressionNamedRerouteUsage_321 {class=MaterialExpressionNamedRerouteUsage}
  MaterialExpressionNamedRerouteUsage_322 {class=MaterialExpressionNamedRerouteUsage}
  MaterialExpressionNamedRerouteUsage_323 {class=MaterialExpressionNamedRerouteUsage}
  MaterialExpressionNamedRerouteUsage_324 {class=MaterialExpressionNamedRerouteUsage}
  MaterialExpressionNamedRerouteUsage_325 {class=MaterialExpressionNamedRerouteUsage}
  MaterialExpressionNamedRerouteUsage_327 {class=MaterialExpressionNamedRerouteUsage}
  MaterialExpressionNamedRerouteUsage_328 {class=MaterialExpressionNamedRerouteUsage}
  MaterialExpressionNamedRerouteUsage_329 {class=MaterialExpressionNamedRerouteUsage}
  MaterialExpressionNamedRerouteUsage_33 {class=MaterialExpressionNamedRerouteUsage}
  MaterialExpressionNamedRerouteUsage_330 {class=MaterialExpressionNamedRerouteUsage}
  MaterialExpressionNamedRerouteUsage_331 {class=MaterialExpressionNamedRerouteUsage}
  MaterialExpressionNamedRerouteUsage_332 {class=MaterialExpressionNamedRerouteUsage}
  MaterialExpressionNamedRerouteUsage_333 {class=MaterialExpressionNamedRerouteUsage}
  MaterialExpressionNamedRerouteUsage_334 {class=MaterialExpressionNamedRerouteUsage}
  MaterialExpressionNamedRerouteUsage_335 {class=MaterialExpressionNamedRerouteUsage}
  MaterialExpressionNamedRerouteUsage_336 {class=MaterialExpressionNamedRerouteUsage}
  MaterialExpressionNamedRerouteUsage_337 {class=MaterialExpressionNamedRerouteUsage}
  MaterialExpressionNamedRerouteUsage_338 {class=MaterialExpressionNamedRerouteUsage}
  MaterialExpressionNamedRerouteUsage_339 {class=MaterialExpressionNamedRerouteUsage}
  MaterialExpressionNamedRerouteUsage_34 {class=MaterialExpressionNamedRerouteUsage}
  MaterialExpressionNamedRerouteUsage_340 {class=MaterialExpressionNamedRerouteUsage}
  MaterialExpressionNamedRerouteUsage_341 {class=MaterialExpressionNamedRerouteUsage}
  MaterialExpressionNamedRerouteUsage_342 {class=MaterialExpressionNamedRerouteUsage}
  MaterialExpressionNamedRerouteUsage_343 {class=MaterialExpressionNamedRerouteUsage}
  MaterialExpressionNamedRerouteUsage_344 {class=MaterialExpressionNamedRerouteUsage}
  MaterialExpressionNamedRerouteUsage_345 {class=MaterialExpressionNamedRerouteUsage}
  MaterialExpressionNamedRerouteUsage_346 {class=MaterialExpressionNamedRerouteUsage}
  MaterialExpressionNamedRerouteUsage_347 {class=MaterialExpressionNamedRerouteUsage}
  MaterialExpressionNamedRerouteUsage_348 {class=MaterialExpressionNamedRerouteUsage}
  MaterialExpressionNamedRerouteUsage_349 {class=MaterialExpressionNamedRerouteUsage}
  MaterialExpressionNamedRerouteUsage_35 {class=MaterialExpressionNamedRerouteUsage}
  MaterialExpressionNamedRerouteUsage_350 {class=MaterialExpressionNamedRerouteUsage}
  MaterialExpressionNamedRerouteUsage_352 {class=MaterialExpressionNamedRerouteUsage}
  MaterialExpressionNamedRerouteUsage_353 {class=MaterialExpressionNamedRerouteUsage}
  MaterialExpressionNamedRerouteUsage_354 {class=MaterialExpressionNamedRerouteUsage}
  MaterialExpressionNamedRerouteUsage_355 {class=MaterialExpressionNamedRerouteUsage}
  MaterialExpressionNamedRerouteUsage_356 {class=MaterialExpressionNamedRerouteUsage}
  MaterialExpressionNamedRerouteUsage_357 {class=MaterialExpressionNamedRerouteUsage}
  MaterialExpressionNamedRerouteUsage_358 {class=MaterialExpressionNamedRerouteUsage}
  MaterialExpressionNamedRerouteUsage_359 {class=MaterialExpressionNamedRerouteUsage}
  MaterialExpressionNamedRerouteUsage_36 {class=MaterialExpressionNamedRerouteUsage}
  MaterialExpressionNamedRerouteUsage_360 {class=MaterialExpressionNamedRerouteUsage}
  MaterialExpressionNamedRerouteUsage_361 {class=MaterialExpressionNamedRerouteUsage}
  MaterialExpressionNamedRerouteUsage_362 {class=MaterialExpressionNamedRerouteUsage}
  MaterialExpressionNamedRerouteUsage_364 {class=MaterialExpressionNamedRerouteUsage}
  MaterialExpressionNamedRerouteUsage_365 {class=MaterialExpressionNamedRerouteUsage}
  MaterialExpressionNamedRerouteUsage_366 {class=MaterialExpressionNamedRerouteUsage}
  MaterialExpressionNamedRerouteUsage_367 {class=MaterialExpressionNamedRerouteUsage}
  MaterialExpressionNamedRerouteUsage_368 {class=MaterialExpressionNamedRerouteUsage}
  MaterialExpressionNamedRerouteUsage_369 {class=MaterialExpressionNamedRerouteUsage}
  MaterialExpressionNamedRerouteUsage_37 {class=MaterialExpressionNamedRerouteUsage}
  MaterialExpressionNamedRerouteUsage_370 {class=MaterialExpressionNamedRerouteUsage}
  MaterialExpressionNamedRerouteUsage_371 {class=MaterialExpressionNamedRerouteUsage}
  MaterialExpressionNamedRerouteUsage_372 {class=MaterialExpressionNamedRerouteUsage}
  MaterialExpressionNamedRerouteUsage_373 {class=MaterialExpressionNamedRerouteUsage}
  MaterialExpressionNamedRerouteUsage_375 {class=MaterialExpressionNamedRerouteUsage}
  MaterialExpressionNamedRerouteUsage_376 {class=MaterialExpressionNamedRerouteUsage}
  MaterialExpressionNamedRerouteUsage_377 {class=MaterialExpressionNamedRerouteUsage}
  MaterialExpressionNamedRerouteUsage_378 {class=MaterialExpressionNamedRerouteUsage}
  MaterialExpressionNamedRerouteUsage_38 {class=MaterialExpressionNamedRerouteUsage}
  MaterialExpressionNamedRerouteUsage_39 {class=MaterialExpressionNamedRerouteUsage}
  MaterialExpressionNamedRerouteUsage_4 {class=MaterialExpressionNamedRerouteUsage}
  MaterialExpressionNamedRerouteUsage_40 {class=MaterialExpressionNamedRerouteUsage}
  MaterialExpressionNamedRerouteUsage_41 {class=MaterialExpressionNamedRerouteUsage}
  MaterialExpressionNamedRerouteUsage_42 {class=MaterialExpressionNamedRerouteUsage}
  MaterialExpressionNamedRerouteUsage_43 {class=MaterialExpressionNamedRerouteUsage}
  MaterialExpressionNamedRerouteUsage_44 {class=MaterialExpressionNamedRerouteUsage}
  MaterialExpressionNamedRerouteUsage_45 {class=MaterialExpressionNamedRerouteUsage}
  MaterialExpressionNamedRerouteUsage_47 {class=MaterialExpressionNamedRerouteUsage}
  MaterialExpressionNamedRerouteUsage_48 {class=MaterialExpressionNamedRerouteUsage}
  MaterialExpressionNamedRerouteUsage_49 {class=MaterialExpressionNamedRerouteUsage}
  MaterialExpressionNamedRerouteUsage_5 {class=MaterialExpressionNamedRerouteUsage}
  MaterialExpressionNamedRerouteUsage_50 {class=MaterialExpressionNamedRerouteUsage}
  MaterialExpressionNamedRerouteUsage_51 {class=MaterialExpressionNamedRerouteUsage}
  MaterialExpressionNamedRerouteUsage_52 {class=MaterialExpressionNamedRerouteUsage}
  MaterialExpressionNamedRerouteUsage_53 {class=MaterialExpressionNamedRerouteUsage}
  MaterialExpressionNamedRerouteUsage_54 {class=MaterialExpressionNamedRerouteUsage}
  MaterialExpressionNamedRerouteUsage_55 {class=MaterialExpressionNamedRerouteUsage}
  MaterialExpressionNamedRerouteUsage_56 {class=MaterialExpressionNamedRerouteUsage}
  MaterialExpressionNamedRerouteUsage_57 {class=MaterialExpressionNamedRerouteUsage}
  MaterialExpressionNamedRerouteUsage_58 {class=MaterialExpressionNamedRerouteUsage}
  MaterialExpressionNamedRerouteUsage_59 {class=MaterialExpressionNamedRerouteUsage}
  MaterialExpressionNamedRerouteUsage_6 {class=MaterialExpressionNamedRerouteUsage}
  MaterialExpressionNamedRerouteUsage_60 {class=MaterialExpressionNamedRerouteUsage}
  MaterialExpressionNamedRerouteUsage_61 {class=MaterialExpressionNamedRerouteUsage}
  MaterialExpressionNamedRerouteUsage_62 {class=MaterialExpressionNamedRerouteUsage}
  MaterialExpressionNamedRerouteUsage_63 {class=MaterialExpressionNamedRerouteUsage}
  MaterialExpressionNamedRerouteUsage_64 {class=MaterialExpressionNamedRerouteUsage}
  MaterialExpressionNamedRerouteUsage_65 {class=MaterialExpressionNamedRerouteUsage}
  MaterialExpressionNamedRerouteUsage_66 {class=MaterialExpressionNamedRerouteUsage}
  MaterialExpressionNamedRerouteUsage_67 {class=MaterialExpressionNamedRerouteUsage}
  MaterialExpressionNamedRerouteUsage_68 {class=MaterialExpressionNamedRerouteUsage}
  MaterialExpressionNamedRerouteUsage_69 {class=MaterialExpressionNamedRerouteUsage}
  MaterialExpressionNamedRerouteUsage_7 {class=MaterialExpressionNamedRerouteUsage}
  MaterialExpressionNamedRerouteUsage_70 {class=MaterialExpressionNamedRerouteUsage}
  MaterialExpressionNamedRerouteUsage_71 {class=MaterialExpressionNamedRerouteUsage}
  MaterialExpressionNamedRerouteUsage_72 {class=MaterialExpressionNamedRerouteUsage}
  MaterialExpressionNamedRerouteUsage_73 {class=MaterialExpressionNamedRerouteUsage}
  MaterialExpressionNamedRerouteUsage_74 {class=MaterialExpressionNamedRerouteUsage}
  MaterialExpressionNamedRerouteUsage_75 {class=MaterialExpressionNamedRerouteUsage}
  MaterialExpressionNamedRerouteUsage_76 {class=MaterialExpressionNamedRerouteUsage}
  MaterialExpressionNamedRerouteUsage_77 {class=MaterialExpressionNamedRerouteUsage}
  MaterialExpressionNamedRerouteUsage_78 {class=MaterialExpressionNamedRerouteUsage}
  MaterialExpressionNamedRerouteUsage_79 {class=MaterialExpressionNamedRerouteUsage}
  MaterialExpressionNamedRerouteUsage_8 {class=MaterialExpressionNamedRerouteUsage}
  MaterialExpressionNamedRerouteUsage_80 {class=MaterialExpressionNamedRerouteUsage}
  MaterialExpressionNamedRerouteUsage_81 {class=MaterialExpressionNamedRerouteUsage}
  MaterialExpressionNamedRerouteUsage_82 {class=MaterialExpressionNamedRerouteUsage}
  MaterialExpressionNamedRerouteUsage_83 {class=MaterialExpressionNamedRerouteUsage}
  MaterialExpressionNamedRerouteUsage_84 {class=MaterialExpressionNamedRerouteUsage}
  MaterialExpressionNamedRerouteUsage_85 {class=MaterialExpressionNamedRerouteUsage}
  MaterialExpressionNamedRerouteUsage_86 {class=MaterialExpressionNamedRerouteUsage}
  MaterialExpressionNamedRerouteUsage_87 {class=MaterialExpressionNamedRerouteUsage}
  MaterialExpressionNamedRerouteUsage_88 {class=MaterialExpressionNamedRerouteUsage}
  MaterialExpressionNamedRerouteUsage_89 {class=MaterialExpressionNamedRerouteUsage}
  MaterialExpressionNamedRerouteUsage_9 {class=MaterialExpressionNamedRerouteUsage}
  MaterialExpressionNamedRerouteUsage_90 {class=MaterialExpressionNamedRerouteUsage}
  MaterialExpressionNamedRerouteUsage_91 {class=MaterialExpressionNamedRerouteUsage}
  MaterialExpressionNamedRerouteUsage_92 {class=MaterialExpressionNamedRerouteUsage}
  MaterialExpressionNamedRerouteUsage_93 {class=MaterialExpressionNamedRerouteUsage}
  MaterialExpressionNamedRerouteUsage_94 {class=MaterialExpressionNamedRerouteUsage}
  MaterialExpressionNamedRerouteUsage_95 {class=MaterialExpressionNamedRerouteUsage}
  MaterialExpressionNamedRerouteUsage_96 {class=MaterialExpressionNamedRerouteUsage}
  MaterialExpressionNamedRerouteUsage_97 {class=MaterialExpressionNamedRerouteUsage}
  MaterialExpressionNamedRerouteUsage_98 {class=MaterialExpressionNamedRerouteUsage}
  MaterialExpressionNamedRerouteUsage_99 {class=MaterialExpressionNamedRerouteUsage}
  MaterialExpressionNormalize_1 {class=MaterialExpressionNormalize, inputs [VectorInput {from=MaterialExpressionCrossProduct_1, output=""}]}
  MaterialExpressionOneMinus_0 {class=MaterialExpressionOneMinus, inputs [None {from=MaterialExpressionMax_6}]}
  MaterialExpressionOneMinus_1 {class=MaterialExpressionOneMinus, inputs [None {from=MaterialExpressionMax_50}]}
  MaterialExpressionOneMinus_10 {class=MaterialExpressionOneMinus, inputs [None {from=MaterialExpressionMax_21}]}
  MaterialExpressionOneMinus_11 {class=MaterialExpressionOneMinus, inputs [None {from=MaterialExpressionMax_49}]}
  MaterialExpressionOneMinus_12 {class=MaterialExpressionOneMinus, inputs [None {from=MaterialExpressionMax_20}]}
  MaterialExpressionOneMinus_13 {class=MaterialExpressionOneMinus, inputs [None {from=MaterialExpressionNamedRerouteUsage_94}]}
  MaterialExpressionOneMinus_14 {class=MaterialExpressionOneMinus, inputs [None {from=MaterialExpressionMax_71}]}
  MaterialExpressionOneMinus_15 {class=MaterialExpressionOneMinus, inputs [None {from=MaterialExpressionMax_68}]}
  MaterialExpressionOneMinus_16 {class=MaterialExpressionOneMinus, inputs [None {from=MaterialExpressionMax_33}]}
  MaterialExpressionOneMinus_17 {class=MaterialExpressionOneMinus, inputs [None {from=MaterialExpressionNamedRerouteUsage_153}]}
  MaterialExpressionOneMinus_18 {class=MaterialExpressionOneMinus, inputs [None {from=MaterialExpressionMax_57}]}
  MaterialExpressionOneMinus_19 {class=MaterialExpressionOneMinus, inputs [None {from=MaterialExpressionMax_47}]}
  MaterialExpressionOneMinus_2 {class=MaterialExpressionOneMinus, inputs [None {from=MaterialExpressionNamedRerouteUsage_142}]}
  MaterialExpressionOneMinus_20 {class=MaterialExpressionOneMinus, inputs [None {from=MaterialExpressionMax_69}]}
  MaterialExpressionOneMinus_21 {class=MaterialExpressionOneMinus, inputs [None {from=MaterialExpressionMax_54}]}
  MaterialExpressionOneMinus_22 {class=MaterialExpressionOneMinus, inputs [None {from=MaterialExpressionMaterialFunctionCall_15, output=DesertGrass_LayerMaskNoise}]}
  MaterialExpressionOneMinus_23 {class=MaterialExpressionOneMinus, inputs [None {from=MaterialExpressionNamedRerouteUsage_359}]}
  MaterialExpressionOneMinus_25 {class=MaterialExpressionOneMinus, inputs [None {from=MaterialExpressionNamedRerouteUsage_310}]}
  MaterialExpressionOneMinus_26 {class=MaterialExpressionOneMinus, inputs [None {from=MaterialExpressionMax_75}]}
  MaterialExpressionOneMinus_3 {class=MaterialExpressionOneMinus, inputs [None {from=MaterialExpressionIf_0}]}
  MaterialExpressionOneMinus_4 {class=MaterialExpressionOneMinus, inputs [None {from=MaterialExpressionMax_40}]}
  MaterialExpressionOneMinus_5 {class=MaterialExpressionOneMinus, inputs [None {from=MaterialExpressionNamedRerouteUsage_47}]}
  MaterialExpressionOneMinus_6 {class=MaterialExpressionOneMinus, inputs [None {from=MaterialExpressionMax_51}]}
  MaterialExpressionOneMinus_7 {class=MaterialExpressionOneMinus, inputs [None {from=MaterialExpressionMax_55}]}
  MaterialExpressionOneMinus_8 {class=MaterialExpressionOneMinus, inputs [None {from=MaterialExpressionMax_10}]}
  MaterialExpressionOneMinus_9 {class=MaterialExpressionOneMinus, inputs [None {from=MaterialExpressionMax_15}]}
  MaterialExpressionPinBase_10 {class=MaterialExpressionPinBase, inputs [], properties {}}
  MaterialExpressionPinBase_11 {
   class=MaterialExpressionPinBase
   inputs [Output {from=MaterialExpressionMultiply_24, output=""}]
   properties {reroutePins [Output {expression=&self:MaterialExpressionReroute_9}]}
  }
  MaterialExpressionPinBase_16 {class=MaterialExpressionPinBase, inputs [], properties {}}
  MaterialExpressionPinBase_17 {
   class=MaterialExpressionPinBase
   inputs [Output {from=MaterialExpressionMultiply_18, output=""}]
   properties {reroutePins [Output {expression=&self:MaterialExpressionReroute_10}]}
  }
  MaterialExpressionPinBase_18 {class=MaterialExpressionPinBase, inputs [], properties {}}
  MaterialExpressionPinBase_19 {
   class=MaterialExpressionPinBase
   inputs [Output {from=MaterialExpressionMultiply_17, output=""}]
   properties {reroutePins [Output {expression=&self:MaterialExpressionReroute_11}]}
  }
  MaterialExpressionPinBase_20 {class=MaterialExpressionPinBase, inputs [], properties {}}
  MaterialExpressionPinBase_21 {
   class=MaterialExpressionPinBase
   inputs [Output {from=MaterialExpressionNamedRerouteUsage_78, output=""}]
   properties {reroutePins [Output {expression=&self:MaterialExpressionReroute_12}]}
  }
  MaterialExpressionPinBase_22 {class=MaterialExpressionPinBase, inputs [], properties {}}
  MaterialExpressionPinBase_23 {
   class=MaterialExpressionPinBase
   inputs [Output {from=MaterialExpressionMultiply_15, output=""}]
   properties {reroutePins [Output {expression=&self:MaterialExpressionReroute_13}]}
  }
  MaterialExpressionPinBase_24 {
   class=MaterialExpressionPinBase
   inputs []
   properties {reroutePins [
     "HighEndMaterial(MA)" {expression=&self:MaterialExpressionReroute_20}
     "MobileMaterial(MA)" {expression=&self:MaterialExpressionReroute_21}
    ]}
  }
  MaterialExpressionPinBase_25 {
   class=MaterialExpressionPinBase
   inputs [Output {from=MaterialExpressionQualitySwitch_2, output=""}]
   properties {reroutePins [Output {expression=&self:MaterialExpressionReroute_18}]}
  }
  MaterialExpressionPower_0 {
   class=MaterialExpressionPower
   inputs [
    Base {from=MaterialExpressionDotProduct_0}
    Exp {from=MaterialExpressionVectorParameter_2, output=A}
   ]
  }
  MaterialExpressionQualitySwitch_0 {
   class=MaterialExpressionQualitySwitch
   inputs [
    Default {from=MaterialExpressionReroute_16, output=""}
    Low {from="", output=""}
    High {from="", output=""}
    Medium {from=MaterialExpressionVirtualTextureFeatureSwitch_0, output=""}
    Epic {from="", output=""}
   ]
   properties {}
  }
  MaterialExpressionQualitySwitch_1 {
   class=MaterialExpressionQualitySwitch
   inputs [
    Default {from=MaterialExpressionSetMaterialAttributes_2, output=""}
    Low {from="", output=""}
    High {from="", output=""}
    Medium {from="", output=""}
    Epic {from=MaterialExpressionSetMaterialAttributes_0, output=""}
   ]
   properties {}
  }
  MaterialExpressionQualitySwitch_2 {
   class=MaterialExpressionQualitySwitch
   inputs [
    Default {from=MaterialExpressionStaticSwitchParameter_9, output=""}
    Low {from=MaterialExpressionReroute_16, output=""}
    High {from="", output=""}
    Medium {from="", output=""}
    Epic {from="", output=""}
   ]
   properties {}
  }
  MaterialExpressionReroute_0 {class=MaterialExpressionReroute, inputs [None {from=MaterialExpressionMax_52}]}
  MaterialExpressionReroute_1 {class=MaterialExpressionReroute, inputs [None {from=MaterialExpressionMaterialFunctionCall_22, output=QuadA_RGBA}]}
  MaterialExpressionReroute_10 {class=MaterialExpressionReroute, inputs [None {from=MaterialExpressionMultiply_18}]}
  MaterialExpressionReroute_11 {class=MaterialExpressionReroute, inputs [None {from=MaterialExpressionMultiply_17}]}
  MaterialExpressionReroute_12 {class=MaterialExpressionReroute, inputs [None {from=MaterialExpressionNamedRerouteUsage_78}]}
  MaterialExpressionReroute_13 {class=MaterialExpressionReroute, inputs [None {from=MaterialExpressionMultiply_15}]}
  MaterialExpressionReroute_16 {class=MaterialExpressionReroute, inputs [None {from=MaterialExpressionMaterialFunctionCall_36, output=ResultMaterial}]}
  MaterialExpressionReroute_18 {class=MaterialExpressionReroute, inputs [None {from=MaterialExpressionQualitySwitch_2}]}
  MaterialExpressionReroute_2 {class=MaterialExpressionReroute, inputs [None {from=MaterialExpressionMaterialFunctionCall_22, output=QuadB_RGBA}]}
  MaterialExpressionReroute_20 {class=MaterialExpressionReroute, inputs [None {from=MaterialExpressionNamedRerouteUsage_136}]}
  MaterialExpressionReroute_21 {class=MaterialExpressionReroute, inputs [None {from=MaterialExpressionNamedRerouteUsage_137}]}
  MaterialExpressionReroute_3 {class=MaterialExpressionReroute, inputs [None {from=MaterialExpressionMaterialFunctionCall_13, output=Result}]}
  MaterialExpressionReroute_4 {class=MaterialExpressionReroute, inputs [None {from=MaterialExpressionMaterialFunctionCall_22, output=QuadC_RGBA}]}
  MaterialExpressionReroute_9 {class=MaterialExpressionReroute, inputs [None {from=MaterialExpressionMultiply_24}]}
  MaterialExpressionRound_3 {class=MaterialExpressionRound, inputs [None {from=MaterialExpressionMultiply_34, output=""}], properties {}}
  MaterialExpressionRuntimeVirtualTextureOutput_1 {
   class=MaterialExpressionRuntimeVirtualTextureOutput
   inputs [
    BaseColor {from=MaterialExpressionGetMaterialAttributes_2, output=Base Color}
    Specular {from=MaterialExpressionGetMaterialAttributes_2, output=Base Color}
    Roughness {from=MaterialExpressionGetMaterialAttributes_2, output=Base Color}
    Normal {from=MaterialExpressionGetMaterialAttributes_2, output=Base Color}
    WorldHeight {from=MaterialExpressionWorldPosition_4, output=Z}
    Opacity {from="", output=""}
    Mask {from="", output=""}
    Displacement {from=MaterialExpressionGetMaterialAttributes_2, output=Base Color}
    Mask4 {from="", output=""}
   ]
   properties {}
  }
  MaterialExpressionRuntimeVirtualTextureSampleParameter_0 {
   class=MaterialExpressionRuntimeVirtualTextureSampleParameter
   inputs [
    UVs {from="", output=""}
    Translated World Position {from="", output=""}
   ]
   properties {
    group=RVT
    materialType=BaseColor_Normal_Specular_YCoCg
    parameterName=RVT Sample Target
    virtualTexture=/Game/Athena/Apollo/Maps/Landscape/VT/RVT_Apollo_Landscape
    worldPositionOriginType=CameraRelative
   }
  }
  MaterialExpressionRuntimeVirtualTextureSampleParameter_1 {
   class=MaterialExpressionRuntimeVirtualTextureSampleParameter
   inputs [
    UVs {from="", output=""}
    Translated World Position {from="", output=""}
   ]
   properties {
    group=RVT
    materialType=Displacement
    parameterName=RVT Displacement
    virtualTexture=/Game/Athena/Apollo/Maps/Landscape/VT/RVT_Helios_Displacement
    worldPositionOriginType=CameraRelative
   }
  }
  MaterialExpressionSaturate_0 {class=MaterialExpressionSaturate, inputs [None {from=MaterialExpressionDivide_8}]}
  MaterialExpressionSaturate_1 {class=MaterialExpressionSaturate, inputs [None {from=MaterialExpressionLinearInterpolate_1}]}
  MaterialExpressionSaturate_10 {class=MaterialExpressionSaturate, inputs [None {from=MaterialExpressionMultiply_29}]}
  MaterialExpressionSaturate_11 {class=MaterialExpressionSaturate, inputs [None {from=MaterialExpressionAdd_4}]}
  MaterialExpressionSaturate_12 {class=MaterialExpressionSaturate, inputs [None {from=MaterialExpressionMultiply_44}]}
  MaterialExpressionSaturate_13 {class=MaterialExpressionSaturate, inputs [None {from=MaterialExpressionAdd_2}]}
  MaterialExpressionSaturate_14 {class=MaterialExpressionSaturate, inputs [None {from=MaterialExpressionAdd_0}]}
  MaterialExpressionSaturate_15 {class=MaterialExpressionSaturate, inputs [None {from=MaterialExpressionLinearInterpolate_8}]}
  MaterialExpressionSaturate_16 {class=MaterialExpressionSaturate, inputs [None {from=MaterialExpressionLinearInterpolate_13}]}
  MaterialExpressionSaturate_17 {class=MaterialExpressionSaturate, inputs [None {from=MaterialExpressionMultiply_4}]}
  MaterialExpressionSaturate_18 {class=MaterialExpressionSaturate, inputs [None {from=MaterialExpressionScalarParameter_11}]}
  MaterialExpressionSaturate_2 {class=MaterialExpressionSaturate, inputs [None {from=MaterialExpressionMultiply_53}]}
  MaterialExpressionSaturate_20 {class=MaterialExpressionSaturate, inputs [None {from=MaterialExpressionMultiply_45}]}
  MaterialExpressionSaturate_23 {class=MaterialExpressionSaturate, inputs [None {from=MaterialExpressionPower_0}]}
  MaterialExpressionSaturate_24 {class=MaterialExpressionSaturate, inputs [None {from=MaterialExpressionLinearInterpolate_6}]}
  MaterialExpressionSaturate_29 {class=MaterialExpressionSaturate, inputs [None {from=MaterialExpressionAdd_1}]}
  MaterialExpressionSaturate_3 {class=MaterialExpressionSaturate, inputs [None {from=MaterialExpressionMultiply_39}]}
  MaterialExpressionSaturate_31 {class=MaterialExpressionSaturate, inputs [None {from=MaterialExpressionMultiply_5}]}
  MaterialExpressionSaturate_4 {class=MaterialExpressionSaturate, inputs [None {from=MaterialExpressionMultiply_54}]}
  MaterialExpressionSaturate_5 {class=MaterialExpressionSaturate, inputs [None {from=MaterialExpressionMultiply_10}]}
  MaterialExpressionSaturate_6 {class=MaterialExpressionSaturate, inputs [None {from=MaterialExpressionMultiply_70}]}
  MaterialExpressionSaturate_7 {class=MaterialExpressionSaturate, inputs [None {from=MaterialExpressionMultiply_67}]}
  MaterialExpressionSaturate_8 {class=MaterialExpressionSaturate, inputs [None {from=MaterialExpressionMultiply_9}]}
  MaterialExpressionSaturate_9 {class=MaterialExpressionSaturate, inputs [None {from=MaterialExpressionAdd_3}]}
  MaterialExpressionScalarParameter_0 {class=MaterialExpressionScalarParameter, properties {parameterName=DEBUG_AsBaseColor}}
  MaterialExpressionScalarParameter_1 {class=MaterialExpressionScalarParameter, properties {defaultValue=2, group=GrassMeshes, parameterName=L0CC_MeshMask_Mod}}
  MaterialExpressionScalarParameter_10 {class=MaterialExpressionScalarParameter, properties {defaultValue=1, group=Displacement, parameterName=Displacement_BASE_Mod}}
  MaterialExpressionScalarParameter_11 {
   class=MaterialExpressionScalarParameter
   properties {
    desc="A value of zero = default displacement input. A value of 1 = a scalar displacement value of .5 which means no displacement."
    group=Displacement
    parameterName=Displacement_BASE_DEBUG
    sortPriority=-9999
   }
  }
  MaterialExpressionScalarParameter_12 {class=MaterialExpressionScalarParameter, properties {defaultValue=1, group=EchoFrost, parameterName=EchoFrost_DefaultLayers_High_RoughMult}}
  MaterialExpressionScalarParameter_13 {class=MaterialExpressionScalarParameter, properties {defaultValue=10, group=EchoFrost, parameterName=EchoFrost_DefaultLayers_RoughMaskMult}}
  MaterialExpressionScalarParameter_14 {class=MaterialExpressionScalarParameter, properties {defaultValue=-2, group=EchoFrost, parameterName=EchoFrost_DefaultLayers_RoughMaskAdd}}
  MaterialExpressionScalarParameter_2 {class=MaterialExpressionScalarParameter, properties {defaultValue=-1, group=GrassMeshes, parameterName=L0CC_MeshMask_Add}}
  MaterialExpressionScalarParameter_21 {class=MaterialExpressionScalarParameter, properties {defaultValue=1, group=Cliff, parameterName=SnowCliff_DEBUGMe2, sortPriority=98}}
  MaterialExpressionScalarParameter_22 {class=MaterialExpressionScalarParameter, properties {defaultValue=1, group=Cliff, parameterName=SnowCliff_DEBUGMe3, sortPriority=98}}
  MaterialExpressionScalarParameter_23 {class=MaterialExpressionScalarParameter, properties {defaultValue=0.5, parameterName=Mud_DarkTest}}
  MaterialExpressionScalarParameter_24 {class=MaterialExpressionScalarParameter, properties {defaultValue=2, group=GrassMeshes, parameterName=DesertGrass_GrassMeshMask02_Mod}}
  MaterialExpressionScalarParameter_25 {class=MaterialExpressionScalarParameter, properties {defaultValue=-0.5, group=GrassMeshes, parameterName=DesertGrass_GrassMeshMask02_Add}}
  MaterialExpressionScalarParameter_26 {class=MaterialExpressionScalarParameter, properties {defaultValue=1, group=Cliff, parameterName=SnowCliff_DEBUGMe4, sortPriority=98}}
  MaterialExpressionScalarParameter_27 {class=MaterialExpressionScalarParameter, properties {defaultValue=1, group=Cliff, parameterName=SnowCliff_DEBUGMe5, sortPriority=98}}
  MaterialExpressionScalarParameter_28 {class=MaterialExpressionScalarParameter, properties {group=Cliff, parameterName=SnowCliff_DEBUGMe6, sortPriority=98}}
  MaterialExpressionScalarParameter_3 {
   class=MaterialExpressionScalarParameter
   properties {
    defaultValue=100
    desc="This offsets the ocean level used by the sand. Positive numbers go lower, negative numbers make it go higher."
    group=GrassMeshes
    parameterName=Beach_LGTMaskDistance
   }
  }
  MaterialExpressionScalarParameter_37 {class=MaterialExpressionScalarParameter, properties {defaultValue=45000, group=FlowerGroup, parameterName=FlowerGroup_MaskScale}}
  MaterialExpressionScalarParameter_4 {
   class=MaterialExpressionScalarParameter
   properties {defaultValue=0.35, group=EchoFrost, parameterName=EchoFrost_DefaultLayers_Low_RoughMult}
  }
  MaterialExpressionScalarParameter_49 {
   class=MaterialExpressionScalarParameter
   properties {
    defaultValue=0.5
    desc="This offsets the ocean level used by the sand. Positive numbers go lower, negative numbers make it go higher."
    group=GrassMeshes
    parameterName=Beach_LGTMaskOffset
   }
  }
  MaterialExpressionScalarParameter_5 {class=MaterialExpressionScalarParameter, properties {defaultValue=2, group=GrassMeshes, parameterName=L0CA_MeshMask_Pow}}
  MaterialExpressionScalarParameter_6 {class=MaterialExpressionScalarParameter, properties {defaultValue=-0.5, group=GrassMeshes, parameterName=L0CA_MeshMask_Add}}
  MaterialExpressionScalarParameter_7 {class=MaterialExpressionScalarParameter, properties {defaultValue=4, group=GrassMeshes, parameterName=L0CB_MeshMask_Pow}}
  MaterialExpressionScalarParameter_8 {class=MaterialExpressionScalarParameter, properties {defaultValue=-0.75, group=GrassMeshes, parameterName=L0CB_MeshMask_Add}}
  MaterialExpressionScalarParameter_9 {
   class=MaterialExpressionScalarParameter
   properties {defaultValue=0.75, desc=Set this to 1 to turn it on, group=Debug, parameterName=DEBUG-TraversalMask}
  }
  MaterialExpressionSetMaterialAttributes_0 {
   class=MaterialExpressionSetMaterialAttributes
   inputs [
    MaterialAttributes {from="", output=""}
    Base Color {from=MaterialExpressionRuntimeVirtualTextureSampleParameter_0, output=BaseColor}
    Specular {from=MaterialExpressionRuntimeVirtualTextureSampleParameter_0, output=BaseColor}
    Roughness {from=MaterialExpressionRuntimeVirtualTextureSampleParameter_0, output=BaseColor}
    Normal {from=MaterialExpressionRuntimeVirtualTextureSampleParameter_0, output=BaseColor}
    Displacement {from=MaterialExpressionShadingPathSwitch_5, output=""}
   ]
   properties {attributeSetTypes [
     #2
     #3
     #4
     #5
     #6
    ]}
  }
  MaterialExpressionSetMaterialAttributes_1 {
   class=MaterialExpressionSetMaterialAttributes
   inputs [
    MaterialAttributes {from=MaterialExpressionGetMaterialAttributes_1, output=MaterialAttributes}
    Base Color {from=MaterialExpressionNamedRerouteUsage_377, output=""}
   ]
   properties {attributeSetTypes [#2]}
  }
  MaterialExpressionSetMaterialAttributes_2 {
   class=MaterialExpressionSetMaterialAttributes
   inputs [
    MaterialAttributes {from="", output=""}
    Base Color {from=MaterialExpressionRuntimeVirtualTextureSampleParameter_0, output=BaseColor}
    Specular {from=MaterialExpressionRuntimeVirtualTextureSampleParameter_0, output=BaseColor}
    Roughness {from=MaterialExpressionRuntimeVirtualTextureSampleParameter_0, output=BaseColor}
    Normal {from=MaterialExpressionRuntimeVirtualTextureSampleParameter_0, output=BaseColor}
   ]
   properties {attributeSetTypes [
     #2
     #3
     #4
     #5
    ]}
  }
  MaterialExpressionSetMaterialAttributes_3 {
   class=MaterialExpressionSetMaterialAttributes
   inputs [
    MaterialAttributes {from=MaterialExpressionGetMaterialAttributes_3, output=MaterialAttributes}
    Base Color {from=MaterialExpressionLinearInterpolate_12, output=""}
   ]
   properties {attributeSetTypes [#2]}
  }
  MaterialExpressionSetMaterialAttributes_4 {
   class=MaterialExpressionSetMaterialAttributes
   inputs [
    MaterialAttributes {from=MaterialExpressionGetMaterialAttributes_4, output=MaterialAttributes}
    Specular {from=MaterialExpressionNamedRerouteUsage_148, output=""}
   ]
   properties {attributeSetTypes [#3]}
  }
  MaterialExpressionSetMaterialAttributes_5 {
   class=MaterialExpressionSetMaterialAttributes
   inputs [
    MaterialAttributes {from=MaterialExpressionGetMaterialAttributes_6, output=MaterialAttributes}
    Base Color {from=MaterialExpressionLinearInterpolate_0, output=""}
    Specular {from=MaterialExpressionGetMaterialAttributes_6, output=MaterialAttributes}
    Roughness {from=MaterialExpressionLinearInterpolate_4, output=""}
    Normal {from=MaterialExpressionGetMaterialAttributes_6, output=MaterialAttributes}
   ]
   properties {attributeSetTypes [
     #2
     #3
     #4
     #5
    ]}
  }
  MaterialExpressionSetMaterialAttributes_8 {
   class=MaterialExpressionSetMaterialAttributes
   inputs [
    MaterialAttributes {from=MaterialExpressionGetMaterialAttributes_7, output=MaterialAttributes}
    Base Color {from=MaterialExpressionLinearInterpolate_3, output=""}
   ]
   properties {attributeSetTypes [#2]}
  }
  MaterialExpressionSetMaterialAttributes_9 {
   class=MaterialExpressionSetMaterialAttributes
   inputs [
    MaterialAttributes {from=MaterialExpressionMaterialFunctionCall_13, output=Result}
    Opacity Mask {from=MaterialExpressionLandscapeVisibilityMask_1, output=""}
   ]
   properties {attributeSetTypes [#1]}
  }
  MaterialExpressionShadingPathSwitch_0 {
   class=MaterialExpressionShadingPathSwitch
   inputs [
    Default {from=MaterialExpressionReroute_20, output="HighEndMaterial(MA)"}
    Deferred {from="", output=""}
    Forward {from="", output=""}
    Mobile {from=MaterialExpressionMaterialFunctionCall_36, output=ResultMaterial}
   ]
   properties {}
  }
  MaterialExpressionShadingPathSwitch_1 {
   class=MaterialExpressionShadingPathSwitch
   inputs [
    Default {from=MaterialExpressionVirtualTextureFeatureSwitch_0, output=""}
    Deferred {from="", output=""}
    Forward {from="", output=""}
    Mobile {from=MaterialExpressionQualitySwitch_0, output=""}
   ]
   properties {}
  }
  MaterialExpressionShadingPathSwitch_2 {
   class=MaterialExpressionShadingPathSwitch
   inputs [
    Default {from=MaterialExpressionMultiply_23, output=""}
    Deferred {from="", output=""}
    Forward {from="", output=""}
    Mobile {from=MaterialExpressionConstant_26, output=""}
   ]
   properties {}
  }
  MaterialExpressionShadingPathSwitch_4 {
   class=MaterialExpressionShadingPathSwitch
   inputs [
    Default {from=MaterialExpressionStaticSwitch_47, output=""}
    Deferred {from="", output=""}
    Forward {from="", output=""}
    Mobile {from=MaterialExpressionReroute_18, output=Output}
   ]
   properties {}
  }
  MaterialExpressionShadingPathSwitch_5 {
   class=MaterialExpressionShadingPathSwitch
   inputs [
    Default {from=MaterialExpressionRuntimeVirtualTextureSampleParameter_1, output=Displacement}
    Deferred {from="", output=""}
    Forward {from="", output=""}
    Mobile {from="", output=""}
   ]
   properties {}
  }
  MaterialExpressionStaticBoolParameter_0 {
   class=MaterialExpressionStaticBoolParameter
   inputs []
   properties {
    desc="Workaround for LobbyIsland so at least the layers can\r\nbe painted until the RVT is setup correctly"
    group=Temp
    parameterName=IsLobbyIsland
   }
  }
  MaterialExpressionStaticBoolParameter_1 {class=MaterialExpressionStaticBoolParameter, inputs [], properties {group=LayerUsage, parameterName=UseSnowForBiomeC, sortPriority=-500}}
  MaterialExpressionStaticBoolParameter_10 {class=MaterialExpressionStaticBoolParameter, inputs [], properties {group=LayerUsage, parameterName=UseGrass_L0CC, sortPriority=-970}}
  MaterialExpressionStaticBoolParameter_11 {
   class=MaterialExpressionStaticBoolParameter
   inputs []
   properties {desc=Only select one option per layer as true, group=LayerUsage, parameterName=UseBiomeA03_DesertWavySand, sortPriority=-780}
  }
  MaterialExpressionStaticBoolParameter_12 {
   class=MaterialExpressionStaticBoolParameter
   inputs []
   properties {desc=Only select one option per layer as true, group=LayerUsage, parameterName=UseAsphalt04, sortPriority=-870}
  }
  MaterialExpressionStaticBoolParameter_13 {
   class=MaterialExpressionStaticBoolParameter
   inputs []
   properties {desc=Only select one option per layer as true, group=LayerUsage, parameterName=UseBiomeC_GoldlandsPostRVT, sortPriority=-490}
  }
  MaterialExpressionStaticBoolParameter_14 {
   class=MaterialExpressionStaticBoolParameter
   inputs []
   properties {desc=Only select one option per layer as true, group=LayerUsage, parameterName=UseBiomeC_Goldlands, sortPriority=-490}
  }
  MaterialExpressionStaticBoolParameter_15 {
   class=MaterialExpressionStaticBoolParameter
   inputs []
   properties {desc=Only select one option per layer as true, group=LayerUsage, parameterName=UseBiomeC_GoldlandsPostRVT, sortPriority=-490}
  }
  MaterialExpressionStaticBoolParameter_16 {
   class=MaterialExpressionStaticBoolParameter
   inputs []
   properties {desc=Only select one option per layer as true, group=LayerUsage, parameterName=UseBiomeC_ArchipelagoSeaweed, sortPriority=-700}
  }
  MaterialExpressionStaticBoolParameter_17 {
   class=MaterialExpressionStaticBoolParameter
   inputs []
   properties {desc=Only select one option per layer as true, group=LayerUsage, parameterName=UseBiomeD_EchoFrost, sortPriority=-700}
  }
  MaterialExpressionStaticBoolParameter_2 {class=MaterialExpressionStaticBoolParameter, inputs [], properties {group=LayerUsage, parameterName=UseGrass_L0CA, sortPriority=-990}}
  MaterialExpressionStaticBoolParameter_3 {class=MaterialExpressionStaticBoolParameter, inputs [], properties {group=LayerUsage, parameterName=UseGrass_L0CB, sortPriority=-980}}
  MaterialExpressionStaticBoolParameter_4 {
   class=MaterialExpressionStaticBoolParameter
   inputs []
   properties {defaultValue=true, desc=Only select one option per layer as true, group=LayerUsage, parameterName=UseBiomeA_ForestFloor, sortPriority=-800}
  }
  MaterialExpressionStaticBoolParameter_5 {
   class=MaterialExpressionStaticBoolParameter
   inputs []
   properties {desc=Only select one option per layer as true, group=LayerUsage, parameterName=UseBiomeA02, sortPriority=-790}
  }
  MaterialExpressionStaticBoolParameter_6 {
   class=MaterialExpressionStaticBoolParameter
   inputs []
   properties {desc=Only select one option per layer as true, group=LayerUsage, parameterName=UseAsphalt02_Wetlands_MossyForest, sortPriority=-890}
  }
  MaterialExpressionStaticBoolParameter_7 {
   class=MaterialExpressionStaticBoolParameter
   inputs []
   properties {defaultValue=true, desc=Only select one option per layer as true, group=LayerUsage, parameterName=UseBiomeB_Scree, sortPriority=-700}
  }
  MaterialExpressionStaticBoolParameter_8 {
   class=MaterialExpressionStaticBoolParameter
   inputs []
   properties {desc=Only select one option per layer as true, group=LayerUsage, parameterName=UseAsphalt03, sortPriority=-880}
  }
  MaterialExpressionStaticBoolParameter_9 {
   class=MaterialExpressionStaticBoolParameter
   inputs []
   properties {defaultValue=true, desc=Only select one option per layer as true, group=LayerUsage, parameterName=UseAsphalt, sortPriority=-900}
  }
  MaterialExpressionStaticBool_0 {class=MaterialExpressionStaticBool, properties {value=true}}
  MaterialExpressionStaticSwitchParameter_0 {
   class=MaterialExpressionStaticSwitchParameter
   inputs [
    True {from=MaterialExpressionMaterialFunctionCall_29, output=Result}
    False {from="", output=""}
   ]
   properties {group=Temp, parameterName=UseTempLayersForBiomeMasks}
  }
  MaterialExpressionStaticSwitchParameter_1 {
   class=MaterialExpressionStaticSwitchParameter
   inputs [
    True {from=MaterialExpressionMeshPartitionChannelSample_1, output=""}
    False {from=MaterialExpressionLandscapeLayerSample_0, output=""}
   ]
   properties {parameterName=IsMeshPartition}
  }
  MaterialExpressionStaticSwitchParameter_2 {
   class=MaterialExpressionStaticSwitchParameter
   inputs [
    True {from=MaterialExpressionMeshPartitionChannelSample_2, output=""}
    False {from=MaterialExpressionLandscapeLayerSample_2, output=""}
   ]
   properties {parameterName=IsMeshPartition}
  }
  MaterialExpressionStaticSwitchParameter_3 {
   class=MaterialExpressionStaticSwitchParameter
   inputs [
    True {from=MaterialExpressionMeshPartitionChannelSample_3, output=""}
    False {from=MaterialExpressionLandscapeLayerSample_1, output=""}
   ]
   properties {parameterName=IsMeshPartition}
  }
  MaterialExpressionStaticSwitchParameter_4 {
   class=MaterialExpressionStaticSwitchParameter
   inputs [
    True {from=MaterialExpressionMeshPartitionChannelSample_4, output=""}
    False {from=MaterialExpressionLandscapeLayerSample_3, output=""}
   ]
   properties {parameterName=IsMeshPartition}
  }
  MaterialExpressionStaticSwitchParameter_5 {
   class=MaterialExpressionStaticSwitchParameter
   inputs [
    True {from=MaterialExpressionMeshPartitionChannelSample_5, output=""}
    False {from=MaterialExpressionLandscapeLayerSample_4, output=""}
   ]
   properties {parameterName=IsMeshPartition}
  }
  MaterialExpressionStaticSwitchParameter_6 {
   class=MaterialExpressionStaticSwitchParameter
   inputs [
    True {from=MaterialExpressionMeshPartitionChannelSample_6, output=""}
    False {from=MaterialExpressionLandscapeLayerSample_5, output=""}
   ]
   properties {parameterName=IsMeshPartition}
  }
  MaterialExpressionStaticSwitchParameter_7 {
   class=MaterialExpressionStaticSwitchParameter
   inputs [
    True {from=MaterialExpressionMeshPartitionChannelSample_7, output=""}
    False {from=MaterialExpressionLandscapeLayerSample_6, output=""}
   ]
   properties {parameterName=IsMeshPartition}
  }
  MaterialExpressionStaticSwitchParameter_8 {
   class=MaterialExpressionStaticSwitchParameter
   inputs [
    True {from=MaterialExpressionMeshPartitionChannelSample_12, output=""}
    False {from=MaterialExpressionLandscapeLayerSample_8, output=""}
   ]
   properties {parameterName=IsMeshPartition}
  }
  MaterialExpressionStaticSwitchParameter_9 {
   class=MaterialExpressionStaticSwitchParameter
   inputs [
    True {from=MaterialExpressionVirtualTextureFeatureSwitch_0, output=""}
    False {from=MaterialExpressionReroute_20, output="HighEndMaterial(MA)"}
   ]
   properties {defaultValue=true, group=RVT, parameterName=EnableRVT}
  }
  MaterialExpressionStaticSwitch_0 {
   class=MaterialExpressionStaticSwitch
   inputs [
    True {from=MaterialExpressionMaterialFunctionCall_26, output=BlendedMaterial}
    False {from=MaterialExpressionNamedRerouteUsage_190, output=""}
    Value {from=MaterialExpressionNamedRerouteDeclaration_23, output=""}
   ]
   properties {}
  }
  MaterialExpressionStaticSwitch_1 {
   class=MaterialExpressionStaticSwitch
   inputs [
    True {from=MaterialExpressionConstant_0, output=""}
    False {from=MaterialExpressionMaterialFunctionCall_43, output=Result}
    Value {from=MaterialExpressionNamedRerouteUsage_10, output=""}
   ]
   properties {}
  }
  MaterialExpressionStaticSwitch_10 {
   class=MaterialExpressionStaticSwitch
   inputs [
    True {from=MaterialExpressionMaterialFunctionCall_15, output=BlendedMaterial}
    False {from=MaterialExpressionNamedRerouteUsage_195, output=""}
    Value {from=MaterialExpressionStaticBoolParameter_10, output=""}
   ]
   properties {}
  }
  MaterialExpressionStaticSwitch_11 {
   class=MaterialExpressionStaticSwitch
   inputs [
    True {from=MaterialExpressionMaterialFunctionCall_19, output=BlendedMaterial}
    False {from=MaterialExpressionNamedRerouteUsage_177, output=""}
    Value {from=MaterialExpressionStaticBoolParameter_2, output=""}
   ]
   properties {}
  }
  MaterialExpressionStaticSwitch_12 {
   class=MaterialExpressionStaticSwitch
   inputs [
    True {from=MaterialExpressionMaterialFunctionCall_15, output=MobileBlendedMaterial}
    False {from=MaterialExpressionNamedRerouteUsage_204, output=""}
    Value {from=MaterialExpressionStaticBoolParameter_10, output=""}
   ]
   properties {}
  }
  MaterialExpressionStaticSwitch_13 {
   class=MaterialExpressionStaticSwitch
   inputs [
    True {from=MaterialExpressionMaterialFunctionCall_19, output=MobileBlendedMaterial}
    False {from=MaterialExpressionNamedRerouteUsage_178, output=""}
    Value {from=MaterialExpressionStaticBoolParameter_2, output=""}
   ]
   properties {}
  }
  MaterialExpressionStaticSwitch_14 {
   class=MaterialExpressionStaticSwitch
   inputs [
    True {from=MaterialExpressionMaterialFunctionCall_26, output=MobileBlendedMaterial}
    False {from=MaterialExpressionNamedRerouteUsage_191, output=""}
    Value {from=MaterialExpressionNamedRerouteDeclaration_23, output=""}
   ]
   properties {}
  }
  MaterialExpressionStaticSwitch_15 {
   class=MaterialExpressionStaticSwitch
   inputs [
    True {from=MaterialExpressionMaterialFunctionCall_33, output=BlendedMaterial}
    False {from=MaterialExpressionNamedRerouteUsage_182, output=""}
    Value {from=MaterialExpressionStaticBoolParameter_3, output=""}
   ]
   properties {}
  }
  MaterialExpressionStaticSwitch_16 {
   class=MaterialExpressionStaticSwitch
   inputs [
    True {from=MaterialExpressionMax_73, output=""}
    False {from=MaterialExpressionConstant_25, output=""}
    Value {from=MaterialExpressionNamedRerouteUsage_274, output=""}
   ]
   properties {}
  }
  MaterialExpressionStaticSwitch_17 {
   class=MaterialExpressionStaticSwitch
   inputs [
    True {from=MaterialExpressionMaterialFunctionCall_33, output=MobileBlendedMaterial}
    False {from=MaterialExpressionNamedRerouteUsage_183, output=""}
    Value {from=MaterialExpressionStaticBoolParameter_3, output=""}
   ]
   properties {}
  }
  MaterialExpressionStaticSwitch_18 {
   class=MaterialExpressionStaticSwitch
   inputs [
    True {from=MaterialExpressionSaturate_13, output=""}
    False {from=MaterialExpressionConstant_10, output=""}
    Value {from=MaterialExpressionStaticBoolParameter_10, output=""}
   ]
   properties {}
  }
  MaterialExpressionStaticSwitch_19 {
   class=MaterialExpressionStaticSwitch
   inputs [
    True {from=MaterialExpressionMaterialFunctionCall_27, output=BlendedMaterial}
    False {from=MaterialExpressionNamedRerouteUsage_231, output=""}
    Value {from=MaterialExpressionNamedRerouteDeclaration_106, output=""}
   ]
   properties {}
  }
  MaterialExpressionStaticSwitch_2 {
   class=MaterialExpressionStaticSwitch
   inputs [
    True {from=MaterialExpressionNamedRerouteUsage_192, output=""}
    False {from=MaterialExpressionConstant_6, output=""}
    Value {from=MaterialExpressionNamedRerouteDeclaration_23, output=""}
   ]
   properties {}
  }
  MaterialExpressionStaticSwitch_20 {
   class=MaterialExpressionStaticSwitch
   inputs [
    True {from=MaterialExpressionMaterialFunctionCall_27, output=MobileBlendedMaterial}
    False {from=MaterialExpressionNamedRerouteUsage_233, output=""}
    Value {from=MaterialExpressionNamedRerouteDeclaration_106, output=""}
   ]
   properties {}
  }
  MaterialExpressionStaticSwitch_21 {
   class=MaterialExpressionStaticSwitch
   inputs [
    True {from=MaterialExpressionMaterialFunctionCall_39, output=Result}
    False {from=MaterialExpressionConstant_0, output=""}
    Value {from=MaterialExpressionNamedRerouteUsage_10, output=""}
   ]
   properties {}
  }
  MaterialExpressionStaticSwitch_22 {
   class=MaterialExpressionStaticSwitch
   inputs [
    True {from=MaterialExpressionMaterialFunctionCall_35, output=BlendedMaterial}
    False {from=MaterialExpressionNamedRerouteUsage_207, output=""}
    Value {from=MaterialExpressionNamedRerouteDeclaration_78, output=""}
   ]
   properties {}
  }
  MaterialExpressionStaticSwitch_23 {
   class=MaterialExpressionStaticSwitch
   inputs [
    True {from=MaterialExpressionMaterialFunctionCall_35, output=MobileBlendedMaterial}
    False {from=MaterialExpressionNamedRerouteUsage_208, output=""}
    Value {from=MaterialExpressionNamedRerouteDeclaration_78, output=""}
   ]
   properties {}
  }
  MaterialExpressionStaticSwitch_24 {
   class=MaterialExpressionStaticSwitch
   inputs [
    True {from=MaterialExpressionMaterialFunctionCall_35, output=LayerMask}
    False {from=MaterialExpressionConstant_13, output=""}
    Value {from=MaterialExpressionNamedRerouteDeclaration_78, output=""}
   ]
   properties {}
  }
  MaterialExpressionStaticSwitch_25 {
   class=MaterialExpressionStaticSwitch
   inputs [
    True {from=MaterialExpressionMaterialFunctionCall_9, output=BlendedMaterial}
    False {from=MaterialExpressionNamedRerouteUsage_141, output=""}
    Value {from=MaterialExpressionNamedRerouteDeclaration_95, output=""}
   ]
   properties {}
  }
  MaterialExpressionStaticSwitch_26 {
   class=MaterialExpressionStaticSwitch
   inputs [
    True {from=MaterialExpressionMaterialFunctionCall_9, output=MobileBlendedMaterial}
    False {from=MaterialExpressionNamedRerouteUsage_189, output=""}
    Value {from=MaterialExpressionNamedRerouteDeclaration_95, output=""}
   ]
   properties {}
  }
  MaterialExpressionStaticSwitch_27 {
   class=MaterialExpressionStaticSwitch
   inputs [
    True {from=MaterialExpressionMaterialFunctionCall_9, output=LayerMask}
    False {from=MaterialExpressionConstant_43, output=""}
    Value {from=MaterialExpressionNamedRerouteDeclaration_95, output=""}
   ]
   properties {}
  }
  MaterialExpressionStaticSwitch_28 {
   class=MaterialExpressionStaticSwitch
   inputs [
    True {from=MaterialExpressionMaterialFunctionCall_31, output=BlendedMaterial}
    False {from=MaterialExpressionNamedRerouteUsage_278, output=""}
    Value {from=MaterialExpressionNamedRerouteDeclaration_86, output=""}
   ]
   properties {}
  }
  MaterialExpressionStaticSwitch_29 {
   class=MaterialExpressionStaticSwitch
   inputs [
    True {from=MaterialExpressionMaterialFunctionCall_31, output=MobileBlendedMaterial}
    False {from=MaterialExpressionNamedRerouteUsage_279, output=""}
    Value {from=MaterialExpressionNamedRerouteDeclaration_86, output=""}
   ]
   properties {}
  }
  MaterialExpressionStaticSwitch_3 {
   class=MaterialExpressionStaticSwitch
   inputs [
    True {from=MaterialExpressionMaterialFunctionCall_0, output=BlendedMaterial}
    False {from=MaterialExpressionNamedRerouteUsage_266, output=""}
    Value {from=MaterialExpressionNamedRerouteDeclaration_72, output=""}
   ]
   properties {}
  }
  MaterialExpressionStaticSwitch_30 {
   class=MaterialExpressionStaticSwitch
   inputs [
    True {from=MaterialExpressionMaterialFunctionCall_31, output=LayerMask}
    False {from=MaterialExpressionConstant_38, output=""}
    Value {from=MaterialExpressionNamedRerouteDeclaration_86, output=""}
   ]
   properties {}
  }
  MaterialExpressionStaticSwitch_31 {
   class=MaterialExpressionStaticSwitch
   inputs [
    True {from=MaterialExpressionMaterialFunctionCall_27, output=LayerMask}
    False {from=MaterialExpressionConstant_66, output=""}
    Value {from=MaterialExpressionNamedRerouteDeclaration_106, output=""}
   ]
   properties {}
  }
  MaterialExpressionStaticSwitch_32 {
   class=MaterialExpressionStaticSwitch
   inputs [
    True {from=MaterialExpressionMaterialFunctionCall_38, output=BlendedMaterial}
    False {from=MaterialExpressionNamedRerouteUsage_281, output=""}
    Value {from=MaterialExpressionNamedRerouteDeclaration_109, output=""}
   ]
   properties {}
  }
  MaterialExpressionStaticSwitch_33 {
   class=MaterialExpressionStaticSwitch
   inputs [
    True {from=MaterialExpressionMaterialFunctionCall_38, output=MobileBlendedMaterial}
    False {from=MaterialExpressionNamedRerouteUsage_283, output=""}
    Value {from=MaterialExpressionNamedRerouteDeclaration_109, output=""}
   ]
   properties {}
  }
  MaterialExpressionStaticSwitch_34 {
   class=MaterialExpressionStaticSwitch
   inputs [
    True {from=MaterialExpressionMaterialFunctionCall_7, output=BlendedMaterial}
    False {from=MaterialExpressionNamedRerouteUsage_264, output=""}
    Value {from=MaterialExpressionNamedRerouteDeclaration_98, output=""}
   ]
   properties {}
  }
  MaterialExpressionStaticSwitch_35 {
   class=MaterialExpressionStaticSwitch
   inputs [
    True {from=MaterialExpressionMaterialFunctionCall_7, output=MobileBlendedMaterial}
    False {from=MaterialExpressionNamedRerouteUsage_265, output=""}
    Value {from=MaterialExpressionNamedRerouteDeclaration_98, output=""}
   ]
   properties {}
  }
  MaterialExpressionStaticSwitch_36 {
   class=MaterialExpressionStaticSwitch
   inputs [
    True {from=MaterialExpressionMaterialFunctionCall_7, output=LayerMask}
    False {from=MaterialExpressionConstant_44, output=""}
    Value {from=MaterialExpressionNamedRerouteDeclaration_98, output=""}
   ]
   properties {}
  }
  MaterialExpressionStaticSwitch_37 {
   class=MaterialExpressionStaticSwitch
   inputs [
    True {from=MaterialExpressionMaterialFunctionCall_38, output=LayerMask}
    False {from=MaterialExpressionConstant_69, output=""}
    Value {from=MaterialExpressionNamedRerouteDeclaration_109, output=""}
   ]
   properties {}
  }
  MaterialExpressionStaticSwitch_38 {
   class=MaterialExpressionStaticSwitch
   inputs [
    True {from=MaterialExpressionMaterialFunctionCall_24, output=BlendedMaterial}
    False {from=MaterialExpressionNamedRerouteUsage_244, output=""}
    Value {from=MaterialExpressionNamedRerouteDeclaration_90, output=""}
   ]
   properties {}
  }
  MaterialExpressionStaticSwitch_39 {
   class=MaterialExpressionStaticSwitch
   inputs [
    True {from=MaterialExpressionMaterialFunctionCall_24, output=MobileBlendedMaterial}
    False {from=MaterialExpressionNamedRerouteUsage_245, output=""}
    Value {from=MaterialExpressionNamedRerouteDeclaration_90, output=""}
   ]
   properties {}
  }
  MaterialExpressionStaticSwitch_40 {
   class=MaterialExpressionStaticSwitch
   inputs [
    True {from=MaterialExpressionMaterialFunctionCall_24, output=LayerMask}
    False {from=MaterialExpressionConstant_39, output=""}
    Value {from=MaterialExpressionNamedRerouteDeclaration_90, output=""}
   ]
   properties {}
  }
  MaterialExpressionStaticSwitch_41 {
   class=MaterialExpressionStaticSwitch
   inputs [
    True {from=MaterialExpressionMaterialFunctionCall_24, output=GrassMeshMask}
    False {from=MaterialExpressionConstant_40, output=""}
    Value {from=MaterialExpressionNamedRerouteDeclaration_90, output=""}
   ]
   properties {}
  }
  MaterialExpressionStaticSwitch_42 {
   class=MaterialExpressionStaticSwitch
   inputs [
    True {from=MaterialExpressionMaterialFunctionCall_26, output=GrassMask}
    False {from=MaterialExpressionConstant_64, output=""}
    Value {from=MaterialExpressionNamedRerouteDeclaration_23, output=""}
   ]
   properties {}
  }
  MaterialExpressionStaticSwitch_43 {
   class=MaterialExpressionStaticSwitch
   inputs [
    True {from=MaterialExpressionMaterialFunctionCall_37, output=BlendedMaterial}
    False {from=MaterialExpressionNamedRerouteUsage_234, output=""}
    Value {from=MaterialExpressionNamedRerouteDeclaration_118, output=""}
   ]
   properties {}
  }
  MaterialExpressionStaticSwitch_44 {
   class=MaterialExpressionStaticSwitch
   inputs [
    True {from=MaterialExpressionMaterialFunctionCall_37, output=MobileBlendedMaterial}
    False {from=MaterialExpressionNamedRerouteUsage_256, output=""}
    Value {from=MaterialExpressionNamedRerouteDeclaration_118, output=""}
   ]
   properties {}
  }
  MaterialExpressionStaticSwitch_45 {
   class=MaterialExpressionStaticSwitch
   inputs [
    True {from=MaterialExpressionMaterialFunctionCall_37, output=LayerMask}
    False {from=MaterialExpressionConstant_76, output=""}
    Value {from=MaterialExpressionNamedRerouteDeclaration_118, output=""}
   ]
   properties {}
  }
  MaterialExpressionStaticSwitch_46 {
   class=MaterialExpressionStaticSwitch
   inputs [
    True {from=MaterialExpressionNamedRerouteUsage_244, output=""}
    False {from=MaterialExpressionStaticSwitch_38, output=""}
    Value {from=MaterialExpressionStaticBoolParameter_13, output=""}
   ]
   properties {}
  }
  MaterialExpressionStaticSwitch_47 {
   class=MaterialExpressionStaticSwitch
   inputs [
    True {from=MaterialExpressionMaterialFunctionCall_28, output=BlendedMaterial}
    False {from=MaterialExpressionReroute_18, output=Output}
    Value {from=MaterialExpressionStaticBoolParameter_15, output=""}
   ]
   properties {}
  }
  MaterialExpressionStaticSwitch_48 {
   class=MaterialExpressionStaticSwitch
   inputs [
    True {from=MaterialExpressionMaterialFunctionCall_40, output=BlendedMaterial}
    False {from=MaterialExpressionNamedRerouteUsage_305, output=""}
    Value {from=MaterialExpressionNamedRerouteDeclaration_114, output=""}
   ]
   properties {}
  }
  MaterialExpressionStaticSwitch_49 {
   class=MaterialExpressionStaticSwitch
   inputs [
    True {from=MaterialExpressionMaterialFunctionCall_40, output=MobileBlendedMaterial}
    False {from=MaterialExpressionNamedRerouteUsage_306, output=""}
    Value {from=MaterialExpressionNamedRerouteDeclaration_114, output=""}
   ]
   properties {}
  }
  MaterialExpressionStaticSwitch_5 {
   class=MaterialExpressionStaticSwitch
   inputs [
    True {from=MaterialExpressionSaturate_9, output=""}
    False {from=MaterialExpressionConstant_17, output=""}
    Value {from=MaterialExpressionStaticBoolParameter_2, output=""}
   ]
   properties {}
  }
  MaterialExpressionStaticSwitch_50 {
   class=MaterialExpressionStaticSwitch
   inputs [
    True {from=MaterialExpressionMaterialFunctionCall_40, output=LayerMask}
    False {from=MaterialExpressionConstant_41, output=""}
    Value {from=MaterialExpressionNamedRerouteDeclaration_114, output=""}
   ]
   properties {}
  }
  MaterialExpressionStaticSwitch_51 {
   class=MaterialExpressionStaticSwitch
   inputs [
    True {from=MaterialExpressionMaterialFunctionCall_44, output=BlendedMaterial}
    False {from=MaterialExpressionNamedRerouteUsage_319, output=""}
    Value {from=MaterialExpressionNamedRerouteDeclaration_122, output=""}
   ]
   properties {}
  }
  MaterialExpressionStaticSwitch_52 {
   class=MaterialExpressionStaticSwitch
   inputs [
    True {from=MaterialExpressionMaterialFunctionCall_44, output=MobileBlendedMaterial}
    False {from=MaterialExpressionNamedRerouteUsage_321, output=""}
    Value {from=MaterialExpressionNamedRerouteDeclaration_122, output=""}
   ]
   properties {}
  }
  MaterialExpressionStaticSwitch_53 {
   class=MaterialExpressionStaticSwitch
   inputs [
    True {from=MaterialExpressionMaterialFunctionCall_44, output=LayerMask}
    False {from=MaterialExpressionConstant_46, output=""}
    Value {from=MaterialExpressionNamedRerouteDeclaration_122, output=""}
   ]
   properties {}
  }
  MaterialExpressionStaticSwitch_55 {
   class=MaterialExpressionStaticSwitch
   inputs [
    True {from=MaterialExpressionMultiply_73, output=""}
    False {from=MaterialExpressionConstant_74, output=""}
    Value {from=MaterialExpressionStaticBoolParameter_10, output=""}
   ]
   properties {}
  }
  MaterialExpressionStaticSwitch_6 {
   class=MaterialExpressionStaticSwitch
   inputs [
    True {from=MaterialExpressionMaterialFunctionCall_0, output=MobileBlendedMaterial}
    False {from=MaterialExpressionNamedRerouteUsage_267, output=""}
    Value {from=MaterialExpressionNamedRerouteDeclaration_72, output=""}
   ]
   properties {}
  }
  MaterialExpressionStaticSwitch_7 {
   class=MaterialExpressionStaticSwitch
   inputs [
    True {from=MaterialExpressionMaterialFunctionCall_1, output=Result}
    False {from=MaterialExpressionReroute_1, output=""}
    Value {from=MaterialExpressionNamedRerouteUsage_29, output=""}
   ]
   properties {}
  }
  MaterialExpressionStaticSwitch_8 {
   class=MaterialExpressionStaticSwitch
   inputs [
    True {from=MaterialExpressionMaterialFunctionCall_0, output=LayerMask}
    False {from=MaterialExpressionConstant_12, output=""}
    Value {from=MaterialExpressionNamedRerouteDeclaration_72, output=""}
   ]
   properties {}
  }
  MaterialExpressionStaticSwitch_9 {
   class=MaterialExpressionStaticSwitch
   inputs [
    True {from=MaterialExpressionSaturate_11, output=""}
    False {from=MaterialExpressionConstant_29, output=""}
    Value {from=MaterialExpressionStaticBoolParameter_3, output=""}
   ]
   properties {}
  }
  MaterialExpressionTextureSample_6 {
   class=MaterialExpressionTextureSample
   inputs [
    UVs {from=MaterialExpressionMaterialFunctionCall_34, output=Result}
    Tex {from=""}
    Apply View MipBias {from=""}
   ]
   properties {
    samplerSource=SSM_Wrap_WorldGroupSettings
    samplerType=SAMPLERTYPE_LinearColor
    texture=/BRCoreMaterials/Hermes/Terrain/Landscape/Textures/T_FlowerGroup01_LSGrassTypeMask
   }
  }
  MaterialExpressionVectorParameter_0 {class=MaterialExpressionVectorParameter, properties {parameterName=DEBUG_AsBaseColorColor, defaultValue=<rgba 1 1 0 1>}}
  MaterialExpressionVectorParameter_1 {class=MaterialExpressionVectorParameter, properties {group=Debug, parameterName=DEBUG-TraversalMaskColor, defaultValue=<rgba 1 0 0 1>}}
  MaterialExpressionVectorParameter_2 {
   class=MaterialExpressionVectorParameter
   properties {group=EchoFrost, parameterName=EchoFrost_DefaultLayers_SnowColor, defaultValue=<rgba 0.385 0.474 0.55 5>}
  }
  MaterialExpressionVectorParameter_3 {
   class=MaterialExpressionVectorParameter
   properties {group=EchoFrost, parameterName=EchoFrost_DefaultLayers_SnowColorDeep, defaultValue=<rgba 0.05 0.077 0.1 0.5>}
  }
  MaterialExpressionVectorParameter_4 {
   class=MaterialExpressionVectorParameter
   properties {
    desc=Set the value to 1 for the biome you want to preview
    parameterName=Preview Biome Masks
    channelNames {a=Desert, b=Boreal, g=RedWoods, r=Chaparral}
   }
  }
  MaterialExpressionVirtualTextureFeatureSwitch_0 {
   class=MaterialExpressionVirtualTextureFeatureSwitch
   inputs [
    No {from=MaterialExpressionReroute_16, output=""}
    Yes {from=MaterialExpressionQualitySwitch_1, output=""}
   ]
   properties {}
  }
  MaterialExpressionWorldPosition_0 {class=MaterialExpressionWorldPosition, inputs [], properties {}}
  MaterialExpressionWorldPosition_1 {class=MaterialExpressionWorldPosition, inputs [], properties {}}
  MaterialExpressionWorldPosition_4 {class=MaterialExpressionWorldPosition, inputs [], properties {}}
 }
 properties {
  bEnableDisplacementFade=true
  bEnableTessellation=true
  bTangentSpaceNormal=false
  bUseLightmapDirectionality=false
  bUseMaterialAttributes=true
  bUsedWithNanite=true
  bUsedWithVirtualHeightfieldMesh=true
  floatPrecisionMode=MFPM_Full_MaterialExpressionOnly
  pixelDepthOffsetMode=PDOM_Legacy
  previewMesh=/Hera/Sets/SeaMist/Meshes/SM_SeaMist_Lighthouse_FresnelLens_A
  shadingModel=MSM_FromMaterialExpression
  usedShadingModels=Default Lit
  displacementFadeRange {endSizePixels=3, startSizePixels=8}
  displacementScaling {center=0.5, magnitude=16}
 }
}
