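# Surface material for the Hera mesh terrain partition, named for use on its HLOD meshes.
# Its only node is a runtime virtual texture sample parameter defaulting to the LobbyIsland base color, normal and specular RVT; no link to material outputs appears.
# It is Default Lit with world-space normals, enables Nanite and virtual heightfield mesh usage, and carries displacement scaling and fade settings.
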Material /Hera/Environment/Terrain/MeshTerrain/Materials/M_Hera_MeshPartition_HLOD
graph {
 properties {
  bTangentSpaceNormal=false
  bUsedWithNanite=true
  bUsedWithVirtualHeightfieldMesh=true
  pixelDepthOffsetMode=PDOM_Legacy
  usedShadingModels=Default Lit
  displacementFadeRange {endSizePixels=3, startSizePixels=8}
  displacementScaling {center=0.5, magnitude=0.15}
 }
 nodes [RuntimeVirtualTextureSampleParameter!(UVs, Translated World Position) {
   group=RVT
   materialType=BaseColor_Normal_Specular_YCoCg
   parameterName=RVT Sample Target
   virtualTexture=/Hera/Environment/Terrain/MeshTerrain/VirtualTextures/RVT_Hera_MeshPartition_LobbyIsland
   worldPositionOriginType=CameraRelative
  }]
}
